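# Niagara idle effect for the Polished Jade female character body, adding eye glow with lights, horn flames and glow, and body flames.
# Burst-spawned particles follow skeletal mesh samples, with per-partition arrays shaping each flame and smoothed positions fed to materials.
# Variant selectors recolor effects and swap Flames for MeltVariant_Flames.
# Timeline fade, time of day and front-end state drive brightness, while user parameters handle glow, helmet, hidden parts and low LOD.
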NiagaraSystem /BRCosmetics/Characters/Player/Female/Medium/Bodies/F_MED_PolishedJade_Mind/FX/NS_Character_PolishedJade_Body_Idle
dependencies {
 usedDataInterfaces [
  &NiagaraDataInterfaceArrayFloat
  &NiagaraDataInterfaceArrayFloat3
  &NiagaraDataInterfaceArrayInt32
  &NiagaraDataInterfaceCurve
  &NiagaraDataInterfaceSkeletalMesh
 ]
 usedDynamicInputs [
  &SetBoolByFloatComparison
  &FN_Select_Float_from_Time_Of_Day
  /BRCosmetics/Weapons/FORT_Melee/Pickaxe_GateHound/FX/VariantManager/VariantVectorSelector
  &FN_Select_Vector_from_Time_Of_Day
 ]
 usedModules [
  &self:SystemSpawnScript.SystemScriptSource.SystemScriptGraph.NiagaraNodeAssignment_0.SetVariables_#1
  &self:SystemSpawnScript.SystemScriptSource.SystemScriptGraph.NiagaraNodeAssignment_3.SetVariables_#2
  &self:SystemSpawnScript.SystemScriptSource.SystemScriptGraph.NiagaraNodeAssignment_2.SetVariables_#3
  &self:SystemSpawnScript.SystemScriptSource.SystemScriptGraph.NiagaraNodeAssignment_4.SetVariables_#4
  &SystemState
  &Timeline
  &FN_VariantManager
  &self:SystemSpawnScript.SystemScriptSource.SystemScriptGraph.NiagaraNodeAssignment_1.SetVariables_#5
  &self:SystemSpawnScript.SystemScriptSource.SystemScriptGraph.NiagaraNodeAssignment_5.SetVariables_#6
 ]
}
emitters {
 EyeGlow {
  emitterUpdateScript [
   &EmitterState(Life Cycle Mode=::NewEnumerator1, Loop Behavior=::NewEnumerator1, Loop Duration Mode=::NewEnumerator1, UseLoopDelay=false)
   &SpawnBurst_Instantaneous(
    Spawn Count=&Multiply_Int(
     A=2
     B=&Multiply_Int(
      A=&MakeIntFromFloat(Float=$User.GlowOn)
      B=&MakeIntFromFloat(Float=$User.HelmetOn)
     )
    )
    Use Spawn Probability=false
    UseLoopCountLimit=true
   )
  ]
  particleSpawnScript [
   &PartitionParticles()
   &InitializeParticle(
    Sprite Rotation Angle=&SelectFloatFromArray(Direct Array Index=$Particles.PartitionParticles.Partition, Float Selection Array=DataInterface() {propertyValues {FloatData [15,-15]}})
    Sprite Rotation Mode=::NewEnumerator0
    Sprite Size=<4.5 2.5>
    Sprite Size Mode=::NewEnumerator4
    UsePositionOffset=false
    Write Material Random=true
   )
   &CameraOffset(Camera Offset Amount=2)
  ]
  particleUpdateScript [
   &ParticleState(
    Kill Particles When Lifetime Has Elapsed=false
    Let Infinitely Lived Particles Die When Emitter Deactivates=true
    Loop Particles Lifetime=true
   )
   &Color(Color=&MultiplyLinearColor(
     Linear Color=&VariantColorSelector(
      Variant Amount=1
      Original Color=<rgba 0.2203 0.9 0.7754 1>
      Variant 1 Color=<rgba 0.9323 0.6167 0.806 1>
      Variant 2 Color=<rgba 0.0551 0.6719 0.5588 1>
     )
     Scale Factor=&MakeFloat4FromVectorAndFloat("Vector (RGB)"=&Multiply_VectorByFloat(Vector=$System.ToD, Float=1), "Float (Alpha)"=1.5)
    ))
   /SetVariables(Particles.AlphaScale:NiagaraFloat=&FN_FrontEndFloatModifier(
     Gameplay=&Lerp_Float(A=1, B=1.6, Alpha=$System.Timeline.NormalizedTime)
     Front End=&MakeCustomFloatFromBool(Bool=$User.bIsFrontEnd, True Float=$System.FrontEndGlowLoop, False Float=0)
    )) {isSetParameters=true}
   &ScaleColor(
    Color Value To Scale=$Particles.Color
    Scale Alpha=&Multiply_Float(
     A=$Particles.AlphaScale
     B=&MakeCustomFloatFromBool(Bool=$User.bisHeadHidden, True Float=0, False Float=1)
    )
    Scale RGB=&VectorFromFloat(Value=1)
    ScaleA=true
    ScaleRGB=true
   )
   &DynamicMaterialParameters(
    Index 0 Param 1=2
    Index 0 Param 2=0.5
    Module.Index0Param0Write=true
    Module.Index0Param1Write=true
    Module.Index0Param2Write=true
    Module.Index0Param3Write=true
   )
   &FN_Light_Attributes(
    Color=$Particles.Color
    Create Color Attribute=true
    Create Exponent Attribute=true
    Create Position Attribute=true
    Create Radius Attribute=true
    Enable Particle Light Enable=true
    Exponent=1
    Light Enabled=true
    Position=&AddVectorToPosition(
     Position=$Particles.Position
     Vector=&RotateVectorByQuaternion(Vector To Rotate=<0 4 -1>, Quaternion=$Particles.SkeletalMeshLocation.SampledRotation)
    )
    Radius=3
    Scale Color=&Multiply_Float(
     A=&Multiply_Float(
      A=/BRCosmetics/Weapons/FORT_Melee/Pickaxe_GateHound/FX/VariantManager/VariantFloatSelector(
       Original Float:NiagaraFloat=25
       Variant 1 Float:NiagaraFloat=4
       Variant 2 Float:NiagaraFloat=25
       Variant 3 Float:NiagaraFloat=?
       Variant Amount:NiagaraInt32=1
      )
      B=&FN_FrontEndFloatModifier(Gameplay=&Lerp_Float(A=0.2, B=1, Alpha=$System.Timeline.NormalizedTime), Front End=0)
     )
     B=&MakeCustomFloatFromBool(Bool=$User.bisHeadHidden, True Float=0, False Float=1)
    )
   )
   &FN_Gen5_Light(
    Attenutation=3
    Color:LinearColor=$Particles.Color
    Intensity:NiagaraFloat=&Multiply_Float(
     A=&Multiply_Float(
      A=&VariantFloatSelector(
       Original Float:NiagaraFloat=20
       Variant 1 Float:NiagaraFloat=3.5
       Variant 2 Float:NiagaraFloat=20
       Variant 3 Float:NiagaraFloat=?
       Variant Amount:NiagaraInt32=1
      )
      B=&FN_FrontEndFloatModifier(Gameplay=&Lerp_Float(A=0.2, B=1, Alpha=$System.Timeline.NormalizedTime), Front End=0)
     )
     B=&MakeCustomFloatFromBool(Bool=$User.bisHeadHidden, True Float=0, False Float=1)
    )
    Position Offset:Vector3f=&RotateVectorByQuaternion(Vector To Rotate=<0 1.2 -1>, Quaternion=$Particles.SkeletalMeshLocation.SampledRotation)
   )
   &SkeletalMeshLocation(
    Bone Sampling Mode=::NewEnumerator4
    MaskSpawnGroup=false
    Sampled Position Offset=&Lerp_Vector(
     A=<3.2 7.8 3.2>
     B=<3.2 7.8 -3.2>
     Alpha=&MakeFloatFromInt(Int=$Particles.PartitionParticles.Partition)
    )
    Skeletal Mesh=$System.Skeletal Mesh
   )
   &SpriteFacingAndAlignment(Alignment Coordinate Space=::World, Sprite Alignment=<0 0 1>, WriteAlignment=true, WriteFacing=true)
  ]
  rendererClasses [
   &NiagaraSpriteRendererProperties
   &NiagaraLightRendererProperties
   &NiagaraComponentRendererProperties
  ]
  renderers [
   {rendererClass=&NiagaraSpriteRendererProperties}
   {rendererClass=&NiagaraLightRendererProperties, rendererIndex=1}
   {rendererClass=&NiagaraComponentRendererProperties, rendererIndex=2}
  ]
 }
 Flames {
  emitterUpdateScript [
   &EmitterState(
    Inactive Response=::NewEnumerator1
    Life Cycle Mode=::NewEnumerator1
    Loop Behavior=::NewEnumerator1
    Loop Duration Mode=::NewEnumerator1
    UseLoopDelay=false
   )
   &SpawnBurst_Instantaneous(
    Spawn Count=&Multiply_Int(
     A=&MakeIntFromFloat(Float=&VariantFloatSelector(
       Original Float:NiagaraFloat=25
       Variant 1 Float:NiagaraFloat=0
       Variant 2 Float:NiagaraFloat=25
       Variant 3 Float:NiagaraFloat=?
       Variant Amount:NiagaraInt32=1
      ))
     B=&MakeIntFromFloat(Float=$User.GlowOn)
    )
    Use Spawn Probability=false
    UseLoopCountLimit=false
   )
  ]
  particleSpawnScript [
   &PartitionParticles(Number Of Partitions=25)
   &InitializeParticle(
    Color=&MultiplyLinearColor(
     Linear Color=&VariantColorSelector(
      Variant Amount=1
      Original Color=<rgba 0.0018 0.0885 0.0857 1>
      Variant 1 Color=<rgba 0.4844 0.0807 0.37 0.1>
      Variant 2 Color=<rgba 0.125 0.0182 0.1125 1>
     )
     Scale Factor=&MakeFloat4FromVectorAndFloat("Vector (RGB)"=$System.ToD, "Float (Alpha)"=1)
    )
    Lifetime Max=12
    Lifetime Min=12
    Lifetime Mode=::NewEnumerator1
    Mesh Scale Mode=::NewEnumerator3
    Mesh Uniform Scale=&SelectFloatFromArray(
     Direct Array Index=$Particles.PartitionParticles.Partition
     Float Selection Array=DataInterface() {propertyValues {FloatData [1,0.5,0.4,0.3,0.5,1,0.45,0.35,1,0.5,1.1,0.8,0.3,1.5,0.5,0.8,0.4,0.25,0.5,0.35,0.2,0.65,0.2,0.5,0.2]}}
    )
    Sprite Size Mode=::NewEnumerator3
    Uniform Sprite Size=&SelectFloatFromArray(
     Direct Array Index=$Particles.PartitionParticles.Partition
     Float Selection Array=DataInterface() {propertyValues {FloatData [4,4,4,4,4,4,4,4,4,4,4,4,4,4,4,4,4,4,0,0,0,0,0,0,0,0,0]}}
    )
    UsePositionOffset=false
    Write Material Random=true
   )
   &InitialMeshOrientation(
    Mesh Axis To Orient=<0 0 1>
    Mesh Orientation Mode=::NewEnumerator2
    Orientation Coordinate Space=::World
    Orientation Vector=&SelectVectorFromArray(
     Direct Array Index=$Particles.PartitionParticles.Partition
     Vector Selection Array=DataInterface() {propertyValues {FloatData [
        <0 0 1>
        <0 0 1>
        <0 0 1>
        <0 0 1>
        <0 0 1>
        <0 0 -1>
        <0 0 -1>
        <0 0 -1>
        <0 0 -1>
        <0 0 -1>
        <-1 0 0>
        <-1 0 0>
        <-1 0 0>
        <-1 0 0>
        <-1 0 0>
        <-1 0 0>
        <-1 0 0>
        <0 0 1>
       ]}}
    )
    Rotation=<0.25 0 0>
    Use Rotation Vector=false
   )
   /SetVariables(Particles.VisibilityTag=&SelectIntFromArray(Direct Array Index=$Particles.PartitionParticles.Partition, Int Selection Array=DataInterface() {propertyValues {IntData [0,0,0,1,1,0,0]}})) {enabled=false, isSetParameters=true, module=SetVariables_#7}
   &DynamicMaterialParameters(
    Index 0 Param 1=&RandomRangeFloat(Minimum=0, Maximum=1)
    Index 0 Param 2=&SelectFloatFromArray(
     Direct Array Index=$Particles.PartitionParticles.Partition
     Float Selection Array=DataInterface() {propertyValues {FloatData [3,1,0.6,1.2,0.8,3,1,1,0.5,1,1,1,0.5,3,1.5,0.8,1,0.8,1,0.75,0.5,1.2,0.8,1.1,0.5]}}
    )
    Module.Index0Param0Write=true
    Module.Index0Param1Write=true
    Module.Index0Param2Write=true
    Module.Index0Param3Write=true
   )
   &CameraOffset(Camera Offset Amount=0.5)
  ]
  particleUpdateScript [
   &ParticleState(Kill Particles When Lifetime Has Elapsed=false, Let Infinitely Lived Particles Die When Emitter Deactivates=true)
   /SetVariables(Particles.AlphaScale:NiagaraFloat=&FN_FrontEndFloatModifier(
     Gameplay=&Lerp_Float(A=0, B=1, Alpha=$System.Timeline.NormalizedTime)
     Front End=&MakeCustomFloatFromBool(Bool=$User.bIsFrontEnd, True Float=$System.FrontEndGlowLoop, False Float=0)
    )) {isSetParameters=true, module=SetVariables_#8}
   &ScaleMeshSize(Scale Factor=<1 1 2>)
   &SolveForcesAndVelocity(Clamp Velocity=false, Limit Acceleration=false)
   &SkeletalMeshLocation(
    Bone / Socket Index=&ReturnExecIndex()
    MaskSpawnGroup=false
    Mesh Sampling Type=::NewEnumerator5
    Skeletal Mesh=$System.Skeletal Mesh
    Socket Sampling Mode=::NewEnumerator5
   )
   &SkeletalMeshLocation(
    Bone / Socket Index=&ReturnExecIndex()
    MaskSpawnGroup=false
    Mesh Sampling Type=::NewEnumerator5
    Orientation Sampling=::NewEnumerator1
    Skeletal Mesh=$System.Skeletal Mesh Low
    Socket Sampling Mode=::NewEnumerator5
   ) {module=SkeletalMeshLocation001}
   /SetVariables(Particles.Position=&MakeCustomVectorFromBool(
     Bool=$User.bisLowLOD
     True Vector=$Particles.SkeletalMeshLocation001.SampledPosition
     False Vector=$Particles.SkeletalMeshLocation.SampledPosition
    )) {isSetParameters=true}
   &ScaleColor(Scale Alpha=$Particles.AlphaScale, Scale RGB=&VectorFromFloat(Value=3.5), ScaleA=true, ScaleRGB=true)
   /SetVariables(Particles.Color_02:LinearColor=&MultiplyLinearColor(
     Linear Color=&MultiplyLinearColor(
      Linear Color=&VariantColorSelector(
       Variant Amount=1
       Original Color=<rgba 0.0018 0.0885 0.0857 1>
       Variant 1 Color=<rgba 0.8594 0.4073 0.7313 0.05>
       Variant 2 Color=<rgba 0.2812 0.041 0.2452 0.5>
      )
      Scale Factor=&MakeFloat4FromVectorAndFloat("Vector (RGB)"=$System.ToD, "Float (Alpha)"=1)
     )
     Scale Factor=&MakeFloat4FromVectorAndFloat("Vector (RGB)"=<1 1 1>, "Float (Alpha)"=$Particles.AlphaScale)
    )) {isSetParameters=true, module=SetVariables_#9}
   /SetVariables(Particles.Color_03:LinearColor=&MultiplyLinearColor(
     Linear Color=&MultiplyLinearColor(
      Linear Color=&VariantColorSelector(
       Variant Amount=1
       Original Color=<rgba 0.2203 0.9 0.7754 1>
       Variant 1 Color=<rgba 0.9323 0.6167 0.806 0.5>
       Variant 2 Color=<rgba 0.6879 0.3357 0.816 1>
      )
      Scale Factor=&MakeFloat4FromVectorAndFloat("Vector (RGB)"=$System.ToD, "Float (Alpha)"=$Particles.AlphaScale)
     )
     Scale Factor=&MakeFloat4FromVectorAndFloat("Vector (RGB)"=<1 1 1>, "Float (Alpha)"=1.5)
    )) {isSetParameters=true, module=SetVariables_#10}
   &InterpolateOverTime(Rate Of Change=3, Target Position Value:NiagaraPosition=$Particles.Position, Variable Type=::NewEnumerator6)
   /SetVariables(Particles.Interpolate Dynamic=&MakeFloat4FromVectorAndFloat(
     "Vector (RGB)"=$Particles.InterpolateOverTime.InterpolateOverTime_Position.Moving Average
     "Float (Alpha)"=&Add_Float(A=$User.Interpolate Intensity, B=-0.5)
    )) {isSetParameters=true, module=SetVariables_#11}
   &FN_Sprite_Attributes(
    Camera Offset:NiagaraFloat=&SelectFloatFromArray(
     Direct Array Index=$Particles.PartitionParticles.Partition
     Float Selection Array=DataInterface() {propertyValues {FloatData [2,2,5,5,5,2,2,5,0,5,2,1.5,1.5,1,1,1.5,1,1.5]}}
    )
    Color=&MultiplyLinearColor(
     Linear Color=&VariantColorSelector(
      Variant Amount=1
      Original Color=<rgba 0.0473 0.4323 0.3617 1>
      Variant 1 Color=<rgba 0.8594 0.2014 0.673 0.2>
      Variant 2 Color=<rgba 0.2812 0.041 0.2452 1>
     )
     Scale Factor=&MakeFloat4FromVectorAndFloat("Vector (RGB)"=$System.ToD, "Float (Alpha)"=&Multiply_Float(A=0.3, B=$Particles.AlphaScale))
    )
    Create Camera Offset Attribute=true
    Create Color Attribute=true
    Write to Module Namespace=true
   )
   /SetVariables(Particles.AirglowDyn:Vector4f=<2 1 0 0>) {isSetParameters=true, module=SetVariables_#12}
   /SetVariables(
    Particles.DecalDyn:Vector4f=&MakeVector4(
     X=0.8
     Y=0.5
     Z=&RandomRangeFloat(Minimum=0, Maximum=1)
     W=&RandomRangeFloat(Minimum=0, Maximum=0)
    )
    Particles.FireDecalColor:LinearColor=&MultiplyLinearColor(
     Linear Color=&Multiply_LinearColorByFloat(
      Linear Color=&VariantColorSelector(Variant Amount=1, Original Color=<rgba 0.1721 0.6323 0.7031 1>, Variant 1 Color=<rgba 0.8594 0.2014 0.673 1>, Variant 2 Color=<rgba 1 1 1 1>)
      Float=&Multiply_Float(
       A=1.5
       B=&SelectFloatFromArray(
        Direct Array Index=$Particles.PartitionParticles.Partition
        Float Selection Array=DataInterface() {propertyValues {FloatData [1,1.5,1.5,1.5,0.8,1.2,1,1.2,1,0.8,1.2,1.2,1.8,1.8,1.5,1.2,2,1]}}
       )
      )
     )
     Scale Factor=&MakeFloat4FromVectorAndFloat("Vector (RGB)"=$System.ToD, "Float (Alpha)"=$Particles.AlphaScale)
    )
    Particles.FireDecalSize:Vector2f=&Multiply_Vector2D_ByFloat(
     Float=&Multiply_Float(
      A=&SelectFloatFromArray(
       Direct Array Index=$Particles.PartitionParticles.Partition
       Float Selection Array=DataInterface() {propertyValues {FloatData [0.5,0.5,0.3,0.5,0.8,0.5,0.5,0.3,0.5,0.8,1,0.5,0.3,0.5,0.3,0.5,0.3,0.5,0,0,0,0,0,0,0]}}
      )
      B=1.2
     )
     Vector2D=$Particles.SpriteSize
    )
    Particles.GlowDecalColor:LinearColor=&MultiplyLinearColor(
     Linear Color=&Multiply_LinearColorByFloat(
      Linear Color=&VariantColorSelector(Variant Amount=1, Original Color=<rgba 0.0054 0.1413 0.195 1>, Variant 1 Color=<rgba 0.4844 0.0807 0.37 1>, Variant 2 Color=<rgba 1 1 1 1>)
      Float=&Multiply_Float(A=0.5, B=$Particles.AlphaScale)
     )
     Scale Factor=&MakeFloat4FromVectorAndFloat("Vector (RGB)"=$System.ToD, "Float (Alpha)"=5)
    )
   ) {isSetParameters=true, module=SetVariables_#13}
   /SetVariables(Particles.VisibilityTag=&MakeCustomIntFromBool(
     Bool=$User.bisFootHidden
     True INT=&self:ScratchDynamicInput(
      Masked Int:NiagaraInt32=1
      Partition Mask:NiagaraInt32=&SelectIntFromArray(
       Direct Array Index=&ReturnExecIndex()
       Int Selection Array=DataInterface() {propertyValues {IntData [-1,-1,-1,-1,-1,-1,-1,-1,-1,-1,-1,-1,-1,-1,-1,-1,-1,-1,18,19,20,21,22,23,24,25]}}
      )
      Passthrough Int:NiagaraInt32=0
     )
     False INT=0
    )) {isSetParameters=true}
  ]
  rendererClasses [&NiagaraMeshRendererProperties, &NiagaraSpriteRendererProperties]
  renderers [
   {rendererClass=&NiagaraMeshRendererProperties, rendererIndex=0}
   {rendererClass=&NiagaraMeshRendererProperties, rendererIndex=1}
   {rendererClass=&NiagaraMeshRendererProperties, rendererIndex=2}
   {rendererClass=&NiagaraMeshRendererProperties, rendererIndex=3}
   {rendererClass=&NiagaraMeshRendererProperties, rendererIndex=4}
   {rendererClass=&NiagaraMeshRendererProperties, rendererIndex=5}
   {rendererClass=&NiagaraSpriteRendererProperties, rendererIndex=6}
   {rendererClass=&NiagaraSpriteRendererProperties, rendererIndex=7}
   {rendererClass=&NiagaraSpriteRendererProperties, rendererIndex=8}
   {rendererClass=&NiagaraSpriteRendererProperties, rendererIndex=9}
   {rendererClass=&NiagaraSpriteRendererProperties, rendererIndex=10}
   {rendererClass=&NiagaraSpriteRendererProperties, rendererIndex=11}
  ]
 }
 Hornflames {
  emitterUpdateScript [
   &EmitterState(
    Inactive Response=::NewEnumerator1
    Life Cycle Mode=::NewEnumerator1
    Loop Behavior=::NewEnumerator1
    Loop Duration Mode=::NewEnumerator1
    UseLoopDelay=false
   )
   &SpawnBurst_Instantaneous(
    Spawn Count=&Multiply_Int(
     A=2
     B=&Multiply_Int(
      A=&MakeIntFromFloat(Float=$User.GlowOn)
      B=&MakeIntFromFloat(Float=$User.HelmetOn)
     )
    )
    Use Spawn Probability=false
    UseLoopCountLimit=false
   )
  ]
  particleSpawnScript [
   &PartitionParticles()
   &InitializeParticle(
    Color=&MultiplyLinearColor(
     Linear Color=&VariantColorSelector(
      Variant Amount=1
      Original Color=<rgba 0.0018 0.0885 0.0857 1>
      Variant 1 Color=<rgba 0.4844 0.0807 0.37 0.15>
      Variant 2 Color=<rgba 0.125 0.0182 0.1125 1>
     )
     Scale Factor=&MakeFloat4FromVectorAndFloat("Vector (RGB)"=$System.ToD, "Float (Alpha)"=1)
    )
    Lifetime Max=15
    Lifetime Min=12
    Lifetime Mode=::NewEnumerator1
    Mesh Scale Mode=::NewEnumerator3
    Sprite Size Mode=::NewEnumerator3
    Uniform Sprite Size=4
    UsePositionOffset=false
    Write Material Random=true
   )
   &InitialMeshOrientation(Mesh Axis To Orient=<0 0 1>, Mesh Orientation Mode=::NewEnumerator2, Orientation Coordinate Space=::World, Use Rotation Vector=false)
   /SetVariables(Particles.VisibilityTag=&SelectIntFromArray(Direct Array Index=$Particles.PartitionParticles.Partition, Int Selection Array=DataInterface() {propertyValues {IntData [0,0,0,1,1,0,0]}})) {enabled=false, isSetParameters=true, module=SetVariables_#7}
   &DynamicMaterialParameters(Write Parameter Index 0=false)
   &CameraOffset(Camera Offset Amount=0.5)
   /SetVariables(Particles.MeshIndex=$Particles.PartitionParticles.Partition) {isSetParameters=true}
  ]
  particleUpdateScript [
   &ParticleState(Kill Particles When Lifetime Has Elapsed=false, Let Infinitely Lived Particles Die When Emitter Deactivates=true)
   /SetVariables(Particles.AlphaScale:NiagaraFloat=&FN_FrontEndFloatModifier(
     Gameplay=&Lerp_Float(A=0, B=1, Alpha=$System.Timeline.NormalizedTime)
     Front End=&MakeCustomFloatFromBool(Bool=$User.bIsFrontEnd, True Float=$System.FrontEndGlowLoop, False Float=0)
    )) {isSetParameters=true, module=SetVariables_#8}
   &ScaleMeshSize(Scale Factor=&Multiply_VectorByFloat(Vector=<1 1 1>, Float=&MakeCustomFloatFromBool(Bool=$User.bisHeadHidden, True Float=0, False Float=1)))
   &SolveForcesAndVelocity(Clamp Velocity=false, Limit Acceleration=false)
   &SkeletalMeshLocation(Bone Sampling Mode=::NewEnumerator4, MaskSpawnGroup=false, Skeletal Mesh=$System.Skeletal Mesh)
   &ScaleColor(Scale Alpha=$Particles.AlphaScale, Scale RGB=&VectorFromFloat(Value=3.5), ScaleA=true, ScaleRGB=true)
   /SetVariables(Particles.Color_02:LinearColor=&MultiplyLinearColor(
     Linear Color=&MultiplyLinearColor(
      Linear Color=&VariantColorSelector(
       Variant Amount=1
       Original Color=<rgba 0.0018 0.0885 0.0857 1>
       Variant 1 Color=<rgba 0.8594 0.4073 0.7313 0.2>
       Variant 2 Color=<rgba 0.2812 0.041 0.2452 1>
      )
      Scale Factor=&MakeFloat4FromVectorAndFloat("Vector (RGB)"=$System.ToD, "Float (Alpha)"=1)
     )
     Scale Factor=&MakeFloat4FromVectorAndFloat("Vector (RGB)"=<1 1 1>, "Float (Alpha)"=$Particles.AlphaScale)
    )) {isSetParameters=true, module=SetVariables_#9}
   /SetVariables(Particles.Color_03:LinearColor=&MultiplyLinearColor(
     Linear Color=&MultiplyLinearColor(
      Linear Color=&VariantColorSelector(
       Variant Amount=1
       Original Color=<rgba 0.2203 0.9 0.7754 1>
       Variant 1 Color=<rgba 0.9323 0.6167 0.806 0.5>
       Variant 2 Color=<rgba 0.6879 0.3357 0.816 1>
      )
      Scale Factor=&MakeFloat4FromVectorAndFloat("Vector (RGB)"=$System.ToD, "Float (Alpha)"=$Particles.AlphaScale)
     )
     Scale Factor=&MakeFloat4FromVectorAndFloat("Vector (RGB)"=<1 1 1>, "Float (Alpha)"=1.5)
    )) {isSetParameters=true, module=SetVariables_#10}
   &InterpolateOverTime(Rate Of Change=3, Target Position Value:NiagaraPosition=$Particles.Position, Variable Type=::NewEnumerator6)
   /SetVariables(Particles.Interpolate Dynamic=&MakeFloat4FromVectorAndFloat(
     "Vector (RGB)"=$Particles.InterpolateOverTime.InterpolateOverTime_Position.Moving Average
     "Float (Alpha)"=&Add_Float(A=$User.Interpolate Intensity, B=-0.5)
    )) {isSetParameters=true, module=SetVariables_#11}
   &DynamicMaterialParameters(
    Index 2 Param 1=&Multiply_Float(
     A=&SelectFloatFromArray(Direct Array Index=$Particles.PartitionParticles.Partition, Float Selection Array=DataInterface() {propertyValues {FloatData [1,0.5]}})
     B=&Add_Float(
      A=&FloatFromCurve(
       FloatCurve=DataInterface() {propertyValues {
         OptimizeThreshold=0
         bOptimizeLUT=false
         Curve {keys [
           {arriveTangent=0.1068, interpMode=RCIM_Cubic, leaveTangent=0.1068, tangentMode=RCTM_User, time=0, value=0}
           {arriveTangent=2.2494, interpMode=RCIM_Cubic, leaveTangent=2.2494, tangentMode=RCTM_User, time=0.3296, value=0.2659}
           {
            arriveTangent=6.1869
            arriveTangentWeight=0.1296
            interpMode=RCIM_Cubic
            leaveTangent=-6.8183
            leaveTangentWeight=0.2592
            tangentMode=RCTM_Break
            time=0.5
            value=1
           }
           {arriveTangent=-2.6667, interpMode=RCIM_Cubic, leaveTangent=0.052, tangentMode=RCTM_Break, time=0.6, value=0}
           <at 1 0>
          ]}
        }}
       CurveIndex=&ModuloFloat(
        A=&Multiply_Float(
         A=&Add_Float(
          A=$Emitter.Age
          B=&SelectFloatFromArray(Direct Array Index=$Particles.PartitionParticles.Partition, Float Selection Array=DataInterface() {propertyValues {FloatData [0,0.5]}})
         )
         B=&SelectFloatFromArray(Direct Array Index=$Particles.PartitionParticles.Partition, Float Selection Array=DataInterface() {propertyValues {FloatData [0.2,0.35]}})
        )
        B=1
       )
      )
      B=&Waveform(
       Global Phase=&Add_Float(
        A=&Multiply_Float(A=$Emitter.Age, B=2)
        B=&SelectFloatFromArray(Direct Array Index=$Particles.PartitionParticles.Partition, Float Selection Array=DataInterface() {propertyValues {FloatData [0,0.5]}})
       )
       "Waveform [1]"=::NewEnumerator12
       "[1] Frequency"=1
       "[1] Amplitude Min/Max"=<0 0.35>
      )
     )
    )
    Index 2 Param 2=&Lerp_Float(
     A=2
     B=0.05
     Alpha=&FloatFromCurve(
      FloatCurve=DataInterface() {propertyValues {
        OptimizeThreshold=0
        bOptimizeLUT=false
        Curve {keys [
          {arriveTangent=0.1068, interpMode=RCIM_Cubic, leaveTangent=0.1068, tangentMode=RCTM_User, time=0.05, value=0}
          {arriveTangent=2.2494, interpMode=RCIM_Cubic, leaveTangent=2.2494, tangentMode=RCTM_User, time=0.38, value=0.2659}
          {
           arriveTangent=6.1869
           arriveTangentWeight=0.1296
           interpMode=RCIM_Cubic
           leaveTangent=0.0352
           leaveTangentWeight=0.2592
           tangentMode=RCTM_Break
           time=0.55
           value=1
          }
          {arriveTangent=0.0279, interpMode=RCIM_Cubic, leaveTangent=-6.6667, tangentMode=RCTM_Break, time=0.6, value=1}
          {arriveTangent=-2.6667, interpMode=RCIM_Cubic, leaveTangent=0.052, tangentMode=RCTM_Break, time=0.7, value=0}
          <at 1 0>
         ]}
       }}
      CurveIndex=&ModuloFloat(
       A=&Multiply_Float(
        A=&Add_Float(
         A=$Emitter.Age
         B=&SelectFloatFromArray(Direct Array Index=$Particles.PartitionParticles.Partition, Float Selection Array=DataInterface() {propertyValues {FloatData [0,0.5]}})
        )
        B=&SelectFloatFromArray(Direct Array Index=$Particles.PartitionParticles.Partition, Float Selection Array=DataInterface() {propertyValues {FloatData [0.2,0.35]}})
       )
       B=1
      )
     )
    )
    Index 2 Param 3=&FloatFromCurve(
     FloatCurve=DataInterface() {propertyValues {
       OptimizeThreshold=0
       bOptimizeLUT=false
       Curve {keys [
         {arriveTangent=0.1068, interpMode=RCIM_Cubic, leaveTangent=0.1068, tangentMode=RCTM_User, time=0.05, value=0}
         {arriveTangent=2.2494, interpMode=RCIM_Cubic, leaveTangent=2.2494, tangentMode=RCTM_User, time=0.38, value=0.2659}
         {
          arriveTangent=6.1869
          arriveTangentWeight=0.1296
          interpMode=RCIM_Cubic
          leaveTangent=-6.8183
          leaveTangentWeight=0.2592
          tangentMode=RCTM_Break
          time=0.55
          value=1
         }
         {arriveTangent=-2.6667, interpMode=RCIM_Cubic, leaveTangent=0.052, tangentMode=RCTM_Break, time=0.65, value=0}
         <at 1 0>
        ]}
      }}
     CurveIndex=&ModuloFloat(
      A=&Multiply_Float(
       A=&Add_Float(
        A=$Emitter.Age
        B=&SelectFloatFromArray(Direct Array Index=$Particles.PartitionParticles.Partition, Float Selection Array=DataInterface() {propertyValues {FloatData [0,0.5]}})
       )
       B=&SelectFloatFromArray(Direct Array Index=$Particles.PartitionParticles.Partition, Float Selection Array=DataInterface() {propertyValues {FloatData [0.2,0.35]}})
      )
      B=1
     )
    )
    Module.Index2Param0Write=true
    Module.Index2Param1Write=true
    Module.Index2Param2Write=true
    Module.Index2Param3Write=true
    Write Parameter Index 0=false
    Write Parameter Index 2=true
   ) {module=DynamicMaterialParameters001}
   &DynamicMaterialParameters(
    Index 2 Param 1=&FloatFromCurve(
     FloatCurve=DataInterface() {propertyValues {
       OptimizeThreshold=0
       bOptimizeLUT=false
       Curve {keys [
         {arriveTangent=0.1068, interpMode=RCIM_Cubic, leaveTangent=0.1068, tangentMode=RCTM_User, time=0, value=0}
         {arriveTangent=2.2494, interpMode=RCIM_Cubic, leaveTangent=2.2494, tangentMode=RCTM_User, time=0.3296, value=0.2659}
         {
          arriveTangent=6.1869
          arriveTangentWeight=0.1296
          interpMode=RCIM_Cubic
          leaveTangent=-6.8183
          leaveTangentWeight=0.2592
          tangentMode=RCTM_Break
          time=0.5
          value=1
         }
         {arriveTangent=-2.6667, interpMode=RCIM_Cubic, leaveTangent=0.052, tangentMode=RCTM_Break, time=0.6, value=0}
         <at 1 0>
        ]}
      }}
     CurveIndex=&ModuloFloat(
      A=&Multiply_Float(
       A=&Add_Float(
        A=$Emitter.Age
        B=&SelectFloatFromArray(Direct Array Index=$Particles.PartitionParticles.Partition, Float Selection Array=DataInterface() {propertyValues {FloatData [0,0.5]}})
       )
       B=&SelectFloatFromArray(Direct Array Index=$Particles.PartitionParticles.Partition, Float Selection Array=DataInterface() {propertyValues {FloatData [0.2,0.35]}})
      )
      B=1
     )
    )
    Index 2 Param 2=1
    Index 2 Param 3=1
    Module.Index2Param0Write=true
    Module.Index2Param1Write=true
    Module.Index2Param2Write=true
    Module.Index2Param3Write=true
    Write Parameter Index 0=false
    Write Parameter Index 2=true
   ) {enabled=false, module=DynamicMaterialParameters002}
  ]
  rendererClasses [&NiagaraMeshRendererProperties]
  renderers [
   {rendererClass=&NiagaraMeshRendererProperties, rendererIndex=0}
   {rendererClass=&NiagaraMeshRendererProperties, rendererIndex=1}
   {rendererClass=&NiagaraMeshRendererProperties, rendererIndex=2}
  ]
 }
 Hornglow {
  emitterUpdateScript [
   &EmitterState(
    Inactive Response=::NewEnumerator1
    Life Cycle Mode=::NewEnumerator1
    Loop Behavior=::NewEnumerator1
    Loop Duration Mode=::NewEnumerator1
    UseLoopDelay=false
   )
   &SpawnBurst_Instantaneous(
    Spawn Count=&Multiply_Int(
     A=1
     B=&Multiply_Int(
      A=&MakeIntFromFloat(Float=$User.GlowOn)
      B=&MakeIntFromFloat(Float=$User.HelmetOn)
     )
    )
    Use Spawn Probability=false
    UseLoopCountLimit=false
   )
   &SpawnRate(SpawnRate=15, Use Spawn Probability=false) {enabled=false}
  ]
  particleSpawnScript [
   &InitializeParticle(
    Color=&MultiplyLinearColor(
     Linear Color=&VariantColorSelector(
      Variant Amount=1
      Original Color=<rgba 0.0018 0.0885 0.0857 1>
      Variant 1 Color=<rgba 0.4844 0.0807 0.37 0.25>
      Variant 2 Color=<rgba 0.2812 0.041 0.2452 1>
     )
     Scale Factor=&MakeFloat4FromVectorAndFloat("Vector (RGB)"=&Multiply_VectorByFloat(Vector=$System.ToD, Float=20), "Float (Alpha)"=0.15)
    )
    Lifetime Max=15
    Lifetime Min=12
    Lifetime Mode=::NewEnumerator1
    Mesh Scale Mode=::NewEnumerator3
    Sprite Rotation Mode=::NewEnumerator2
    UsePositionOffset=false
    Write Material Random=true
   )
   &InitialMeshOrientation(Mesh Axis To Orient=<0 0 1>, Mesh Orientation Mode=::NewEnumerator2, Orientation Coordinate Space=::World, Use Rotation Vector=false)
   &DynamicMaterialParameters(
    Index 0 Param 1=1
    Index 0 Param 2=1
    Module.Index0Param0Write=true
    Module.Index0Param1Write=true
    Module.Index0Param2Write=true
    Module.Index0Param3Write=true
   )
   /SetVariables(Particles.VisibilityTag=&SelectIntFromArray(
     Direct Array Index=&MakeIntFromFloat(Float=$User.Variant)
     Int Selection Array=DataInterface() {propertyValues {IntData [0,0,1]}}
    )) {isSetParameters=true}
  ]
  particleUpdateScript [
   &ParticleState(Kill Particles When Lifetime Has Elapsed=false)
   &ScaleMeshSize(Scale Factor=&Multiply_VectorByFloat(Vector=<1 1 1>, Float=&MakeCustomFloatFromBool(Bool=$User.bisHeadHidden, True Float=0, False Float=1)))
   &SkeletalMeshLocation(Bone Sampling Mode=::NewEnumerator4, MaskSpawnGroup=false, Skeletal Mesh=$System.Skeletal Mesh)
   /SetVariables(Particles.AlphaScale:NiagaraFloat=&FN_FrontEndFloatModifier(
     Gameplay=&Lerp_Float(A=0, B=0.8, Alpha=$System.Timeline.NormalizedTime)
     Front End=&MakeCustomFloatFromBool(Bool=$User.bIsFrontEnd, True Float=$System.FrontEndGlowLoop, False Float=0)
    )) {isSetParameters=true}
   &ScaleColor(Scale Alpha=$Particles.AlphaScale, ScaleA=true, ScaleRGB=true)
  ]
  rendererClasses [&NiagaraMeshRendererProperties]
  renderers [
   {rendererClass=&NiagaraMeshRendererProperties, rendererIndex=0}
   {rendererClass=&NiagaraMeshRendererProperties, rendererIndex=1}
   {rendererClass=&NiagaraMeshRendererProperties, rendererIndex=2}
  ]
 }
 MeltVariant_Flames {
  emitterUpdateScript [
   &EmitterState(
    Inactive Response=::NewEnumerator1
    Life Cycle Mode=::NewEnumerator1
    Loop Behavior=::NewEnumerator1
    Loop Duration Mode=::NewEnumerator1
    UseLoopDelay=false
   )
   &SpawnBurst_Instantaneous(
    Spawn Count=&Multiply_Int(
     A=&MakeIntFromFloat(Float=&VariantFloatSelector(
       Original Float:NiagaraFloat=0
       Variant 1 Float:NiagaraFloat=25
       Variant 2 Float:NiagaraFloat=0
       Variant 3 Float:NiagaraFloat=?
       Variant Amount:NiagaraInt32=1
      ))
     B=&MakeIntFromFloat(Float=$User.GlowOn)
    )
    Use Spawn Probability=false
    UseLoopCountLimit=false
   )
  ]
  particleSpawnScript [
   &PartitionParticles(Number Of Partitions=25)
   &InitializeParticle(
    Color=&MultiplyLinearColor(
     Linear Color=&VariantColorSelector(
      Variant Amount=1
      Original Color=<rgba 0.0018 0.0885 0.0857 1>
      Variant 1 Color=<rgba 0.4844 0.0807 0.37 0.7>
      Variant 2 Color=<rgba 0.125 0.0182 0.1125 1>
      Use FrontEnd?=true
      Front End Color=<rgba 0.8385 0.6398 0.7 0.7>
      Gameplay Color=<rgba 1 1 1 1>
     )
     Scale Factor=&MakeFloat4FromVectorAndFloat("Vector (RGB)"=$System.ToD, "Float (Alpha)"=1)
    )
    Lifetime Max=15
    Lifetime Min=12
    Lifetime Mode=::NewEnumerator1
    Mesh Scale Mode=::NewEnumerator3
    Mesh Uniform Scale=&SelectFloatFromArray(
     Direct Array Index=$Particles.PartitionParticles.Partition
     Float Selection Array=DataInterface() {propertyValues {FloatData [1,0.5,0.4,0.3,0.5,1,0.45,0.35,1,0.5,1.1,0.8,0.3,1.5,0.5,0.8,0.4,0.25,0.5,0.35,0.2,0.65,0.2,0.5,0.2]}}
    )
    Sprite Size Mode=::NewEnumerator3
    Uniform Sprite Size=&SelectFloatFromArray(
     Direct Array Index=$Particles.PartitionParticles.Partition
     Float Selection Array=DataInterface() {propertyValues {FloatData [4,4,4,4,4,4,4,4,4,4,4,4,4,4,4,4,4,4,0,0,0,0,0,0,0,0,0]}}
    )
    UsePositionOffset=false
    Write Material Random=true
   )
   &InitialMeshOrientation(
    Mesh Axis To Orient=<0 0 1>
    Mesh Orientation Mode=::NewEnumerator2
    Orientation Coordinate Space=::World
    Orientation Vector=&SelectVectorFromArray(
     Direct Array Index=$Particles.PartitionParticles.Partition
     Vector Selection Array=DataInterface() {propertyValues {FloatData [
        <0 0 1>
        <0 0 1>
        <0 0 1>
        <0 0 1>
        <0 0 1>
        <0 0 -1>
        <0 0 -1>
        <0 0 -1>
        <0 0 -1>
        <0 0 -1>
        <-1 0 0>
        <-1 0 0>
        <-1 0 0>
        <-1 0 0>
        <-1 0 0>
        <-1 0 0>
        <-1 0 0>
        <0 0 1>
       ]}}
    )
    Rotation=<0.25 0 0>
    Use Rotation Vector=false
   )
   /SetVariables(Particles.VisibilityTag=&SelectIntFromArray(Direct Array Index=$Particles.PartitionParticles.Partition, Int Selection Array=DataInterface() {propertyValues {IntData [0,0,0,1,1,0,0]}})) {enabled=false, isSetParameters=true, module=SetVariables_#7}
   &DynamicMaterialParameters(
    Index 0 Param 1=&RandomRangeFloat(Minimum=0, Maximum=1)
    Index 0 Param 2=&SelectFloatFromArray(
     Direct Array Index=$Particles.PartitionParticles.Partition
     Float Selection Array=DataInterface() {propertyValues {FloatData [3,1,0.6,1.2,0.8,3,1,1,0.5,1,1,1,0.5,3,1.5,0.8,1,0.8,1,0.75,0.5,1.2,0.8,1.1,0.5]}}
    )
    Module.Index0Param0Write=true
    Module.Index0Param1Write=true
    Module.Index0Param2Write=true
    Module.Index0Param3Write=true
   )
   &CameraOffset(Camera Offset Amount=0.5)
  ]
  particleUpdateScript [
   &ParticleState(Kill Particles When Lifetime Has Elapsed=false, Let Infinitely Lived Particles Die When Emitter Deactivates=true)
   /SetVariables(Particles.AlphaScale:NiagaraFloat=&FN_FrontEndFloatModifier(
     Gameplay=&Lerp_Float(A=0, B=1, Alpha=$System.Timeline.NormalizedTime)
     Front End=&MakeCustomFloatFromBool(Bool=$User.bIsFrontEnd, True Float=$System.FrontEndGlowLoop, False Float=0)
    )) {isSetParameters=true, module=SetVariables_#8}
   &ScaleMeshSize(Scale Factor=<1 1 2>)
   &SolveForcesAndVelocity(Clamp Velocity=false, Limit Acceleration=false)
   &SkeletalMeshLocation(
    Bone / Socket Index=&ReturnExecIndex()
    MaskSpawnGroup=false
    Mesh Sampling Type=::NewEnumerator5
    Skeletal Mesh=$System.Skeletal Mesh 02
    Socket Sampling Mode=::NewEnumerator5
   )
   &SkeletalMeshLocation(
    Bone / Socket Index=&ReturnExecIndex()
    MaskSpawnGroup=false
    Mesh Sampling Type=::NewEnumerator5
    Orientation Sampling=::NewEnumerator1
    Skeletal Mesh=$System.Skeletal Mesh 02 Low
    Socket Sampling Mode=::NewEnumerator5
   ) {module=SkeletalMeshLocation001}
   /SetVariables(Particles.Position=&MakeCustomVectorFromBool(
     Bool=$User.bisLowLOD
     True Vector=$Particles.SkeletalMeshLocation001.SampledPosition
     False Vector=$Particles.SkeletalMeshLocation.SampledPosition
    )) {isSetParameters=true}
   &ScaleColor(Scale Alpha=$Particles.AlphaScale, Scale RGB=&VectorFromFloat(Value=3.5), ScaleA=true, ScaleRGB=true)
   /SetVariables(Particles.Color_02:LinearColor=&MultiplyLinearColor(
     Linear Color=&MultiplyLinearColor(
      Linear Color=&VariantColorSelector(
       Variant Amount=1
       Original Color=<rgba 0.0018 0.0885 0.0857 1>
       Variant 1 Color=<rgba 0.8594 0.4073 0.7313 0.25>
       Variant 2 Color=<rgba 0.2812 0.041 0.2452 0.5>
       Use FrontEnd?=true
       Front End Color=<rgba 0.9219 0.6578 0.7722 0.25>
       Gameplay Color=<rgba 1 1 1 1>
      )
      Scale Factor=&MakeFloat4FromVectorAndFloat("Vector (RGB)"=$System.ToD, "Float (Alpha)"=1)
     )
     Scale Factor=&MakeFloat4FromVectorAndFloat("Vector (RGB)"=<1 1 1>, "Float (Alpha)"=$Particles.AlphaScale)
    )) {isSetParameters=true, module=SetVariables_#9}
   /SetVariables(Particles.Color_03:LinearColor=&MultiplyLinearColor(
     Linear Color=&MultiplyLinearColor(
      Linear Color=&VariantColorSelector(
       Variant Amount=1
       Original Color=<rgba 0.2203 0.9 0.7754 1>
       Variant 1 Color=<rgba 0.9323 0.6167 0.806 1>
       Variant 2 Color=<rgba 0.6879 0.3357 0.816 1>
       Use FrontEnd?=true
       Front End Color=<rgba 1 0.8177 0.8937 1>
       Gameplay Color=<rgba 1 1 1 1>
      )
      Scale Factor=&MakeFloat4FromVectorAndFloat("Vector (RGB)"=$System.ToD, "Float (Alpha)"=$Particles.AlphaScale)
     )
     Scale Factor=&MakeFloat4FromVectorAndFloat("Vector (RGB)"=<1 1 1>, "Float (Alpha)"=1.5)
    )) {isSetParameters=true, module=SetVariables_#10}
   &InterpolateOverTime(Rate Of Change=3, Target Position Value:NiagaraPosition=$Particles.Position, Variable Type=::NewEnumerator6)
   /SetVariables(Particles.Interpolate Dynamic=&MakeFloat4FromVectorAndFloat(
     "Vector (RGB)"=$Particles.InterpolateOverTime.InterpolateOverTime_Position.Moving Average
     "Float (Alpha)"=&Add_Float(A=$User.Interpolate Intensity, B=-0.5)
    )) {isSetParameters=true, module=SetVariables_#11}
   &FN_Sprite_Attributes(
    Camera Offset:NiagaraFloat=&SelectFloatFromArray(
     Direct Array Index=$Particles.PartitionParticles.Partition
     Float Selection Array=DataInterface() {propertyValues {FloatData [2,2,5,5,5,2,2,5,0,5,2,1.5,1.5,1,1,1.5,1,1.5]}}
    )
    Color=&MultiplyLinearColor(
     Linear Color=&VariantColorSelector(
      Variant Amount=1
      Original Color=<rgba 0.0473 0.4323 0.3617 1>
      Variant 1 Color=<rgba 0.8594 0.2014 0.673 0.2>
      Variant 2 Color=<rgba 0.2812 0.041 0.2452 1>
     )
     Scale Factor=&MakeFloat4FromVectorAndFloat("Vector (RGB)"=$System.ToD, "Float (Alpha)"=&Multiply_Float(A=0.3, B=$Particles.AlphaScale))
    )
    Create Camera Offset Attribute=true
    Create Color Attribute=true
    Write to Module Namespace=true
   )
   /SetVariables(Particles.AirglowDyn:Vector4f=<2 1 0 0>) {isSetParameters=true, module=SetVariables_#12}
   /SetVariables(
    Particles.DecalDyn:Vector4f=&MakeVector4(
     X=0.8
     Y=0.5
     Z=&RandomRangeFloat(Minimum=0, Maximum=1)
     W=&RandomRangeFloat(Minimum=0, Maximum=0)
    )
    Particles.FireDecalColor:LinearColor=&MultiplyLinearColor(
     Linear Color=&Multiply_LinearColorByFloat(
      Linear Color=&VariantColorSelector(Variant Amount=1, Original Color=<rgba 0.1721 0.6323 0.7031 1>, Variant 1 Color=<rgba 0.8594 0.2014 0.673 1>, Variant 2 Color=<rgba 1 1 1 1>)
      Float=&Multiply_Float(
       A=1.5
       B=&SelectFloatFromArray(
        Direct Array Index=$Particles.PartitionParticles.Partition
        Float Selection Array=DataInterface() {propertyValues {FloatData [1,1.5,1.5,1.5,0.8,1.2,1,1.2,1,0.8,1.2,1.2,1.8,1.8,1.5,1.2,2,1]}}
       )
      )
     )
     Scale Factor=&MakeFloat4FromVectorAndFloat("Vector (RGB)"=$System.ToD, "Float (Alpha)"=$Particles.AlphaScale)
    )
    Particles.FireDecalSize:Vector2f=&Multiply_Vector2D_ByFloat(
     Float=&Multiply_Float(
      A=&SelectFloatFromArray(
       Direct Array Index=$Particles.PartitionParticles.Partition
       Float Selection Array=DataInterface() {propertyValues {FloatData [0.5,0.5,0.3,0.5,0.8,0.5,0.5,0.3,0.5,0.8,1,0.5,0.3,0.5,0.3,0.5,0.3,0.5,0,0,0,0,0,0,0]}}
      )
      B=1.2
     )
     Vector2D=$Particles.SpriteSize
    )
    Particles.GlowDecalColor:LinearColor=&MultiplyLinearColor(
     Linear Color=&Multiply_LinearColorByFloat(
      Linear Color=&VariantColorSelector(Variant Amount=1, Original Color=<rgba 0.0054 0.1413 0.195 1>, Variant 1 Color=<rgba 0.4844 0.0807 0.37 1>, Variant 2 Color=<rgba 1 1 1 1>)
      Float=&Multiply_Float(A=0.5, B=$Particles.AlphaScale)
     )
     Scale Factor=&MakeFloat4FromVectorAndFloat("Vector (RGB)"=$System.ToD, "Float (Alpha)"=5)
    )
   ) {isSetParameters=true, module=SetVariables_#13}
   /SetVariables(Particles.VisibilityTag=&MakeCustomIntFromBool(
     Bool=$User.bisFootHidden
     True INT=&self:ScratchDynamicInput(
      Masked Int:NiagaraInt32=1
      Partition Mask:NiagaraInt32=&SelectIntFromArray(
       Direct Array Index=&ReturnExecIndex()
       Int Selection Array=DataInterface() {propertyValues {IntData [-1,-1,-1,-1,-1,-1,-1,-1,-1,-1,-1,-1,-1,-1,-1,-1,-1,-1,18,19,20,21,22,23,24,25]}}
      )
      Passthrough Int:NiagaraInt32=0
     )
     False INT=0
    )) {isSetParameters=true}
  ]
  rendererClasses [&NiagaraMeshRendererProperties, &NiagaraSpriteRendererProperties]
  renderers [
   {rendererClass=&NiagaraMeshRendererProperties, rendererIndex=0}
   {rendererClass=&NiagaraMeshRendererProperties, rendererIndex=1}
   {rendererClass=&NiagaraMeshRendererProperties, rendererIndex=2}
   {rendererClass=&NiagaraMeshRendererProperties, rendererIndex=3}
   {rendererClass=&NiagaraMeshRendererProperties, rendererIndex=4}
   {rendererClass=&NiagaraMeshRendererProperties, rendererIndex=5}
   {rendererClass=&NiagaraSpriteRendererProperties, rendererIndex=6}
   {rendererClass=&NiagaraSpriteRendererProperties, rendererIndex=7}
   {rendererClass=&NiagaraSpriteRendererProperties, rendererIndex=8}
   {rendererClass=&NiagaraSpriteRendererProperties, rendererIndex=9}
   {rendererClass=&NiagaraSpriteRendererProperties, rendererIndex=10}
   {rendererClass=&NiagaraSpriteRendererProperties, rendererIndex=11}
  ]
 }
}
summary {
 systemName=NS_Character_PolishedJade_Body_Idle
 emitters [
  {emitterName=Flames, rendererClasses [&NiagaraMeshRendererProperties, &NiagaraSpriteRendererProperties]}
  {emitterName=Hornflames, rendererClasses [&NiagaraMeshRendererProperties]}
  {emitterName=Hornglow, rendererClasses [&NiagaraMeshRendererProperties]}
  {
   emitterName=EyeGlow
   rendererClasses [
    &NiagaraSpriteRendererProperties
    &NiagaraLightRendererProperties
    &NiagaraComponentRendererProperties
   ]
  }
  {
   emitterName=MeltVariant_Flames
   rendererClasses [&NiagaraMeshRendererProperties, &NiagaraSpriteRendererProperties]
  }
 ]
 userVariables [
  User.bisFootHidden {type=NiagaraBool}
  User.bIsFrontEnd {type=NiagaraBool}
  User.bIsFrontEndPreview {type=NiagaraBool}
  User.bisHeadHidden {type=NiagaraBool}
  User.bisLowLOD {type=NiagaraBool}
  User.GlowOn {type=NiagaraFloat, defaultValue=1, description=""}
  User.HelmetOn {type=NiagaraFloat, defaultValue=1, description=""}
  User.IdleGlow {type=NiagaraFloat, defaultValue=0, description=""}
  User.Interpolate Intensity {type=NiagaraFloat, defaultValue=1, description=""}
  User.RibbonIDOffset {type=NiagaraInt32}
  User.StartFrontEndLoop {type=NiagaraBool}
  User.Variant {type=NiagaraFloat, defaultValue=0, description=""}
 ]
}
