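# Material function that adds animated emissive glow to the Polished Jade character's hair, intended to show only at night.
# A mask texture's strand channel and a tint color shape the glow, animated by two panning noise layers combined with max.
# The "Is MIndMelt?" scalar blends to a variant modulated by the mask's red channel and base color red.
# The glow blends into incoming emissive, with Scalar_Time_of_Day supplying a night-only factor.
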MaterialFunction /BRCosmetics/Characters/Player/Female/Medium/Bodies/F_MED_PolishedJade_Mind/Materials/MF_HairEmissive_PolishedJade
graph {nodes [
  n1 = GetMaterialAttributes(None=FunctionInput() {inputType=FunctionInput_MaterialAttributes}) {attributeGetTypes [#1, #2]}
  n2 = NamedRerouteUsage()
  n3 = TextureSampleParameter2D() {
   desc="R:Hair mask G; Hairstrands"
   group=FN_Hair
   parameterName=Hair Mask
   samplerSource=SSM_Wrap_WorldGroupSettings
   samplerType=SAMPLERTYPE_Masks
   texture=&T_BlackMask
  }
  n4 = ComponentMask(None=@n1.BaseColor) {r=true}
  n5 = Multiply(A=@n3.R, B=Multiply(A=@n4, B=@n2))
  n6 = ScalarParameter() {parameterName=Is MIndMelt?}
  n7 = LinearInterpolate(
   A=@n1.EmissiveColor
   B=Max(A=Multiply(A=@n1.EmissiveColor) {constB=0.5}, B=LinearInterpolate(A=@n2, B=@n5, Alpha=@n6))
   Alpha=NamedRerouteUsage()
  )
  FunctionOutput(None=SetMaterialAttributes!(MaterialAttributes=@n1.MaterialAttributes, EmissiveColor=@n7) {attributeSetTypes [#1]})
  MaterialFunctionCall(Emissive=@n7) {materialFunction=&MF_FX_EmissiveAdjustments}
  n8 = NamedRerouteUsage()
  n9 = Power(Base=@n3.G)
  n10 = Add(A=@n9, B=@n4)
  n11 = Multiply(
   A=LinearInterpolate(A=@n9, B=Multiply(A=@n3.R, B=@n10), Alpha=ScalarParameter() {parameterName=Is MIndMelt?})
   B=VectorParameter() {parameterName=HairEmissive color, defaultValue=<rgba 0.2505 3.0746 4 1>}.RGB
  )
  NamedRerouteDeclaration(None=Multiply(A=Multiply(A=@n8, B=@n11)) {constB=0.5}) {name=Hair emissive color, nodeColor=<rgba 0 1 0.9766 1>}
  n12 = TextureCoordinate() {uTiling=0.8}
  n13 = TextureSample(UVs=Panner(Coordinate=Multiply(A=@n12)) {speedY=-0.05}) {samplerSource=SSM_Wrap_WorldGroupSettings, samplerType=SAMPLERTYPE_Grayscale, texture=&LowResBlurredNoise}
  n14 = TextureCoordinate() {uTiling=0.8}
  n15 = TextureSample(UVs=Panner(Coordinate=Multiply(A=@n14) {constB=2}) {speedY=-0.1}) {samplerSource=SSM_Wrap_WorldGroupSettings, samplerType=SAMPLERTYPE_Grayscale, texture=&LowResBlurredNoise}
  NamedRerouteDeclaration(None=Power(Base=Max(A=@n13.R, B=@n15.R)) {constExponent=1.5}) {name=Hair noise, nodeColor=<rgba 0 1 0.4844 1>}
  n16 = Constant()
  NamedRerouteDeclaration(None=MaterialFunctionCall(Value_Morning=@n16, Value_Day=@n16, Value_Evening=@n16, Value_Night=Constant() {r=1}) {materialFunction=&Scalar_Time_of_Day}.Result) {name=Glow only at night, nodeColor=<rgba 0.5938 0 1 1>}
 ]}
