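# Physics asset for the down-but-not-out (DBNO) mesh of the PolishedJade_Mind female body, setting up jacket secondary motion.
# Front, side and back capsule chains hang from base spheres, mostly limited to one swing axis, with some front twist.
# Chain ends cross-link: front-to-side rotates freely, while outer back tips lock angularly to the middle.
# Pelvis, spine and thigh capsules provide body shapes, and a huge sunken root sphere approximates a floor.
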PhysicsAsset /BRCosmetics/Characters/Player/Female/Medium/Bodies/F_MED_PolishedJade_Mind/Meshes/F_MED_PolishedJade_MindDBNO_Physics
bodies {
 jacket_back01_l [{
   shapeType=Capsule
   length=9.9208
   radius=4.1476
   shapeName=jacket_side01_l_capsule
   center=<4.8114 -1.6052 1.016>
   rotation=<rot 0.0001 -1.3219 -163.602>
  }]
 jacket_back01_m [{shapeType=Capsule, length=9.4685, radius=4.5998, shapeName=jacket_side01_l_capsule, center=<5.7262 -2.3561 0>}]
 jacket_back01_r [{
   shapeType=Capsule
   length=9.9208
   radius=4.1476
   shapeName=jacket_side01_l_capsule
   center=<4.8001 -1.1145 -0.8715>
   rotation=<rot 0.0001 1.322 -20.6232>
  }]
 jacket_back02_l [{
   shapeType=Capsule
   length=11.333
   radius=3.8648
   shapeName=jacket_side02_l_capsule
   center=<4.344 -1.7413 3.2628>
   rotation=<rot 0.0001 0.0002 16.8834>
  }]
 jacket_back02_m [{shapeType=Capsule, length=11.5965, radius=3.9511, shapeName=jacket_side02_l_capsule, center=<2.2519 -2.6995 0>}]
 jacket_back02_r [{
   shapeType=Capsule
   length=11.333
   radius=3.8648
   shapeName=jacket_side02_l_capsule
   center=<3.6695 -1.3035 -3.1441>
   rotation=<rot 3.0471 -1.1096 -20.0295>
  }]
 jacket_back03_l [{
   shapeType=Capsule
   length=8.6112
   radius=4.2967
   shapeName=jacket_side03_l_capsule
   center=<3.322 -1.4472 3.0646>
   rotation=<rot 0.0001 0.0002 14.2881>
  }]
 jacket_back03_m [{
   shapeType=Capsule
   length=11.7568
   radius=4.2967
   shapeName=jacket_side03_l_capsule
   center=<3.322 -0.3426 -0.0118>
   rotation=<rot 0 0.0002 0>
  }]
 jacket_back03_r [{
   shapeType=Capsule
   length=8.6112
   radius=4.2967
   shapeName=jacket_side03_l_capsule
   center=<3.5755 0.1956 -3.1143>
   rotation=<rot 2.3486 -1.3432 -33.1203>
  }]
 jacket_back_base_l [{shapeType=Sphere, radius=1.2424, shapeName=jacket_back_base_l_sphere}]
 jacket_back_base_m [{shapeType=Sphere, radius=1.3433, shapeName=jacket_back_base_m_sphere}]
 jacket_back_base_r [{shapeType=Sphere, radius=1.2424, shapeName=jacket_back_base_l_sphere}]
 jacket_front01_l [{shapeType=Capsule, length=7.6299, radius=3.7183, shapeName=jacket_front01_l_capsule, center=<3.2575 1.252 0>, rotation=<rot 0 0 -4.0043>}]
 jacket_front01_r [{shapeType=Capsule, length=7.6299, radius=3.7183, shapeName=jacket_front01_l_capsule, center=<-3.2575 -1.252 0>, rotation=<rot 0 0 175.9957>}]
 jacket_front02_l [{
   shapeType=Capsule
   length=9.6687
   radius=2.8929
   shapeName=jacket_front02_l_capsule
   center=<3.4154 0.5422 -0.3645>
   rotation=<rot -4.6813 -91.4042 2.7956>
  }]
 jacket_front02_r [{
   shapeType=Capsule
   length=9.6022
   radius=2.9594
   shapeName=jacket_front02_l_capsule
   center=<-3.4154 -0.5422 0.3645>
   rotation=<rot -4.6813 -91.4042 -177.2044>
  }]
 jacket_front03_l [{
   shapeType=Capsule
   length=10.4005
   radius=2.9127
   shapeName=jacket_front03_l_capsule
   center=<3.274 0.9524 -0.671>
   rotation=<rot 10.8756 -89.0958 1.6062>
  }]
 jacket_front03_r [{
   shapeType=Capsule
   length=10.4005
   radius=2.9127
   shapeName=jacket_front03_l_capsule
   center=<-3.274 -0.9524 0.671>
   rotation=<rot 7.4796 -89.2029 -178.4183>
  }]
 jacket_front04_l [{
   shapeType=Capsule
   length=12.4915
   radius=3.6262
   shapeName=jacket_front04_l_capsule
   center=<3.6478 2.0107 -0.4649>
   rotation=<rot 12.3979 -90.0911 0.8223>
  }]
 jacket_front04_r [{
   shapeType=Capsule
   length=12.4915
   radius=3.6262
   shapeName=jacket_front04_l_capsule
   center=<-3.6478 -2.0107 0.4649>
   rotation=<rot 12.3979 -90.0911 -179.1777>
  }]
 jacket_front05_l [{
   shapeType=Capsule
   length=10.7836
   radius=3.2443
   shapeName=jacket_front05_l_capsule
   center=<2.1679 1.4432 0.0598>
   rotation=<rot 12.5846 -90.1231 -0.5694>
  }]
 jacket_front05_r [{
   shapeType=Capsule
   length=10.6472
   radius=3.3808
   shapeName=jacket_front05_l_capsule
   center=<-2.1679 -1.4432 -0.0598>
   rotation=<rot 12.5846 -90.1231 179.4306>
  }]
 jacket_front_base_l [{shapeType=Sphere, radius=1.7704, shapeName=jacket_front_base_l_sphere}]
 jacket_front_base_r [{shapeType=Sphere, radius=1.7704, shapeName=jacket_front_base_l_sphere}]
 jacket_side01_l [{
   shapeType=Capsule
   length=8.8689
   radius=4.1476
   shapeName=jacket_side01_l_capsule
   center=<5.5718 -0.8018 0.2618>
   rotation=<rot 0 -1.3223 -85.9665>
  }]
 jacket_side01_r [{
   shapeType=Capsule
   length=8.8689
   radius=4.1476
   shapeName=jacket_side01_l_capsule
   center=<-5.5718 0.8018 -0.2618>
   rotation=<rot 0 -1.3223 90>
  }]
 jacket_side02_l [{
   shapeType=Capsule
   length=11.333
   radius=3.8648
   shapeName=jacket_side02_l_capsule
   center=<4.344 -1.3824 3.1224>
   rotation=<rot 0.1493 -0.9643 -88.8472>
  }]
 jacket_side02_r [{
   shapeType=Capsule
   length=11.333
   radius=3.8648
   shapeName=jacket_side02_l_capsule
   center=<-4.344 1.3824 -3.1224>
   rotation=<rot 0.1493 0.9643 90>
  }]
 jacket_side03_l [{
   shapeType=Capsule
   length=7.37
   radius=4.2967
   shapeName=jacket_side03_l_capsule
   center=<2.0498 -1.8562 3.1345>
   rotation=<rot 0 179.0684 -91.9584>
  }]
 jacket_side03_r [{
   shapeType=Capsule
   length=11.3522
   radius=4.2967
   shapeName=jacket_side03_l_capsule
   center=<-2.0498 1.8562 -3.1345>
   rotation=<rot 0 179.0684 91>
  }]
 jacket_side_base_l [{shapeType=Sphere, radius=1.7704, shapeName=jacket_front_base_l_sphere}]
 jacket_side_base_r [{shapeType=Sphere, radius=1.7704, shapeName=jacket_front_base_l_sphere}]
 pelvis [{shapeType=Capsule, length=8.1546, radius=12.5928, shapeName=pelvis_capsule, center=<-2.4391 -0.5027 0>}]
 root [{shapeType=Sphere, radius=227.1117, shapeName=root_sphere, center=<0 0 -225.7164>}]
 spine_01 [{shapeType=Capsule, length=4.3113, radius=11.3144, shapeName=spine_01_capsule, center=<0 1.2392 0>, rotation=<rot 0 -90 0>}]
 spine_02 [{shapeType=Capsule, length=4.4691, radius=10.2934, shapeName=spine_02_capsule, center=<0 2.6529 0>}]
 thigh_l [
  {
   shapeType=Capsule
   length=18.5486
   radius=8.6154
   shapeName=thigh_l_capsule
   center=<-15.5207 1.9707 -0.1203>
   rotation=<rot -89.7816 -179.9958 -176.0531>
  }
  {
   shapeType=Capsule
   length=20.3566
   radius=7.6972
   shapeName=thigh_l_capsule
   center=<-15.6011 -1.0719 0.0339>
   rotation=<rot -89.8755 -179.9982 173.6003>
  }
  {
   shapeType=Capsule
   length=14.0771
   radius=6.2727
   shapeName=thigh_l_capsule
   center=<-33.8819 0.7961 0.127>
   rotation=<rot -89.7852 -179.9976 -171.3657>
  }
 ]
 thigh_r [
  {
   shapeType=Capsule
   length=18.5486
   radius=8.6154
   shapeName=thigh_l_capsule
   center=<15.5126 -1.8531 0.1203>
   rotation=<rot -89.7816 -179.9958 3.9469>
  }
  {
   shapeType=Capsule
   length=20.3566
   radius=7.6972
   shapeName=thigh_l_capsule
   center=<15.6011 1.0719 -0.0339>
   rotation=<rot -89.8755 -179.9982 -6.3997>
  }
  {
   shapeType=Capsule
   length=14.0771
   radius=6.2727
   shapeName=thigh_l_capsule
   center=<33.8819 -0.7961 -0.127>
   rotation=<rot -89.7852 -179.9976 8.6343>
  }
 ]
}
constraints [
 {
  bone1Name=jacket_front01_l
  bone2Name=jacket_front_base_l
  swing1LimitDegrees=30
  swing2LimitDegrees=45
  swing2Motion=Locked
  twistLimitDegrees=45
  twistMotion=Locked
 }
 {
  bone1Name=jacket_front02_l
  bone2Name=jacket_front01_l
  swing1LimitDegrees=60
  swing2LimitDegrees=45
  swing2Motion=Locked
  twistLimitDegrees=20
 }
 {
  bone1Name=jacket_front03_l
  bone2Name=jacket_front02_l
  swing1LimitDegrees=65
  swing2LimitDegrees=45
  swing2Motion=Locked
  twistLimitDegrees=30
 }
 {
  bone1Name=jacket_front04_l
  bone2Name=jacket_front03_l
  swing1LimitDegrees=75
  swing2LimitDegrees=45
  swing2Motion=Locked
  twistLimitDegrees=25
 }
 {
  bone1Name=jacket_front05_l
  bone2Name=jacket_front04_l
  swing1LimitDegrees=30
  swing2LimitDegrees=45
  swing2Motion=Locked
  twistLimitDegrees=30
  twistMotion=Locked
 }
 {
  bone1Name=jacket_side02_l
  bone2Name=jacket_side01_l
  swing1LimitDegrees=45
  swing1Motion=Locked
  swing2LimitDegrees=20
  twistLimitDegrees=45
  twistMotion=Locked
 }
 {
  bone1Name=jacket_side01_l
  bone2Name=jacket_side_base_l
  swing1LimitDegrees=45
  swing1Motion=Locked
  swing2LimitDegrees=10
  twistLimitDegrees=45
  twistMotion=Locked
 }
 {
  bone1Name=jacket_side03_l
  bone2Name=jacket_side02_l
  swing1LimitDegrees=45
  swing1Motion=Locked
  swing2LimitDegrees=20
  twistLimitDegrees=45
  twistMotion=Locked
 }
 {
  bone1Name=jacket_front05_l
  bone2Name=jacket_side03_l
  swing1LimitDegrees=45
  swing1Motion=Free
  swing2LimitDegrees=45
  swing2Motion=Free
  twistLimitDegrees=45
  twistMotion=Free
 }
 {
  bone1Name=jacket_back03_l
  bone2Name=jacket_back02_l
  swing1LimitDegrees=20
  swing2LimitDegrees=45
  swing2Motion=Locked
  twistLimitDegrees=45
  twistMotion=Locked
 }
 {
  bone1Name=jacket_back02_l
  bone2Name=jacket_back01_l
  swing1LimitDegrees=15
  swing2LimitDegrees=45
  swing2Motion=Locked
  twistLimitDegrees=45
  twistMotion=Locked
 }
 {
  bone1Name=jacket_back01_l
  bone2Name=jacket_back_base_l
  swing1LimitDegrees=5
  swing2LimitDegrees=45
  swing2Motion=Locked
  twistLimitDegrees=45
  twistMotion=Locked
 }
 {
  bone1Name=jacket_back01_m
  bone2Name=jacket_back_base_m
  swing1LimitDegrees=5
  swing2LimitDegrees=45
  swing2Motion=Locked
  twistLimitDegrees=45
  twistMotion=Locked
 }
 {
  bone1Name=jacket_back02_m
  bone2Name=jacket_back01_m
  swing1LimitDegrees=15
  swing2LimitDegrees=45
  swing2Motion=Locked
  twistLimitDegrees=45
  twistMotion=Locked
 }
 {
  bone1Name=jacket_back03_m
  bone2Name=jacket_back02_m
  swing1LimitDegrees=20
  swing2LimitDegrees=45
  swing2Motion=Locked
  twistLimitDegrees=45
  twistMotion=Locked
 }
 {
  bone1Name=jacket_back03_l
  bone2Name=jacket_back03_m
  swing1LimitDegrees=45
  swing1Motion=Locked
  swing2LimitDegrees=45
  swing2Motion=Locked
  twistLimitDegrees=45
  twistMotion=Locked
 }
 {
  bone1Name=jacket_back01_r
  bone2Name=jacket_back_base_r
  swing1LimitDegrees=5
  swing2LimitDegrees=45
  swing2Motion=Locked
  twistLimitDegrees=45
  twistMotion=Locked
 }
 {
  bone1Name=jacket_back02_r
  bone2Name=jacket_back01_r
  swing1LimitDegrees=15
  swing2LimitDegrees=45
  swing2Motion=Locked
  twistLimitDegrees=45
  twistMotion=Locked
 }
 {
  bone1Name=jacket_back03_r
  bone2Name=jacket_back02_r
  swing1LimitDegrees=20
  swing2LimitDegrees=45
  swing2Motion=Locked
  twistLimitDegrees=45
  twistMotion=Locked
 }
 {
  bone1Name=jacket_front01_r
  bone2Name=jacket_front_base_r
  swing1LimitDegrees=20
  swing2LimitDegrees=45
  swing2Motion=Locked
  twistLimitDegrees=45
  twistMotion=Locked
 }
 {
  bone1Name=jacket_front02_r
  bone2Name=jacket_front01_r
  swing1LimitDegrees=35
  swing2LimitDegrees=45
  swing2Motion=Locked
  twistLimitDegrees=20
 }
 {
  bone1Name=jacket_front03_r
  bone2Name=jacket_front02_r
  swing1LimitDegrees=45
  swing2LimitDegrees=45
  swing2Motion=Locked
  twistLimitDegrees=30
 }
 {
  bone1Name=jacket_front04_r
  bone2Name=jacket_front03_r
  swing1LimitDegrees=45
  swing2LimitDegrees=45
  swing2Motion=Locked
  twistLimitDegrees=25
 }
 {
  bone1Name=jacket_front05_r
  bone2Name=jacket_front04_r
  swing1LimitDegrees=20
  swing2LimitDegrees=45
  swing2Motion=Locked
  twistLimitDegrees=30
  twistMotion=Locked
 }
 {
  bone1Name=jacket_side01_r
  bone2Name=jacket_side_base_r
  swing1LimitDegrees=45
  swing1Motion=Locked
  swing2LimitDegrees=10
  twistLimitDegrees=45
  twistMotion=Locked
 }
 {
  bone1Name=jacket_side02_r
  bone2Name=jacket_side01_r
  swing1LimitDegrees=20
  swing1Motion=Locked
  swing2LimitDegrees=20
  twistLimitDegrees=45
  twistMotion=Locked
 }
 {
  bone1Name=jacket_side03_r
  bone2Name=jacket_side02_r
  swing1LimitDegrees=45
  swing1Motion=Locked
  swing2LimitDegrees=20
  twistLimitDegrees=45
  twistMotion=Locked
 }
 {
  bone1Name=jacket_front05_r
  bone2Name=jacket_side03_r
  swing1LimitDegrees=45
  swing1Motion=Free
  swing2LimitDegrees=45
  swing2Motion=Free
  twistLimitDegrees=45
  twistMotion=Free
 }
 {
  bone1Name=jacket_back03_r
  bone2Name=jacket_back03_m
  swing1LimitDegrees=45
  swing1Motion=Locked
  swing2LimitDegrees=45
  swing2Motion=Locked
  twistLimitDegrees=45
  twistMotion=Locked
 }
]
