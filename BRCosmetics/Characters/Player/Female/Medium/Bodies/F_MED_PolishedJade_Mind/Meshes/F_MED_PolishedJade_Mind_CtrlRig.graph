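# Control Rig for the PolishedJade Mind female medium body that adds corrective bone motion to its jacket, arm pads, belt shield and emblem.
# A shared DeformRig function runs alongside twist, aim and parent-constraint fixes for the upper body.
# Spherical pose readers on arms and thighs, plus remapped spine and pelvis rotations, weight additive bone offsets.
# DBNO, crouch, melee and skydive variables switch or scale corrections, with Kalman filtering smoothing skydive weights.
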ControlRigBlueprint /BRCosmetics/Characters/Player/Female/Medium/Bodies/F_MED_PolishedJade_Mind/Meshes/F_MED_PolishedJade_Mind_CtrlRig
graphs [
 Maximum_1_ContainedGraph {nodes [
   n1 = RigVMFunctionEntryNode()
   Maximum = &RigVMFunction_MathDoubleMax(A=@n1.A, B=@n1.B)
   RigVMFunctionReturnNode(Result=&RigVMFunction_MathDoubleMax(B=@n1.D, A=&RigVMFunction_MathDoubleMax(B=@n1.C, A=@Maximum.Result).Result).Result)
  ]}
 Multiply_18_1_ContainedGraph {nodes [
   n1 = RigVMFunctionEntryNode()
   RigVMFunctionReturnNode(Result=&RigVMFunction_MathDoubleMul(B=@n1.C, A=&RigVMFunction_MathDoubleMul(A=@n1.A, B=@n1.B).Result).Result)
  ]}
 New Function_ContainedGraph {nodes [
   RigVMFunctionEntryNode()
   n1 = &RigUnit_GetTransform(Item=(Type=Bone,Name="spine_03"), bInitial=False)
   n2 = &RigUnit_GetTransform(Item=(Type=Bone,Name="pelvis"), bInitial=False)
   n3 = RigVMDispatchNode(Struct=/Script/RigVM.RigVMFunction_MathQuaternionToRotator(Value=&RigVMFunction_MathTransformMakeRelative(Global=@n1.Transform, Parent=@n2.Transform).Local.Rotation) {is=struct RigVMUnitNode, template="Cast::Execute(in Value,out Result)"}.Result) {template="DISPATCH_RigVMDispatch_BreakStruct(in Struct,out Elements)"}
   n4 = &RigUnit_GetTransform(Item=(Type=Bone,Name="root"), bInitial=False)
   n5 = &RigUnit_GetTransform(Item=(Type=Bone,Name="pelvis"), bInitial=False)
   n6 = RigVMDispatchNode(Struct=&RigVMFunction_MathQuaternionToRotator(Value=&RigVMFunction_MathTransformMakeRelative(Parent=@n4.Transform, Global=@n5.Transform).Local.Rotation) {template="Cast::Execute(in Value,out Result)"}.Result) {template="DISPATCH_RigVMDispatch_BreakStruct(in Struct,out Elements)"}
   > RigVMFunctionReturnNode(Result=&RigVMFunction_MathDoubleMul(
     B=&RigVMFunction_MathDoubleRemap(SourceMaximum=-30, SourceMinimum=0, TargetMaximum=1, TargetMinimum=0, Value=@n3.Elements.Yaw).Result
     A=&RigVMFunction_MathDoubleRemap(SourceMaximum=0, SourceMinimum=40, TargetMaximum=1, TargetMinimum=0, Value=@n6.Elements.Pitch).Result
    ).Result)
  ]}
 RigVMFunctionLibrary {nodes [CalculatePelvisSkydive = RigVMCollapseNode()]}
 RigVMFunction_Sequence_1_ContainedGraph {nodes [
   n1 = RigVMFunctionEntryNode()
   RigVMFunction_Sequence = &RigVMFunction_Sequence(ExecuteContext=@n1.ExecuteContext)
   n2 = &RigVMFunction_Sequence(ExecuteContext=@RigVMFunction_Sequence.B)
   n3 = &RigVMFunction_Sequence(ExecuteContext=@n2.B)
   n4 = &RigVMFunction_Sequence(ExecuteContext=@n3.B)
   RigVMFunctionReturnNode(A=@RigVMFunction_Sequence.A, B=@n2.A, C=@n3.A, D=@n4.A, E=@n4.B)
  ]}
 RigVMFunction_Sequence_3_1_ContainedGraph {nodes [
   n1 = RigVMFunctionEntryNode()
   n2 = &RigVMFunction_Sequence(ExecuteContext=@n1.ExecuteContext)
   n3 = &RigVMFunction_Sequence(ExecuteContext=@n2.B)
   n4 = &RigVMFunction_Sequence(ExecuteContext=@n3.B)
   n5 = &RigVMFunction_Sequence(ExecuteContext=@n4.B)
   RigVMFunctionReturnNode(A=@n2.A, B=@n3.A, C=@n4.A, D=@n5.A, E=@n5.B)
  ]}
 RigVMFunction_Sequence_5_1_ContainedGraph {nodes [
   n1 = RigVMFunctionEntryNode()
   n2 = &RigVMFunction_Sequence(ExecuteContext=@n1.ExecuteContext)
   n3 = &RigVMFunction_Sequence(ExecuteContext=@n2.B)
   RigVMFunctionReturnNode(A=@n2.A, B=@n3.A, C=@n3.B)
  ]}
 RigVMModel {nodes [
   RigUnit_BeginExecution = &RigUnit_BeginExecution()
   &RigUnit_ModifyTransforms(
    ItemToModify=((Item=(Type=Bone,Name="dyn_JacketCloseVertex"),Transform=(Rotation=(X=0,Y=0,Z=0,W=1),Translation=(X=0,Y=0,Z=0),Scale3D=(X=0.1,Y=0.1,Z=0.1))))
    Weight=1
    WeightMinimum=0
    ExecutePin=@RigUnit_BeginExecution.ExecutePin
   )
   > &RigUnit_ModifyTransforms(
    ItemToModify=((Item=(Type=Bone,Name="spine_03"),Transform=(Rotation=(X=0,Y=0,Z=0,W=1),Translation=(X=0,Y=0,Z=0),Scale3D=(X=1,Y=1,Z=1))))
    Weight=1
    WeightMinimum=0
   )
   > &RigUnit_ParentConstraint(
    AdvancedSettings=(InterpolationType=Average,RotationOrderForFilter=XZY)
    Child=(Type=Bone,Name="dyn_emblem")
    Parents="((Item=(Type=Bone,Name=\"upperarm_twist_01_l\"),Weight=0.037),(Item=(Type=Bone,Name=\"spine_04\"),Weight=0.77),(Item=(Type=Bone,Name=\"spine_05\"),Weight=0.15),(Item=(Type=Bone,Name=\"clavicle_l\"),Weight=0.01),(Item=(Type=Bone,Name=\"deform_upperarm_fr_l\"),Weight= ...(truncated)"
    Weight=1
   )
   n1 = &RigVMFunction_MathQuaternionToRotator(Value=&RigUnit_GetRelativeTransformForItem(Child=(Type=Bone,Name="spine_05"), Parent=(Type=Bone,Name="pelvis"), bChildInitial=False, bParentInitial=False).RelativeTransform.Rotation) {template="Cast::Execute(in Value,out Result)"}
   > &RigUnit_ModifyTransforms(
    ItemToModify=((Item=(Type=Bone,Name="jacket_spine_corr_base"),Transform=(Rotation=(X=0,Y=0,Z=0.1305,W=0.9914),Translation=(X=0,Y=0,Z=0),Scale3D=(X=1,Y=1,Z=1))))
    WeightMinimum=0
    Weight=&RigVMFunction_MathDoubleRemap(
     SourceMaximum=-25
     SourceMinimum=-5
     TargetMaximum=1
     TargetMinimum=0
     Value=RigVMDispatchNode(Struct=@n1.Result) {template="DISPATCH_RigVMDispatch_BreakStruct(in Struct,out Elements)"}.Elements.Yaw
    ).Result
   )
   > RigVMFunction_Sequence = RigVMAggregateNode()
   FN_DeformRig_FNC = RigVMFunctionReferenceNode(
    DeformRig_ Curve="(deformRig_l_upperarm_up_alpha_72_#1=true,deformRig_r_upperarm_up_alpha_73_#2=true,deformRig_l_upperarm_fr_alpha_74_#3=true,deformRig_r_upperarm_fr_alpha_75_CB51B80C4 ...(truncated)"
    DeformRig_Bool="(upperarm_1_#4=True,elbow_3_#5=True,wrist_7_#6=True,groin_8_#7=True,glute_17_#8=True,knee_9_40C408324CA261 ...(truncated)"
    DeformRig_Debug=(upperarm_debug_18_#4=False,wrist_debug_20_#6=False,lat_debug_26_#9=False,groinGlute_debug_27_#7=False)
    DeformRig_Value="(upperarm_up_18_#10=(Rotation=(X=0,Y=0,Z=0,W=1),Translation=(X=0,Y=0,Z=8),Scale3D=(X=0.7,Y=1,Z=1)),upperarm_fwd_fr_19_#11=(Rotation=(X=0,Y=0,Z=0,W=1),Translation=(X=0,Y=10,Z=0),Scale3D=(X=0.5,Y=1,Z= ...(truncated)"
    ExecuteContext=@RigVMFunction_Sequence.A
   )
   Get Delta Ref = RigVMFunctionReferenceNode(Item=(Type=Bone,Name="upperarm_twist_01_r"), ExecuteContext=@RigVMFunction_Sequence.C)
   Multiply = &RigVMFunction_MathDoubleMul(B=-1, A=@"Get Delta Ref".Result.Rotation.Roll)
   Make = RigVMDispatchNode(Elements=(Pitch=0,Yaw=0,Roll=0), Elements.Roll=&RigVMFunction_MathDoubleMul(B=0.5, A=@Multiply.Result).Result) {template="DISPATCH_RigVMDispatch_MakeStruct(in Elements,out Struct)"}
   Cast = /Script/RigVM.RigVMFunction_MathQuaternionFromRotatorV2(Value=@Make.Struct) {is=struct RigVMUnitNode, template="Cast::Execute(in Value,out Result)"}
   > &RigUnit_ModifyTransforms(
    ItemToModify=((Item=(Type=Bone,Name="upperarm_pad_twist_corr_r"),Transform=(Rotation=(X=0,Y=0,Z=0,W=1),Translation=(X=0,Y=0,Z=0),Scale3D=(X=1,Y=1,Z=1))))
    Weight=1
    WeightMinimum=0
    ItemToModify.0.Transform.Rotation=@Cast.Result
   )
   > Get Delta Ref_1 = RigVMFunctionReferenceNode(Item=(Type=Bone,Name="upperarm_twist_01_l"))
   n2 = &RigVMFunction_MathDoubleMul(B=0.5, A=&RigVMFunction_MathDoubleMul(B=-1, A=@"Get Delta Ref_1".Result.Rotation.Roll).Result)
   > &RigUnit_ModifyTransforms(
    ItemToModify=((Item=(Type=Bone,Name="upperarm_pad_twist_corr_l"),Transform=(Rotation=(X=0,Y=0,Z=0,W=1),Translation=(X=0,Y=0,Z=0),Scale3D=(X=1,Y=1,Z=1))))
    Weight=1
    WeightMinimum=0
    ItemToModify.0.Transform.Rotation=&RigVMFunction_MathQuaternionFromRotatorV2(Value=RigVMDispatchNode(Elements=(Pitch=0,Yaw=0,Roll=0), Elements.Roll=@n2.Result) {template="DISPATCH_RigVMDispatch_MakeStruct(in Elements,out Struct)"}.Struct) {template="Cast::Execute(in Value,out Result)"}.Result
   )
   n3 = &RigVMFunction_MathBoolToFloat(Value=&RigVMFunction_MathBoolNot(Value=RigVMVariableNode() {variable=IsSkydive}.Value).Result)
   n4 = &RigVMFunction_MathDoubleRemap(SourceMaximum=30, SourceMinimum=0, TargetMaximum=1, TargetMinimum=0, Value=@"Get Delta Ref".Result.Rotation.Roll)
   > &RigUnit_ModifyTransforms(
    ItemToModify=((Item=(Type=Bone,Name="upperarm_pad_twist_corr_r"),Transform=(Rotation=(X=0,Y=0,Z=0,W=1),Translation=(X=0,Y=2,Z=0),Scale3D=(X=1,Y=1,Z=1))))
    WeightMinimum=0
    Weight=&RigVMFunction_MathDoubleMul(A=@n3.Result, B=@n4.Result).Result
   )
   n5 = &RigVMFunction_MathDoubleRemap(SourceMaximum=30, SourceMinimum=0, TargetMaximum=1, TargetMinimum=0, Value=@"Get Delta Ref_1".Result.Rotation.Roll)
   > &RigUnit_ModifyTransforms(
    ItemToModify=((Item=(Type=Bone,Name="upperarm_pad_twist_corr_l"),Transform=(Rotation=(X=0,Y=0,Z=0,W=1),Translation=(X=0,Y=-2,Z=0),Scale3D=(X=1,Y=1,Z=1))))
    WeightMinimum=0
    Weight=&RigVMFunction_MathDoubleMul(B=@n3.Result, A=@n5.Result).Result
   )
   n6 = &RigVMFunction_MathBoolToFloat(Value=RigVMVariableNode() {variable=IsSkydive}.Value)
   > &RigUnit_ModifyTransforms(
    ItemToModify=((Item=(Type=Bone,Name="upperarm_pad_twist_corr_r"),Transform=(Rotation=(X=0,Y=0,Z=0,W=1),Translation=(X=-1,Y=-2,Z=0),Scale3D=(X=1,Y=1,Z=1))))
    WeightMinimum=0
    Weight=&RigVMFunction_MathDoubleMul(A=@n6.Result, B=@n4.Result).Result
   )
   > &RigUnit_ModifyTransforms(
    ItemToModify="((Item=(Type=Bone,Name=\"upperarm_pad_twist_corr_l\"),Transform=(Rotation=(X=0,Y=0,Z=0,W=1),Translation=(X=1,Y=2,Z=0),Scale3D=(X=1,Y=1,Z=1))),(Item=(Type=Bone,Name=\"armpad_moon_l\"),Transform=(Rotation=(X=0,Y=0,Z=0,W=1),Translation=(X=0,Y=0,Z=0),Scale3D=(X=1, ...(truncated)"
    WeightMinimum=0
    Weight=&RigVMFunction_MathDoubleMul(A=@n6.Result, B=@n5.Result).Result
   )
   > n7 = &RigUnit_SphericalPoseReader(
    ActiveRegionScaleFactors=(PositiveWidth=0.1,NegativeWidth=1,PositiveHeight=1,NegativeHeight=1)
    ActiveRegionSize=0.1
    Debug=(bDrawDebug=false,bDraw2D=False,bDrawLocalAxes=False,DebugScale=25,DebugSegments=20,DebugThickness=0.25)
    DriverAxis=(X=1,Y=0,Z=0)
    DriverItem=(Type=Bone,Name="upperarm_l")
    FalloffRegionScaleFactors=(PositiveWidth=1,NegativeWidth=1,PositiveHeight=1,NegativeHeight=1)
    FalloffSize=0.4
    FlipHeightScaling=False
    FlipWidthScaling=false
    OptionalParentItem=(Type=Bone,Name="None")
    RotationOffset=(X=0,Y=-55,Z=0)
   )
   > n8 = &RigUnit_SphericalPoseReader(
    ActiveRegionScaleFactors=(PositiveWidth=0.1,NegativeWidth=1,PositiveHeight=1,NegativeHeight=1)
    ActiveRegionSize=0.1
    Debug=(bDrawDebug=false,bDraw2D=False,bDrawLocalAxes=False,DebugScale=25,DebugSegments=20,DebugThickness=0.25)
    DriverAxis=(X=-1,Y=0,Z=0)
    DriverItem=(Type=Bone,Name="upperarm_r")
    FalloffRegionScaleFactors=(PositiveWidth=1,NegativeWidth=1,PositiveHeight=1,NegativeHeight=1)
    FalloffSize=0.4
    FlipHeightScaling=False
    FlipWidthScaling=false
    OptionalParentItem=(Type=Bone,Name="None")
    RotationOffset=(X=0,Y=125,Z=0)
   )
   n9 = &RigVMFunction_MathDoubleMul(A=@n7.OutputParam, B=@n8.OutputParam)
   > &RigUnit_ModifyTransforms(
    ItemToModify="((Item=(Type=Bone,Name=\"armPad_l\"),Transform=(Rotation=(X=0,Y=0,Z=0.0872,W=0.9962),Translation=(X=5,Y=-2,Z=0),Scale3D=(X=1,Y=1,Z=1))),(Item=(Type=Bone,Name=\"armpad_lower_01_l\"),Transform=(Rotation=(X=0,Y=0,Z=0,W=1),Translation=(X=-1.5,Y=0,Z=0),Scale3D=(X=1 ...(truncated)"
    WeightMinimum=0
    Weight=@n9.Result
   )
   > &RigUnit_ModifyTransforms(
    ItemToModify="((Item=(Type=Bone,Name=\"armPad_r\"),Transform=(Rotation=(X=0,Y=0,Z=0.1305,W=0.9914),Translation=(X=-5,Y=2,Z=0),Scale3D=(X=1,Y=1,Z=1))),(Item=(Type=Bone,Name=\"armpad_lower_01_r\"),Transform=(Rotation=(X=0,Y=0,Z=0,W=1),Translation=(X=1.5,Y=0,Z=0),Scale3D=(X=1, ...(truncated)"
    WeightMinimum=0
    Weight=@n9.Result
   )
   n10 = &RigUnit_GetTransform(Item=(Type=Bone,Name="lowerarm_l"), Space=LocalSpace, bInitial=False)
   n11 = /Script/ControlRig.RigUnit_AimBoneMath(
    DebugSettings=(bEnabled=False,Scale=10,WorldOffset=(Rotation=(X=0,Y=0,Z=0,W=1),Translation=(X=0,Y=0,Z=0),Scale3D=(X=1,Y=1,Z=1)))
    Primary=(Weight=1,Axis=(X=1,Y=0,Z=0),Target=(X=1,Y=0,Z=0),Kind=Location,Space=(Type=Bone,Name="None"))
    Secondary=(Weight=1,Axis=(X=0,Y=0,Z=1),Target=(X=1,Y=0,Z=0),Kind=Location,Space=(Type=Bone,Name="None"))
    Weight=1
    Primary.Target=@n10.Transform.Translation
    Secondary.Target=@n10.Transform.Translation
    InputTransform=&RigUnit_GetTransform(Item=(Type=Bone,Name="upperarm_pad_twist_corr_l"), Space=LocalSpace, bInitial=False).Transform
   ) {is=struct RigVMUnitNode}
   n12 = &RigVMFunction_MathQuaternionToEuler(RotationOrder=ZYX, Value=@n11.Result.Rotation)
   > &RigUnit_ModifyTransforms(
    ItemToModify=((Item=(Type=Bone,Name="armPadPivotHigh_l"),Transform=(Rotation=(X=0,Y=0,Z=0,W=1),Translation=(X=0,Y=0,Z=0),Scale3D=(X=1,Y=1,Z=1))))
    Weight=1
    WeightMinimum=0
    ItemToModify.0.Transform.Rotation=&RigVMFunction_MathQuaternionFromEuler(Euler=(X=0,Y=0,Z=0), Euler.Y=@n12.Result.Y, Euler.Z=@n12.Result.Z).Result
   )
   n13 = &RigUnit_GetTransform(Item=(Type=Bone,Name="upperarm_l"), Space=LocalSpace, bInitial=False)
   n14 = &RigUnit_AimBoneMath(
    DebugSettings=(bEnabled=False,Scale=10,WorldOffset=(Rotation=(X=0,Y=0,Z=0,W=1),Translation=(X=0,Y=0,Z=0),Scale3D=(X=1,Y=1,Z=1)))
    Primary=(Weight=1,Axis=(X=-1,Y=0,Z=0),Target=(X=1,Y=0,Z=0),Kind=Location,Space=(Type=Bone,Name="None"))
    Secondary=(Weight=1,Axis=(X=0,Y=0,Z=1),Target=(X=1,Y=0,Z=0),Kind=Location,Space=(Type=Bone,Name="None"))
    Weight=1
    Primary.Target=@n13.Transform.Translation
    Secondary.Target=@n13.Transform.Translation
    InputTransform=&RigUnit_GetTransform(Item=(Type=Bone,Name="armPadPivotLow_l"), Space=LocalSpace, bInitial=False).Transform
   )
   n15 = &RigVMFunction_MathQuaternionToEuler(RotationOrder=ZYX, Value=@n14.Result.Rotation)
   > &RigUnit_ModifyTransforms(
    ItemToModify=((Item=(Type=Bone,Name="armPadPivotLow_l"),Transform=(Rotation=(X=0,Y=0,Z=0,W=1),Translation=(X=0,Y=0,Z=0),Scale3D=(X=1,Y=1,Z=1))))
    Weight=1
    WeightMinimum=0
    ItemToModify.0.Transform.Rotation=&RigVMFunction_MathQuaternionFromEuler(Euler=(X=0,Y=0,Z=0), Euler.Z=&RigVMFunction_MathDoubleMul(B=-1, A=@n15.Result.Z).Result).Result
   )
   n16 = &RigUnit_GetTransform(Item=(Type=Bone,Name="lowerarm_r"), Space=LocalSpace, bInitial=False)
   n17 = &RigUnit_AimBoneMath(
    DebugSettings=(bEnabled=False,Scale=10,WorldOffset=(Rotation=(X=0,Y=0,Z=0,W=1),Translation=(X=0,Y=0,Z=0),Scale3D=(X=1,Y=1,Z=1)))
    Primary=(Weight=1,Axis=(X=-1,Y=0,Z=0),Target=(X=1,Y=0,Z=0),Kind=Location,Space=(Type=Bone,Name="None"))
    Secondary=(Weight=1,Axis=(X=0,Y=0,Z=1),Target=(X=1,Y=0,Z=0),Kind=Location,Space=(Type=Bone,Name="None"))
    Weight=1
    Primary.Target=@n16.Transform.Translation
    Secondary.Target=@n16.Transform.Translation
    InputTransform=&RigUnit_GetTransform(Item=(Type=Bone,Name="upperarm_pad_twist_corr_r"), Space=LocalSpace, bInitial=False).Transform
   )
   n18 = &RigVMFunction_MathQuaternionToEuler(RotationOrder=ZYX, Value=@n17.Result.Rotation)
   > &RigUnit_ModifyTransforms(
    ItemToModify=((Item=(Type=Bone,Name="armPadPivotHigh_r"),Transform=(Rotation=(X=0,Y=0,Z=0,W=1),Translation=(X=0,Y=0,Z=0),Scale3D=(X=1,Y=1,Z=1))))
    Weight=1
    WeightMinimum=0
    ItemToModify.0.Transform.Rotation=&RigVMFunction_MathQuaternionFromEuler(Euler=(X=0,Y=0,Z=0), Euler.Y=@n18.Result.Y, Euler.Z=@n18.Result.Z).Result
   )
   n19 = &RigUnit_GetTransform(Item=(Type=Bone,Name="upperarm_r"), Space=LocalSpace, bInitial=False)
   n20 = &RigUnit_AimBoneMath(
    DebugSettings=(bEnabled=False,Scale=10,WorldOffset=(Rotation=(X=0,Y=0,Z=0,W=1),Translation=(X=0,Y=0,Z=0),Scale3D=(X=1,Y=1,Z=1)))
    Primary=(Weight=1,Axis=(X=1,Y=0,Z=0),Target=(X=1,Y=0,Z=0),Kind=Location,Space=(Type=Bone,Name="None"))
    Secondary=(Weight=1,Axis=(X=0,Y=0,Z=1),Target=(X=1,Y=0,Z=0),Kind=Location,Space=(Type=Bone,Name="None"))
    Weight=1
    Primary.Target=@n19.Transform.Translation
    Secondary.Target=@n19.Transform.Translation
    InputTransform=&RigUnit_GetTransform(Item=(Type=Bone,Name="armPadPivotLow_r"), Space=LocalSpace, bInitial=False).Transform
   )
   n21 = &RigVMFunction_MathQuaternionToEuler(RotationOrder=ZYX, Value=@n20.Result.Rotation)
   > &RigUnit_ModifyTransforms(
    ItemToModify=((Item=(Type=Bone,Name="armPadPivotLow_r"),Transform=(Rotation=(X=0,Y=0,Z=0,W=1),Translation=(X=0,Y=0,Z=0),Scale3D=(X=1,Y=1,Z=1))))
    Weight=1
    WeightMinimum=0
    ItemToModify.0.Transform.Rotation=&RigVMFunction_MathQuaternionFromEuler(Euler=(X=0,Y=0,Z=0), Euler.Z=&RigVMFunction_MathDoubleMul(B=-1, A=@n21.Result.Z).Result).Result
   )
   > Get Delta Ref_2_2 = RigVMFunctionReferenceNode(Item=(Type=Bone,Name="lowerarm_l"))
   > &RigUnit_ModifyTransforms(
    ItemToModify="((Item=(Type=Bone,Name=\"upperarm_pad_twist_corr_l\"),Transform=(Rotation=(X=0,Y=0,Z=0,W=1),Translation=(X=0,Y=0,Z=0),Scale3D=(X=1,Y=1,Z=1))),(Item=(Type=Bone,Name=\"armpad_lower_02_l\"),Transform=(Rotation=(X=0,Y=0,Z=0,W=1),Translation=(X=-3,Y=0,Z=0),Scale3D= ...(truncated)"
    WeightMinimum=0
    Weight=&RigVMFunction_MathDoubleRemap(SourceMaximum=60, SourceMinimum=0, TargetMaximum=1, TargetMinimum=0, Value=@"Get Delta Ref_2_2".Result.Rotation.Yaw).Result
   )
   > Get Delta Ref_2_2_1 = RigVMFunctionReferenceNode(Item=(Type=Bone,Name="lowerarm_r"))
   > &RigUnit_ModifyTransforms(
    ItemToModify="((Item=(Type=Bone,Name=\"upperarm_pad_twist_corr_r\"),Transform=(Rotation=(X=0,Y=0,Z=0,W=1),Translation=(X=0,Y=0,Z=0),Scale3D=(X=1,Y=1,Z=1))),(Item=(Type=Bone,Name=\"armpad_lower_02_r\"),Transform=(Rotation=(X=0,Y=0,Z=0,W=1),Translation=(X=3,Y=0,Z=0),Scale3D=( ...(truncated)"
    WeightMinimum=0
    Weight=&RigVMFunction_MathDoubleRemap(SourceMaximum=60, SourceMinimum=0, TargetMaximum=1, TargetMinimum=0, Value=@"Get Delta Ref_2_2_1".Result.Rotation.Yaw).Result
   )
   n22 = &RigVMFunction_ControlFlowBranch(ExecuteContext=@RigVMFunction_Sequence.D, Condition=RigVMVariableNode() {variable=isDBNO}.Value)
   RigVMFunction_Sequence_3 = RigVMAggregateNode(ExecuteContext=@n22.False)
   RigVMFunction_Sequence_5 = RigVMAggregateNode(ExecuteContext=@RigVMFunction_Sequence_3.A)
   n23 = &RigUnit_SphericalPoseReader(
    ActiveRegionScaleFactors=(PositiveWidth=1.5,NegativeWidth=1,PositiveHeight=1.5,NegativeHeight=1.5)
    ActiveRegionSize=0.09
    Debug=(bDrawDebug=false,bDraw2D=false,bDrawLocalAxes=false,DebugScale=25,DebugSegments=20,DebugThickness=0.25)
    DriverAxis=(X=1,Y=0,Z=0)
    DriverItem=(Type=Bone,Name="thigh_r")
    FalloffRegionScaleFactors=(PositiveWidth=1,NegativeWidth=0.5,PositiveHeight=1,NegativeHeight=1)
    FalloffSize=0.35
    OptionalParentItem=(Type=Bone,Name="pelvis")
    RotationOffset=(X=-90,Y=180,Z=75)
    ExecutePin=@RigVMFunction_Sequence_5.A
   )
   > &RigUnit_ModifyTransforms(
    ItemToModify=((Item=(Type=Bone,Name="beltShield_base"),Transform=(Rotation=(X=0.0218,Y=0,Z=0,W=0.9998),Translation=(X=0,Y=0.5,Z=0),Scale3D=(X=1,Y=1,Z=1))))
    WeightMinimum=0
    Weight=@n23.OutputParam
   )
   > &RigUnit_ModifyTransforms(
    ItemToModify=((Item=(Type=Bone,Name="beltShiledSide_r"),Transform=(Rotation=(X=0.0218,Y=-0.001,Z=-0.0436,W=0.9988),Translation=(X=0,Y=0.5,Z=0),Scale3D=(X=1,Y=1,Z=1))))
    WeightMinimum=0
    Weight=&RigVMFunction_MathDoubleMul(A=@n23.OutputParam, B=RigVMVariableNode() {variable=Custom Idle is Playing Float}.Value).Result
   )
   > n24 = &RigUnit_SphericalPoseReader(
    ActiveRegionScaleFactors=(PositiveWidth=1,NegativeWidth=0.3,PositiveHeight=1,NegativeHeight=0.5)
    ActiveRegionSize=0.3
    Debug=(bDrawDebug=false,bDraw2D=false,bDrawLocalAxes=false,DebugScale=30,DebugSegments=30,DebugThickness=0.2)
    DriverAxis=(X=1,Y=0,Z=0)
    DriverItem=(Type=Bone,Name="thigh_r")
    FalloffRegionScaleFactors=(PositiveWidth=1,NegativeWidth=1,PositiveHeight=1,NegativeHeight=1)
    FalloffSize=0.6
    FlipHeightScaling=false
    FlipWidthScaling=false
    OptionalParentItem=(Type=Bone,Name="pelvis")
    RotationOffset=(X=-90,Y=180,Z=-10)
   )
   > &RigUnit_ModifyTransforms(
    ItemToModify=((Item=(Type=Bone,Name="jacket_front_base_r"),Transform=(Rotation=(X=0.3503,Y=0.0091,Z=0.2711,W=0.8965),Translation=(X=3,Y=-0.5,Z=-0.5),Scale3D=(X=1,Y=1,Z=1))))
    WeightMinimum=0
    Weight=@n24.OutputParam
   )
   > &RigUnit_ModifyTransforms(
    ItemToModify=((Item=(Type=Bone,Name="jacket_side_base_r"),Transform=(Rotation=(X=0,Y=0,Z=-0.0872,W=0.9962),Translation=(X=3,Y=0,Z=-3),Scale3D=(X=1,Y=1,Z=1))))
    WeightMinimum=0
    Weight=@n24.OutputParam
   )
   > &RigUnit_ModifyTransforms(
    ItemToModify=((Item=(Type=Bone,Name="jacket_back_base_r"),Transform=(Rotation=(X=0.0509,Y=0.083,Z=-0.0905,W=0.9911),Translation=(X=0,Y=0,Z=0),Scale3D=(X=1,Y=1,Z=1))))
    WeightMinimum=0
    Weight=@n24.OutputParam
   )
   > &RigUnit_ModifyTransforms(
    ItemToModify=((Item=(Type=Bone,Name="jacket_back_base_m"),Transform=(Rotation=(X=0,Y=0,Z=-0.0872,W=0.9962),Translation=(X=0,Y=0,Z=0),Scale3D=(X=1,Y=1,Z=1))))
    WeightMinimum=0
    Weight=@n24.OutputParam
   )
   > n25 = &RigUnit_SphericalPoseReader(
    ActiveRegionScaleFactors=(PositiveWidth=1,NegativeWidth=0.3,PositiveHeight=0.5,NegativeHeight=1)
    ActiveRegionSize=0.3
    Debug=(bDrawDebug=false,bDraw2D=false,bDrawLocalAxes=false,DebugScale=30,DebugSegments=30,DebugThickness=0.2)
    DriverAxis=(X=1,Y=0,Z=0)
    DriverItem=(Type=Bone,Name="thigh_r")
    FalloffRegionScaleFactors=(PositiveWidth=1,NegativeWidth=1,PositiveHeight=1,NegativeHeight=1)
    FalloffSize=0.6
    FlipHeightScaling=false
    FlipWidthScaling=false
    OptionalParentItem=(Type=Bone,Name="pelvis")
    RotationOffset=(X=-180,Y=195,Z=-10)
   )
   > &RigUnit_ModifyTransforms(
    ItemToModify=((Item=(Type=Bone,Name="jacket_front_base_r"),Transform=(Rotation=(X=0.4079,Y=-0.4307,Z=0.4399,W=0.6742),Translation=(X=4,Y=0,Z=-2),Scale3D=(X=1,Y=1,Z=1))))
    WeightMinimum=0
    Weight=@n25.OutputParam
   )
   > &RigUnit_ModifyTransforms(
    ItemToModify=((Item=(Type=Bone,Name="jacket_side_base_r"),Transform=(Rotation=(X=0.3062,Y=-0.5303,Z=0.0474,W=0.7891),Translation=(X=4,Y=1,Z=0),Scale3D=(X=1,Y=1,Z=1))))
    WeightMinimum=0
    Weight=@n25.OutputParam
   )
   > &RigUnit_ModifyTransforms(
    ItemToModify=((Item=(Type=Bone,Name="jacket_back_base_r"),Transform=(Rotation=(X=-0.3769,Y=-0.1604,Z=-0.0665,W=0.9098),Translation=(X=0,Y=0,Z=2),Scale3D=(X=1,Y=1,Z=1))))
    WeightMinimum=0
    Weight=@n25.OutputParam
   )
   > &RigUnit_ModifyTransforms(
    ItemToModify=((Item=(Type=Bone,Name="jacket_back_base_m"),Transform=(Rotation=(X=0,Y=0,Z=-0.0436,W=0.999),Translation=(X=0,Y=-2,Z=0),Scale3D=(X=1,Y=1,Z=1))))
    WeightMinimum=0
    Weight=@n25.OutputParam
   )
   RigVMFunction_MathBoolNot = &RigVMFunction_MathBoolNot(Value=RigVMVariableNode() {variable=IsCrouching}.Value)
   RigVMFunction_MathBoolAnd = &RigVMFunction_MathBoolAnd(B=@RigVMFunction_MathBoolNot.Result, A=&RigVMFunction_MathBoolNot(Value=RigVMVariableNode() {variable=IsSkydive}.Value).Result)
   n26 = &RigVMFunction_MathBoolToFloat(Value=@RigVMFunction_MathBoolAnd.Result)
   n27 = RigVMDispatchNode(Struct=&RigVMFunction_MathQuaternionToRotator(Value=&RigUnit_GetRelativeTransformForItem(Child=(Type=Bone,Name="spine_05"), Parent=(Type=Bone,Name="pelvis"), bChildInitial=False, bParentInitial=False).RelativeTransform.Rotation) {template="Cast::Execute(in Value,out Result)"}.Result) {template="DISPATCH_RigVMDispatch_BreakStruct(in Struct,out Elements)"}
   n28 = &RigVMFunction_MathDoubleRemap(SourceMaximum=16, SourceMinimum=4, TargetMaximum=1, TargetMinimum=0, Value=@n27.Elements.Pitch)
   > &RigUnit_ModifyTransforms(
    ItemToModify=((Item=(Type=Bone,Name="jacket_spine_corr_base"),Transform=(Rotation=(X=-0.0038,Y=0.0871,Z=0.0435,W=0.9952),Translation=(X=0,Y=0,Z=0),Scale3D=(X=1,Y=1,Z=1))))
    WeightMinimum=0
    Weight=&RigVMFunction_MathDoubleMul(
     B=&RigVMFunction_MathDoubleMul(A=@n26.Result, B=@n28.Result).Result
     A=&RigVMFunction_MathDoubleRemap(SourceMaximum=-13, SourceMinimum=-40, TargetMaximum=1, TargetMinimum=0, Value=@n27.Elements.Roll).Result
    ).Result
   )
   n29 = &RigUnit_SphericalPoseReader(
    ActiveRegionScaleFactors=(PositiveWidth=1.5,NegativeWidth=1,PositiveHeight=1.5,NegativeHeight=1.5)
    ActiveRegionSize=0.09
    Debug=(bDrawDebug=false,bDraw2D=false,bDrawLocalAxes=false,DebugScale=25,DebugSegments=20,DebugThickness=0.25)
    DriverAxis=(X=-1,Y=0,Z=0)
    DriverItem=(Type=Bone,Name="thigh_l")
    FalloffRegionScaleFactors=(PositiveWidth=1,NegativeWidth=0.5,PositiveHeight=1,NegativeHeight=1)
    FalloffSize=0.35
    OptionalParentItem=(Type=Bone,Name="pelvis")
    RotationOffset=(X=90,Y=0,Z=75)
    ExecutePin=@RigVMFunction_Sequence_5.B
   )
   > &RigUnit_ModifyTransforms(
    ItemToModify=((Item=(Type=Bone,Name="beltShield_base"),Transform=(Rotation=(X=0.0218,Y=0,Z=0,W=0.9998),Translation=(X=0,Y=0.5,Z=0),Scale3D=(X=1,Y=1,Z=1))))
    WeightMinimum=0
    Weight=@n29.OutputParam
   )
   > n30 = &RigUnit_ModifyTransforms(
    ItemToModify=((Item=(Type=Bone,Name="beltShiledSide_l"),Transform=(Rotation=(X=0.0218,Y=-0.001,Z=-0.0436,W=0.9988),Translation=(X=0,Y=0.5,Z=0),Scale3D=(X=1,Y=1,Z=1))))
    WeightMinimum=0
    Weight=&RigVMFunction_MathDoubleMul(A=@n29.OutputParam, B=RigVMVariableNode() {variable=Custom Idle is Playing Float}.Value).Result
   )
   n31 = &RigVMFunction_ControlFlowBranch(ExecuteContext=@n30.ExecutePin, Condition=RigVMVariableNode() {variable=IsCrouching}.Value)
   n32 = &RigUnit_SphericalPoseReader(
    ActiveRegionScaleFactors=(PositiveWidth=1,NegativeWidth=0.3,PositiveHeight=1,NegativeHeight=0.5)
    ActiveRegionSize=0.3
    Debug=(bDrawDebug=false,bDraw2D=false,bDrawLocalAxes=false,DebugScale=30,DebugSegments=30,DebugThickness=0.2)
    DriverAxis=(X=-1,Y=0,Z=0)
    DriverItem=(Type=Bone,Name="thigh_l")
    FalloffRegionScaleFactors=(PositiveWidth=1,NegativeWidth=1,PositiveHeight=1,NegativeHeight=1)
    FalloffSize=0.6
    FlipHeightScaling=false
    FlipWidthScaling=true
    OptionalParentItem=(Type=Bone,Name="pelvis")
    RotationOffset=(X=90,Y=180,Z=-10)
    ExecutePin=@n31.False
   )
   > &RigUnit_ModifyTransforms(
    ItemToModify=((Item=(Type=Bone,Name="jacket_front_base_l"),Transform=(Rotation=(X=0.1893,Y=-0.0381,Z=0.2685,W=0.9437),Translation=(X=-3,Y=0.5,Z=0.5),Scale3D=(X=1,Y=1,Z=1))))
    WeightMinimum=0
    Weight=@n32.OutputParam
   )
   > &RigUnit_ModifyTransforms(
    ItemToModify=((Item=(Type=Bone,Name="jacket_side_base_l"),Transform=(Rotation=(X=-0.0262,Y=-0.2996,Z=-0.0831,W=0.9501),Translation=(X=-3,Y=-1,Z=1),Scale3D=(X=1,Y=1,Z=1))))
    WeightMinimum=0
    Weight=@n32.OutputParam
   )
   > &RigUnit_ModifyTransforms(
    ItemToModify=((Item=(Type=Bone,Name="jacket_back_base_l"),Transform=(Rotation=(X=-0.0472,Y=-0.0851,Z=-0.0472,W=0.9941),Translation=(X=0,Y=-1,Z=1),Scale3D=(X=1,Y=1,Z=1))))
    WeightMinimum=0
    Weight=@n32.OutputParam
   )
   > &RigUnit_ModifyTransforms(
    ItemToModify=((Item=(Type=Bone,Name="jacket_back_base_m"),Transform=(Rotation=(X=0,Y=0,Z=0,W=1),Translation=(X=0,Y=0,Z=0),Scale3D=(X=1,Y=1,Z=1))))
    WeightMinimum=0
    Weight=@n32.OutputParam
   )
   > n33 = &RigUnit_SphericalPoseReader(
    ActiveRegionScaleFactors=(PositiveWidth=1,NegativeWidth=0.3,PositiveHeight=0.5,NegativeHeight=1)
    ActiveRegionSize=0.3
    Debug=(bDrawDebug=false,bDraw2D=false,bDrawLocalAxes=false,DebugScale=30,DebugSegments=30,DebugThickness=0.2)
    DriverAxis=(X=-1,Y=0,Z=0)
    DriverItem=(Type=Bone,Name="thigh_l")
    FalloffRegionScaleFactors=(PositiveWidth=1,NegativeWidth=1,PositiveHeight=1,NegativeHeight=1)
    FalloffSize=0.6
    FlipHeightScaling=false
    FlipWidthScaling=true
    OptionalParentItem=(Type=Bone,Name="pelvis")
    RotationOffset=(X=0,Y=195,Z=-10)
   )
   > &RigUnit_ModifyTransforms(
    ItemToModify=((Item=(Type=Bone,Name="jacket_front_base_l"),Transform=(Rotation=(X=0.4079,Y=-0.4307,Z=0.4399,W=0.6742),Translation=(X=-4,Y=0,Z=-2),Scale3D=(X=1,Y=1,Z=1))))
    WeightMinimum=0
    Weight=@n33.OutputParam
   )
   > &RigUnit_ModifyTransforms(
    ItemToModify=((Item=(Type=Bone,Name="jacket_side_base_l"),Transform=(Rotation=(X=0.0915,Y=-0.5245,Z=-0.0915,W=0.8415),Translation=(X=-4,Y=-1,Z=0),Scale3D=(X=1,Y=1,Z=1))))
    WeightMinimum=0
    Weight=@n33.OutputParam
   )
   > &RigUnit_ModifyTransforms(
    ItemToModify=((Item=(Type=Bone,Name="jacket_back_base_l"),Transform=(Rotation=(X=0.3946,Y=0.1099,Z=-0.1846,W=0.8934),Translation=(X=0,Y=0,Z=-2),Scale3D=(X=1,Y=1,Z=1))))
    WeightMinimum=0
    Weight=@n33.OutputParam
   )
   > &RigUnit_ModifyTransforms(
    ItemToModify=((Item=(Type=Bone,Name="jacket_back_base_m"),Transform=(Rotation=(X=0,Y=0,Z=-0.0436,W=0.999),Translation=(X=0,Y=-2,Z=0),Scale3D=(X=1,Y=1,Z=1))))
    WeightMinimum=0
    Weight=@n33.OutputParam
   )
   n34 = &RigUnit_SphericalPoseReader(
    ActiveRegionScaleFactors=(PositiveWidth=1,NegativeWidth=0.3,PositiveHeight=1,NegativeHeight=1)
    ActiveRegionSize=0.3
    Debug=(bDrawDebug=false,bDraw2D=false,bDrawLocalAxes=false,DebugScale=30,DebugSegments=30,DebugThickness=0.2)
    DriverAxis=(X=-1,Y=0,Z=0)
    DriverItem=(Type=Bone,Name="thigh_l")
    FalloffRegionScaleFactors=(PositiveWidth=1,NegativeWidth=1,PositiveHeight=1,NegativeHeight=1)
    FalloffSize=0.5
    OptionalParentItem=(Type=Bone,Name="pelvis")
    RotationOffset=(X=90,Y=0,Z=-10)
    ExecutePin=@n31.True
   )
   > &RigUnit_ModifyTransforms(
    ItemToModify=((Item=(Type=Bone,Name="jacket_front_base_l"),Transform=(Rotation=(X=0.1691,Y=-0.07,Z=0.3919,W=0.9016),Translation=(X=-3,Y=-2,Z=2),Scale3D=(X=1,Y=1,Z=1))))
    WeightMinimum=0
    Weight=@n34.OutputParam
   )
   > &RigUnit_ModifyTransforms(
    ItemToModify=((Item=(Type=Bone,Name="jacket_side_base_l"),Transform=(Rotation=(X=0,Y=-0.1305,Z=0,W=0.9914),Translation=(X=0,Y=0,Z=0),Scale3D=(X=1,Y=1,Z=1))))
    WeightMinimum=0
    Weight=@n34.OutputParam
   )
   > &RigUnit_ModifyTransforms(
    ItemToModify=((Item=(Type=Bone,Name="jacket_back_base_l"),Transform=(Rotation=(X=-0.0435,Y=0.0038,Z=-0.0871,W=0.9952),Translation=(X=0,Y=0,Z=0),Scale3D=(X=1,Y=1,Z=1))))
    WeightMinimum=0
    Weight=@n34.OutputParam
   )
   > &RigUnit_ModifyTransforms(
    ItemToModify=((Item=(Type=Bone,Name="jacket_back_base_m"),Transform=(Rotation=(X=0,Y=0,Z=0,W=1),Translation=(X=0,Y=0,Z=0),Scale3D=(X=1,Y=1,Z=1))))
    WeightMinimum=0
    Weight=@n34.OutputParam
   )
   n35 = RigVMVariableNode() {variable=IsCrouching}
   n36 = &RigVMFunction_ControlFlowBranch(
    Condition=&RigVMFunction_MathBoolAnd(A=&RigVMFunction_MathBoolNot(Value=@n35.Value).Result, B=RigVMVariableNode() {variable=isMelee}.Value).Result
    ExecuteContext=@RigVMFunction_Sequence_5.C
   )
   n37 = RigVMDispatchNode(Struct=&RigVMFunction_MathQuaternionToRotator(Value=&RigUnit_GetRelativeTransformForItem(Child=(Type=Bone,Name="spine_05"), Parent=(Type=Bone,Name="pelvis"), bChildInitial=False, bParentInitial=False).RelativeTransform.Rotation) {template="Cast::Execute(in Value,out Result)"}.Result) {template="DISPATCH_RigVMDispatch_BreakStruct(in Struct,out Elements)"}
   &RigUnit_ModifyTransforms(
    ItemToModify="((Item=(Type=Bone,Name=\"jacket_front_base_r\"),Transform=(Rotation=(X=0.2841,Y=-0.0345,Z=0.2424,W=0.927),Translation=(X=4.4258,Y=-1.6316,Z=-12.0541),Scale3D=(X=1,Y=1,Z=1))),(Item=(Type=Bone,Name=\"jacket_side_base_r\"),Transform=(Rotation=(X=0.2378,Y=-0.0183, ...(truncated)"
    WeightMinimum=0
    Weight=&RigVMFunction_MathDoubleMul(
     A=&RigVMFunction_MathDoubleRemap(SourceMaximum=35, SourceMinimum=10, TargetMaximum=1, TargetMinimum=0, Value=@n37.Elements.Yaw).Result
     B=&RigVMFunction_MathDoubleRemap(SourceMaximum=50, SourceMinimum=10, TargetMaximum=1, TargetMinimum=0, Value=@n37.Elements.Roll).Result
    ).Result
    ExecutePin=@n36.True
   )
   > &RigUnit_ModifyTransforms(
    ItemToModify="((Item=(Type=Bone,Name=\"jacket_front_base_l\"),Transform=(Rotation=(X=0.0236,Y=0.073,Z=0.1456,W=0.9864),Translation=(X=-1.4826,Y=4.2499,Z=0.0963),Scale3D=(X=1,Y=1,Z=1))),(Item=(Type=Bone,Name=\"jacket_spine_corr_base\"),Transform=(Rotation=(X=0.0076,Y=-0.0868 ...(truncated)"
    WeightMinimum=0
    Weight=&RigVMFunction_MathDoubleRemap(SourceMaximum=-45, SourceMinimum=-10, TargetMaximum=1, TargetMinimum=0, Value=@n37.Elements.Roll).Result
   )
   n38 = &RigVMFunction_ControlFlowBranch(Condition=RigVMVariableNode() {variable=IsSkydive}.Value, ExecuteContext=@RigVMFunction_Sequence_3.D)
   CalculatePelvisSkydive = RigVMFunctionReferenceNode()
   n39 = &RigVMFunction_KalmanFloat(BufferSize=16, Value=&RigVMFunction_MathDoubleMul(B=1, A=@CalculatePelvisSkydive.Result).Result) {template="KalmanFilter::Execute(in Value,in BufferSize,out Result)"}
   &RigUnit_ModifyTransforms(
    ItemToModify="((Item=(Type=Bone,Name=\"jacket_back_base_m\"),Transform=(Rotation=(X=0,Y=0,Z=-0.1268,W=0.9919),Translation=(X=0,Y=0,Z=0),Scale3D=(X=1,Y=1,Z=1))),(Item=(Type=Bone,Name=\"jacket_back_base_l\"),Transform=(Rotation=(X=0,Y=0,Z=-0.1564,W=0.9877),Translation=(X=0,Y= ...(truncated)"
    WeightMinimum=0
    Weight=@n39.Result
    ExecutePin=@n38.True
   )
   > &RigUnit_ModifyTransforms(
    ItemToModify="((Item=(Type=Bone,Name=\"jacket_side_base_l\"),Transform=(Rotation=(X=-0.1211,Y=-0.0985,Z=0.0833,W=0.9842),Translation=(X=-2.6,Y=-0.1,Z=1),Scale3D=(X=1,Y=1,Z=1))),(Item=(Type=Bone,Name=\"jacket_side_base_r\"),Transform=(Rotation=(X=-0.1211,Y=-0.0985,Z=0.0833,W ...(truncated)"
    WeightMinimum=0
    Weight=@n39.Result
   )
   > &RigUnit_ModifyTransforms(
    ItemToModify="((Item=(Type=Bone,Name=\"jacket_front_base_l\"),Transform=(Rotation=(X=-0.1736,Y=0.0424,Z=0.0731,W=0.9812),Translation=(X=-0.2024,Y=3.1895,Z=-1.4283),Scale3D=(X=1,Y=1,Z=1))),(Item=(Type=Bone,Name=\"jacket_front_base_r\"),Transform=(Rotation=(X=-0.1738,Y=0.0416 ...(truncated)"
    WeightMinimum=0
    Weight=@n39.Result
   )
   > &RigUnit_ModifyTransforms(
    ItemToModify=((Item=(Type=Bone,Name="jacket_spine_corr_base"),Transform=(Rotation=(X=0,Y=0,Z=0.0872,W=0.9962),Translation=(X=0,Y=0,Z=0),Scale3D=(X=1,Y=1,Z=1))))
    WeightMinimum=0
    Weight=@n39.Result
   )
   Remap = &RigVMFunction_MathDoubleRemap(
    SourceMaximum=20
    SourceMinimum=-20
    TargetMaximum=-20
    TargetMinimum=20
    Value=&RigVMFunction_MathQuaternionToEuler(
     RotationOrder=ZYX
     Value=&RigUnit_GetRelativeTransformForItem(Child=(Type=Bone,Name="spine_04"), Parent=(Type=Bone,Name="pelvis"), bChildInitial=False, bParentInitial=False).RelativeTransform.Rotation
    ).Result.Y
   )
   > &RigUnit_ModifyTransforms(
    ItemToModify=((Item=(Type=Bone,Name="jacket_spine_corr_base"),Transform=(Rotation=(X=0,Y=0,Z=0.0872,W=0.9962),Translation=(X=0,Y=0,Z=0),Scale3D=(X=1,Y=1,Z=1))))
    WeightMinimum=0
    ItemToModify.0.Transform.Rotation=&RigVMFunction_MathQuaternionFromEuler(Euler=(X=0,Y=0,Z=0), Euler.Y=@Remap.Result).Result
    Weight=@n39.Result
   )
   n40 = &RigVMFunction_MathDoubleRemap(
    SourceMaximum=0
    SourceMinimum=-30
    TargetMaximum=10
    TargetMinimum=0
    Value=&RigVMFunction_MathQuaternionToEuler(
     RotationOrder=ZYX
     Value=&RigUnit_GetRelativeTransformForItem(Child=(Type=Bone,Name="spine_01"), Parent=(Type=Bone,Name="pelvis"), bChildInitial=False, bParentInitial=False).RelativeTransform.Rotation
    ).Result.Z
   )
   > &RigUnit_ModifyTransforms(
    ItemToModify=((Item=(Type=Bone,Name="jacket_spine_corr_base"),Transform=(Rotation=(X=0,Y=0,Z=0.0872,W=0.9962),Translation=(X=0,Y=0,Z=0),Scale3D=(X=1,Y=1,Z=1))))
    WeightMinimum=0
    Weight=&RigVMFunction_KalmanFloat(BufferSize=16, Value=@n40.Result) {template="KalmanFilter::Execute(in Value,in BufferSize,out Result)"}.Result
    ItemToModify.0.Transform.Rotation=&RigVMFunction_MathQuaternionFromEuler(Euler=(X=0,Y=0,Z=0), Euler.Z=@n40.Result).Result
   )
   n41 = &RigUnit_GetTransform(Item=(Type=Bone,Name="pelvis"), bInitial=False)
   > /Script/ControlRig.RigUnit_AimItem(
    DebugSettings=(bEnabled=False,Scale=10,WorldOffset=(Rotation=(X=0,Y=0,Z=0,W=1),Translation=(X=0,Y=0,Z=0),Scale3D=(X=1,Y=1,Z=1)))
    Item=(Type=Bone,Name="jacket_spine_corr_base")
    Primary=(Weight=1,Axis=(X=-1,Y=0,Z=0),Target=(X=1,Y=0,Z=0),Kind=Location,Space=(Type=Bone,Name="None"))
    Secondary=(Weight=1,Axis=(X=0,Y=0,Z=-1),Target=(X=1,Y=0,Z=0),Kind=Location,Space=(Type=Bone,Name="None"))
    Weight=1
    Primary.Target=@n41.Transform.Translation
    Secondary.Target=@n41.Transform.Translation
   ) {is=struct RigVMUnitNode}
   > &RigUnit_ModifyTransforms(
    ItemToModify=((Item=(Type=Bone,Name="jacket_spine_corr_base"),Transform=(Rotation=(X=0,Y=0,Z=0.0436,W=0.999),Translation=(X=0,Y=0,Z=0),Scale3D=(X=1,Y=1,Z=1))))
    Weight=1
    WeightMinimum=0
   )
   n42 = RigVMVariableNode() {variable=IsCrouching}
   n43 = &RigVMFunction_ControlFlowBranch(Condition=@n42.Value, ExecuteContext=@RigVMFunction_Sequence_3.C)
   n44 = &RigUnit_SphericalPoseReader(
    ActiveRegionScaleFactors=(PositiveWidth=1,NegativeWidth=0.1,PositiveHeight=1,NegativeHeight=1)
    ActiveRegionSize=0.3
    Debug=(bDrawDebug=false,bDraw2D=false,bDrawLocalAxes=false,DebugScale=30,DebugSegments=30,DebugThickness=0.2)
    DriverAxis=(X=-1,Y=0,Z=0)
    DriverItem=(Type=Bone,Name="thigh_l")
    FalloffRegionScaleFactors=(PositiveWidth=1,NegativeWidth=1,PositiveHeight=1,NegativeHeight=1)
    FalloffSize=0.5
    OptionalParentItem=(Type=Bone,Name="pelvis")
    RotationOffset=(X=110,Y=-5,Z=0)
    ExecutePin=@n43.True
   )
   > n45 = &RigUnit_SphericalPoseReader(
    ActiveRegionScaleFactors=(PositiveWidth=1,NegativeWidth=0.1,PositiveHeight=1,NegativeHeight=1)
    ActiveRegionSize=0.3
    Debug=(bDrawDebug=false,bDraw2D=false,bDrawLocalAxes=false,DebugScale=30,DebugSegments=30,DebugThickness=0.2)
    DriverAxis=(X=1,Y=0,Z=0)
    DriverItem=(Type=Bone,Name="thigh_r")
    FalloffRegionScaleFactors=(PositiveWidth=1,NegativeWidth=1,PositiveHeight=1,NegativeHeight=1)
    FalloffSize=0.5
    FlipHeightScaling=false
    FlipWidthScaling=false
    OptionalParentItem=(Type=Bone,Name="pelvis")
    RotationOffset=(X=-90,Y=170,Z=0)
   )
   RigVMFunction_MathBoolToFloat = &RigVMFunction_MathBoolToFloat(Value=@n42.Value)
   n46 = &RigVMFunction_MathDoubleMul(A=@RigVMFunction_MathBoolToFloat.Result, B=@n44.OutputParam)
   > &RigUnit_ModifyTransforms(
    ItemToModify="((Item=(Type=Bone,Name=\"jacket_front_base_l\"),Transform=(Rotation=(X=0.2147,Y=-0.1888,Z=0.3977,W=0.8718),Translation=(X=1,Y=-1,Z=1),Scale3D=(X=1,Y=1,Z=1))),(Item=(Type=Bone,Name=\"jacket_side_base_l\"),Transform=(Rotation=(X=0.096,Y=-0.2896,Z=-0.2028,W=0.930 ...(truncated)"
    WeightMinimum=0
    Weight=@n46.Result
   )
   n47 = &RigVMFunction_MathDoubleMul(B=@n45.OutputParam, A=@RigVMFunction_MathBoolToFloat.Result)
   > &RigUnit_ModifyTransforms(
    ItemToModify="((Item=(Type=Bone,Name=\"jacket_front_base_r\"),Transform=(Rotation=(X=0.1175,Y=0.0917,Z=0.0325,W=0.9883),Translation=(X=0,Y=-1,Z=0),Scale3D=(X=1,Y=1,Z=1))),(Item=(Type=Bone,Name=\"jacket_side_base_r\"),Transform=(Rotation=(X=-0.0019,Y=0.0436,Z=0.0436,W=0.9981 ...(truncated)"
    WeightMinimum=0
    Weight=@n47.Result
   )
   > &RigUnit_ModifyTransforms(
    ItemToModify="((Item=(Type=Bone,Name=\"jacket_back_base_l\"),Transform=(Rotation=(X=0,Y=0,Z=-0.0436,W=0.999),Translation=(X=-1,Y=-1,Z=0),Scale3D=(X=1,Y=1,Z=1))),(Item=(Type=Bone,Name=\"jacket_back_base_m\"),Transform=(Rotation=(X=0,Y=0,Z=-0.0872,W=0.9962),Translation=(X=2,Y ...(truncated)"
    WeightMinimum=0
    Weight=@RigVMFunction_MathBoolToFloat.Result
   )
   n48 = &RigVMFunction_MathBoolNot(Value=RigVMVariableNode() {variable=isMelee}.Value)
   n49 = &RigVMFunction_MathQuaternionToRotator(Value=&RigUnit_GetRelativeTransformForItem(Child=(Type=Bone,Name="spine_05"), Parent=(Type=Bone,Name="spine_01"), bChildInitial=False, bParentInitial=False).RelativeTransform.Rotation) {template="Cast::Execute(in Value,out Result)"}
   Multiply_18 = RigVMAggregateNode(
    A=@n47.Result
    B=&RigVMFunction_MathDoubleRemap(
     SourceMaximum=70
     SourceMinimum=50
     TargetMaximum=1
     TargetMinimum=0
     Value=&RigUnit_GetRelativeTransformForItem(Child=(Type=Bone,Name="pelvis"), Parent=(Type=Bone,Name="root"), bChildInitial=False, bParentInitial=False).RelativeTransform.Translation.Z
    ).Result
    C=&RigVMFunction_MathDoubleRemap(
     SourceMaximum=26
     SourceMinimum=16
     TargetMaximum=1
     TargetMinimum=0
     Value=RigVMDispatchNode(Struct=@n49.Result) {template="DISPATCH_RigVMDispatch_BreakStruct(in Struct,out Elements)"}.Elements.Yaw
    ).Result
   )
   > &RigUnit_ModifyTransforms(
    ItemToModify="((Item=(Type=Bone,Name=\"jacket_front_base_r\"),Transform=(Rotation=(X=0.1648,Y=0.1006,Z=0.0704,W=0.9786),Translation=(X=2,Y=-3,Z=0),Scale3D=(X=1,Y=1,Z=1))),(Item=(Type=Bone,Name=\"jacket_side_base_r\"),Transform=(Rotation=(X=0.1394,Y=-0.1306,Z=0.0296,W=0.9811 ...(truncated)"
    WeightMinimum=0
    Weight=&RigVMFunction_MathDoubleMul(B=&RigVMFunction_MathBoolToFloat(Value=@n48.Result).Result, A=@Multiply_18.Result).Result
   )
   n50 = RigVMDispatchNode(Struct=&RigVMFunction_MathQuaternionToRotator(Value=&RigUnit_GetRelativeTransformForItem(Child=(Type=Bone,Name="spine_05"), Parent=(Type=Bone,Name="spine_01"), bChildInitial=False, bParentInitial=False).RelativeTransform.Rotation) {template="Cast::Execute(in Value,out Result)"}.Result) {template="DISPATCH_RigVMDispatch_BreakStruct(in Struct,out Elements)"}
   > &RigUnit_ModifyTransforms(
    ItemToModify="((Item=(Type=Bone,Name=\"jacket_front_base_l\"),Transform=(Rotation=(X=-0.0871,Y=0.0038,Z=-0.0435,W=0.9952),Translation=(X=-2,Y=6,Z=0),Scale3D=(X=1,Y=1,Z=1))),(Item=(Type=Bone,Name=\"jacket_side_base_l\"),Transform=(Rotation=(X=-0.0538,Y=0.0229,Z=0.0623,W=0.99 ...(truncated)"
    WeightMinimum=0
    Weight=&RigVMFunction_MathDoubleMul(
     B=&RigVMFunction_MathDoubleRemap(SourceMaximum=24, SourceMinimum=10, TargetMaximum=1, TargetMinimum=0, Value=@n50.Elements.Yaw).Result
     A=@n46.Result
    ).Result
   )
   > &RigUnit_ModifyTransforms(
    ItemToModify="((Item=(Type=Bone,Name=\"jacket_front_base_l\"),Transform=(Rotation=(X=0.0199,Y=-0.1776,Z=-0.1209,W=0.9764),Translation=(X=0,Y=7,Z=0),Scale3D=(X=1,Y=1,Z=1))),(Item=(Type=Bone,Name=\"jacket_side_base_l\"),Transform=(Rotation=(X=0.0698,Y=0.0007,Z=0.0104,W=0.9975 ...(truncated)"
    WeightMinimum=0
    Weight=&RigVMFunction_MathDoubleRemap(SourceMaximum=-40, SourceMinimum=0, TargetMaximum=1, TargetMinimum=0, Value=@n50.Elements.Roll).Result
   )
   n51 = &RigVMFunction_MathBoolToFloat(Value=RigVMVariableNode() {variable=isMelee}.Value)
   n52 = &RigVMFunction_MathDoubleMul(
    A=@n51.Result
    B=&RigVMFunction_MathDoubleRemap(SourceMaximum=-80, SourceMinimum=-40, TargetMaximum=1, TargetMinimum=0, Value=@n50.Elements.Roll).Result
   )
   > &RigUnit_ModifyTransforms(
    ItemToModify="((Item=(Type=Bone,Name=\"jacket_front_base_l\"),Transform=(Rotation=(X=0.0141,Y=-0.0476,Z=0.2686,W=0.962),Translation=(X=0,Y=0,Z=0),Scale3D=(X=1,Y=1,Z=1))),(Item=(Type=Bone,Name=\"jacket_side_base_l\"),Transform=(Rotation=(X=0.0196,Y=-0.0121,Z=0.1078,W=0.9939) ...(truncated)"
    WeightMinimum=0
    Weight=@n52.Result
   )
   > &RigUnit_ModifyTransforms(
    ItemToModify=((Item=(Type=Bone,Name="jacket_spine_corr_base"),Transform=(Rotation=(X=-0.2588,Y=0,Z=0,W=0.9659),Translation=(X=0,Y=0,Z=0),Scale3D=(X=1,Y=1,Z=1))))
    WeightMinimum=0
    Weight=@n52.Result
   )
   n53 = &RigVMFunction_MathDoubleMul(
    A=@n51.Result
    B=&RigVMFunction_MathDoubleRemap(SourceMaximum=60, SourceMinimum=0, TargetMaximum=1, TargetMinimum=0, Value=@n50.Elements.Roll).Result
   )
   > &RigUnit_ModifyTransforms(
    ItemToModify="((Item=(Type=Bone,Name=\"jacket_front_base_r\"),Transform=(Rotation=(X=0.4184,Y=-0.0546,Z=0.1496,W=0.8942),Translation=(X=4.3893,Y=-11.0969,Z=-14.0763),Scale3D=(X=1,Y=1,Z=1))),(Item=(Type=Bone,Name=\"jacket_side_base_r\"),Transform=(Rotation=(X=0.3717,Y=-0.165 ...(truncated)"
    WeightMinimum=0
    Weight=&RigVMFunction_MathDoubleMul(
     A=@n53.Result
     B=&RigVMFunction_MathDoubleRemap(SourceMaximum=35, SourceMinimum=10, TargetMaximum=1, TargetMinimum=0, bClamp=False, Value=@n50.Elements.Yaw).Result
    ).Result
   )
   n54 = &RigVMFunction_ControlFlowBranch(ExecuteContext=@n22.True, Condition=RigVMVariableNode() {variable=isDBNO}.Value)
   n55 = &RigUnit_GetTransform(Item=(Type=Bone,Name="pelvis"), bInitial=False)
   &RigUnit_AimItem(
    DebugSettings=(bEnabled=False,Scale=10,WorldOffset=(Rotation=(X=0,Y=0,Z=0,W=1),Translation=(X=0,Y=0,Z=0),Scale3D=(X=1,Y=1,Z=1)))
    Item=(Type=Bone,Name="jacket_spine_corr_base")
    Primary=(Weight=1,Axis=(X=-1,Y=0,Z=0),Target=(X=1,Y=0,Z=0),Kind=Location,Space=(Type=Bone,Name="None"))
    Secondary=(Weight=1,Axis=(X=0,Y=0,Z=-1),Target=(X=1,Y=0,Z=0),Kind=Location,Space=(Type=Bone,Name="None"))
    Weight=1
    Primary.Target=@n55.Transform.Translation
    Secondary.Target=@n55.Transform.Translation
    ExecutePin=@n54.True
   )
   > &RigUnit_ModifyTransforms(
    ItemToModify="((Item=(Type=Bone,Name=\"jacket_front_base_l\"),Transform=(Rotation=(X=0.6037,Y=0.0694,Z=0.2177,W=0.7638),Translation=(X=0.684,Y=0.996,Z=3.2162),Scale3D=(X=1,Y=1,Z=1))),(Item=(Type=Bone,Name=\"jacket_side_base_l\"),Transform=(Rotation=(X=0.0774,Y=-0.018,Z=0.22 ...(truncated)"
    Weight=1
    WeightMinimum=0
   )
   > &RigUnit_ModifyTransforms(
    ItemToModify="((Item=(Type=Bone,Name=\"jacket_front_base_r\"),Transform=(Rotation=(X=0.335,Y=0.0235,Z=0.2978,W=0.8936),Translation=(X=2.6224,Y=-3.7711,Z=-4.3934),Scale3D=(X=1,Y=1,Z=1))),(Item=(Type=Bone,Name=\"jacket_side_base_r\"),Transform=(Rotation=(X=0.0893,Y=-0.0315,Z= ...(truncated)"
    Weight=1
    WeightMinimum=0
   )
   > &RigUnit_SphericalPoseReader(
    ActiveRegionScaleFactors=(PositiveWidth=1,NegativeWidth=1,PositiveHeight=1,NegativeHeight=0.1)
    ActiveRegionSize=0.3
    Debug=(bDrawDebug=false,bDraw2D=False,bDrawLocalAxes=False,DebugScale=25,DebugSegments=20,DebugThickness=0.25)
    DriverAxis=(X=-1,Y=0,Z=0)
    DriverItem=(Type=Bone,Name="thigh_l")
    FalloffRegionScaleFactors=(PositiveWidth=1,NegativeWidth=1,PositiveHeight=1,NegativeHeight=1)
    FalloffSize=0.5
    FlipHeightScaling=False
    FlipWidthScaling=False
    OptionalParentItem=(Type=Bone,Name="None")
    RotationOffset=(X=120,Y=15,Z=20)
   )
   n56 = &RigUnit_GetRelativeTransformForItem(Child=(Type=Bone,Name="thigh_l"), Parent=(Type=Bone,Name="thigh_l"), bChildInitial=False, bParentInitial=true)
   > &RigUnit_ModifyTransforms(
    ItemToModify="((Item=(Type=Bone,Name=\"jacket_front_base_l\"),Transform=(Rotation=(X=0.0858,Y=-0.0503,Z=0.1347,W=0.9859),Translation=(X=-1.3706,Y=6.2969,Z=1.0972),Scale3D=(X=1,Y=1,Z=1))),(Item=(Type=Bone,Name=\"jacket_side_base_l\"),Transform=(Rotation=(X=0.2376,Y=-0.2374,Z ...(truncated)"
    WeightMinimum=0
    Weight=&RigVMFunction_MathDoubleRemap(
     SourceMaximum=-25
     SourceMinimum=-10
     TargetMaximum=1
     TargetMinimum=0
     Value=&RigVMFunction_MathQuaternionToEuler(RotationOrder=ZYX, Value=@n56.RelativeTransform.Rotation).Result.Y
    ).Result
   )
   > &RigUnit_SphericalPoseReader(
    ActiveRegionScaleFactors=(PositiveWidth=1,NegativeWidth=1,PositiveHeight=1,NegativeHeight=0.1)
    ActiveRegionSize=0.3
    Debug=(bDrawDebug=false,bDraw2D=False,bDrawLocalAxes=False,DebugScale=25,DebugSegments=20,DebugThickness=0.25)
    DriverAxis=(X=1,Y=0,Z=0)
    DriverItem=(Type=Bone,Name="thigh_r")
    FalloffRegionScaleFactors=(PositiveWidth=1,NegativeWidth=1,PositiveHeight=1,NegativeHeight=1)
    FalloffSize=0.5
    FlipHeightScaling=false
    FlipWidthScaling=true
    OptionalParentItem=(Type=Bone,Name="None")
    RotationOffset=(X=-60,Y=15,Z=20)
   )
   n57 = &RigUnit_GetRelativeTransformForItem(Child=(Type=Bone,Name="thigh_r"), Parent=(Type=Bone,Name="thigh_r"), bChildInitial=False, bParentInitial=true)
   > &RigUnit_ModifyTransforms(
    ItemToModify="((Item=(Type=Bone,Name=\"jacket_front_base_r\"),Transform=(Rotation=(X=0.0425,Y=-0.2254,Z=0.0732,W=0.9706),Translation=(X=-0.393,Y=-0.5862,Z=-0.227),Scale3D=(X=1,Y=1,Z=1))),(Item=(Type=Bone,Name=\"jacket_side_base_r\"),Transform=(Rotation=(X=0.0516,Y=-0.1267,Z ...(truncated)"
    WeightMinimum=0
    Weight=&RigVMFunction_MathDoubleRemap(
     SourceMaximum=-25
     SourceMinimum=-10
     TargetMaximum=1
     TargetMinimum=0
     Value=&RigVMFunction_MathQuaternionToEuler(RotationOrder=ZYX, Value=@n57.RelativeTransform.Rotation).Result.Y
    ).Result
   )
   Get Delta Ref_3 = RigVMFunctionReferenceNode(Item=(Type=Bone,Name="thigh_r"), ExecuteContext=@RigVMFunction_Sequence.B)
   > Get Delta Ref_3_1 = RigVMFunctionReferenceNode(Item=(Type=Bone,Name="thigh_l"))
   n58 = &RigVMFunction_MathDoubleRemap(SourceMaximum=-90, SourceMinimum=0, TargetMaximum=1, TargetMinimum=0, Value=@"Get Delta Ref_3".Result.Rotation.Yaw)
   > &RigUnit_ModifyTransforms(
    ItemToModify="((Item=(Type=Bone,Name=\"beltShiledSide_r\"),Transform=(Rotation=(X=0.0308,Y=-0.0308,Z=-0.7064,W=0.7064),Translation=(X=0,Y=5,Z=0),Scale3D=(X=0.9,Y=1,Z=1))),(Item=(Type=Bone,Name=\"belt_base_main\"),Transform=(Rotation=(X=0.3003,Y=0.1193,Z=-0.0557,W=0.9447),Tr ...(truncated)"
    WeightMinimum=0
    Weight=@n58.Result
   )
   n59 = &RigVMFunction_MathDoubleRemap(SourceMaximum=-90, SourceMinimum=0, TargetMaximum=1, TargetMinimum=0, Value=@"Get Delta Ref_3_1".Result.Rotation.Yaw)
   > &RigUnit_ModifyTransforms(
    ItemToModify="((Item=(Type=Bone,Name=\"beltShiledSide_l\"),Transform=(Rotation=(X=0.0308,Y=-0.0308,Z=-0.7064,W=0.7064),Translation=(X=0,Y=-5,Z=0),Scale3D=(X=0.9,Y=1,Z=1))),(Item=(Type=Bone,Name=\"belt_base_main\"),Transform=(Rotation=(X=0.3003,Y=-0.1193,Z=0.0557,W=0.9447),T ...(truncated)"
    WeightMinimum=0
    Weight=@n59.Result
   )
   > &RigUnit_ModifyTransforms(
    ItemToModify="((Item=(Type=Bone,Name=\"beltShield_01\"),Transform=(Rotation=(X=0,Y=0,Z=0,W=1),Translation=(X=0,Y=1,Z=3),Scale3D=(X=1,Y=1,Z=1))),(Item=(Type=Bone,Name=\"belt_base_main\"),Transform=(Rotation=(X=-0.4073,Y=0.0439,Z=0.0094,W=0.9122),Translation=(X=-2.3323,Y=-6.1 ...(truncated)"
    WeightMinimum=0
    Weight=&RigVMFunction_MathDoubleMul(A=@n58.Result, B=@n59.Result).Result
   )
   > &RigUnit_ParentConstraint(
    AdvancedSettings=(InterpolationType=Average,RotationOrderForFilter=XZY)
    Child=(Type=Bone,Name="belt_side_l")
    Filter=(TranslationFilter=(bX=false,bY=false,bZ=true),RotationFilter=(bX=true,bY=True,bZ=True),ScaleFilter=(bX=True,bY=True,bZ=True))
    Parents=((Item=(Type=Bone,Name="pelvis"),Weight=0.1),(Item=(Type=Bone,Name="belt_base_main"),Weight=1))
    Weight=1
    bMaintainOffset=true
   )
   > n60 = &RigUnit_ParentConstraint(
    AdvancedSettings=(InterpolationType=Average,RotationOrderForFilter=XZY)
    Child=(Type=Bone,Name="belt_side_r")
    Filter=(TranslationFilter=(bX=false,bY=false,bZ=true),RotationFilter=(bX=true,bY=true,bZ=true),ScaleFilter=(bX=True,bY=True,bZ=True))
    Parents=((Item=(Type=Bone,Name="pelvis"),Weight=0.1),(Item=(Type=Bone,Name="belt_base_main"),Weight=1))
    Weight=1
    bMaintainOffset=true
   )
   RigVMFunction_ControlFlowBranch = &RigVMFunction_ControlFlowBranch(ExecuteContext=@n60.ExecutePin, Condition=RigVMVariableNode() {variable=isMelee}.Value)
   n61 = RigVMDispatchNode(Struct=&RigVMFunction_MathQuaternionToRotator(Value=&RigUnit_GetRelativeTransformForItem(Child=(Type=Bone,Name="spine_05"), Parent=(Type=Bone,Name="pelvis"), bChildInitial=False, bParentInitial=False).RelativeTransform.Rotation) {template="Cast::Execute(in Value,out Result)"}.Result) {template="DISPATCH_RigVMDispatch_BreakStruct(in Struct,out Elements)"}
   &RigUnit_ModifyTransforms(
    ItemToModify="((Item=(Type=Bone,Name=\"beltShield_corrUpCorner_l\"),Transform=(Rotation=(X=0,Y=0,Z=0,W=1),Translation=(X=0,Y=0,Z=0),Scale3D=(X=1,Y=1,Z=1))),(Item=(Type=Bone,Name=\"beltShield_corrUpCorner_r\"),Transform=(Rotation=(X=0,Y=0,Z=0,W=1),Translation=(X=-1,Y=-3,Z=0) ...(truncated)"
    WeightMinimum=0
    Weight=&RigVMFunction_MathDoubleRemap(SourceMaximum=60, SourceMinimum=10, TargetMaximum=1, TargetMinimum=0, Value=@n61.Elements.Roll).Result
    ExecutePin=@RigVMFunction_ControlFlowBranch.True
   )
   > &RigUnit_ModifyTransforms(
    ItemToModify="((Item=(Type=Bone,Name=\"beltShield_corrUpCorner_l\"),Transform=(Rotation=(X=0,Y=0,Z=0,W=1),Translation=(X=-1,Y=-2,Z=-2),Scale3D=(X=0.5,Y=0.5,Z=0.5))),(Item=(Type=Bone,Name=\"beltShield_corrUpCorner_r\"),Transform=(Rotation=(X=0,Y=0,Z=0,W=1),Translation=(X=-2, ...(truncated)"
    WeightMinimum=0
    Weight=&RigVMFunction_MathDoubleRemap(SourceMaximum=-46, SourceMinimum=-10, TargetMaximum=1, TargetMinimum=0, Value=@n61.Elements.Roll).Result
   )
   &RigUnit_ParentConstraint(
    AdvancedSettings=(RotationOrderForFilter=XZY)
    Child=(Type=Bone,Name="Skirt_corrective")
    Filter=(TranslationFilter=(bX=false,bY=false,bZ=false),RotationFilter=(bX=true,bY=true,bZ=true),ScaleFilter=(bX=true,bY=true,bZ=true))
    Parents=((Item=(Type=Bone,Name="thigh_l"),Weight=0.5),(Item=(Type=Bone,Name="thigh_r"),Weight=0.5))
    Weight=1
    ExecutePin=@RigVMFunction_Sequence.E
   )
   > n62 = &RigUnit_SphericalPoseReader(
    ActiveRegionScaleFactors=(PositiveWidth=3,NegativeWidth=2.5,PositiveHeight=3,NegativeHeight=3)
    ActiveRegionSize=0.08
    Debug=(bDrawDebug=false,DebugScale=25,DebugSegments=20,DebugThickness=0.25)
    DriverAxis=(X=-1,Y=0,Z=0)
    DriverItem=(Type=Bone,Name="thigh_l")
    FalloffRegionScaleFactors=(PositiveWidth=1,NegativeWidth=1,PositiveHeight=1,NegativeHeight=1)
    FalloffSize=0.43
    FlipHeightScaling=false
    OptionalParentItem=(Type=Bone)
    RotationOffset=(X=100,Y=0,Z=0)
   )
   > n63 = &RigUnit_SphericalPoseReader(
    ActiveRegionScaleFactors=(PositiveWidth=2.5,NegativeWidth=3,PositiveHeight=3,NegativeHeight=3)
    ActiveRegionSize=0.08
    Debug=(bDrawDebug=false,DebugScale=25,DebugSegments=20,DebugThickness=0.25)
    DriverAxis=(X=1,Y=0,Z=0)
    DriverItem=(Type=Bone,Name="thigh_r")
    FalloffRegionScaleFactors=(PositiveWidth=1,NegativeWidth=1,PositiveHeight=1,NegativeHeight=1)
    FalloffSize=0.43
    FlipHeightScaling=false
    FlipWidthScaling=false
    OptionalParentItem=(Type=Bone)
    RotationOffset=(X=-80,Y=0,Z=0)
   )
   > n64 = &RigUnit_SphericalPoseReader(
    ActiveRegionScaleFactors=(PositiveWidth=1.5,NegativeWidth=1,PositiveHeight=2.5,NegativeHeight=2.5)
    ActiveRegionSize=0.05
    Debug=(bDrawDebug=false,DebugScale=25,DebugSegments=20,DebugThickness=0.25)
    DriverAxis=(X=-1,Y=0,Z=0)
    DriverItem=(Type=Bone,Name="thigh_l")
    FalloffRegionScaleFactors=(PositiveWidth=1,NegativeWidth=1,PositiveHeight=1,NegativeHeight=1)
    FalloffSize=0.18
    FlipHeightScaling=false
    OptionalParentItem=(Type=Bone)
    RotationOffset=(X=-90,Y=0,Z=-60)
   )
   > n65 = &RigUnit_SphericalPoseReader(
    ActiveRegionScaleFactors=(PositiveWidth=1,NegativeWidth=1.5,PositiveHeight=2.5,NegativeHeight=2.5)
    ActiveRegionSize=0.05
    Debug=(bDrawDebug=false,DebugScale=25,DebugSegments=20,DebugThickness=0.25)
    DriverAxis=(X=1,Y=0,Z=0)
    DriverItem=(Type=Bone,Name="thigh_r")
    FalloffRegionScaleFactors=(PositiveWidth=1,NegativeWidth=1,PositiveHeight=1,NegativeHeight=1)
    FalloffSize=0.18
    FlipHeightScaling=false
    OptionalParentItem=(Type=Bone)
    RotationOffset=(X=90,Y=0,Z=-60)
   )
   Maximum = RigVMAggregateNode(D=@n65.OutputParam, A=@n62.OutputParam, B=@n63.OutputParam, C=@n64.OutputParam)
   Subtract = &RigVMFunction_MathDoubleSub(A=1, B=RigVMVariableNode() {variable=Custom Idle is Playing Float}.Value)
   > &RigUnit_ModifyTransforms(
    ItemToModify=((Item=(Type=Bone,Name="Skirt_corrective"),Transform=(Rotation=(X=0,Y=0,Z=-0.0872,W=0.9962),Translation=(X=-13,Y=-13,Z=0),Scale3D=(X=1,Y=1,Z=1))))
    Weight=&RigVMFunction_MathDoubleMul(A=@Maximum.Result, B=@Subtract.Result).Result
   )
   &RigUnit_ModifyTransforms(
    ExecutePin=()
    ItemToModify="((Item=(Type=Bone,Name=\"jacket_front_base_r\"),Transform=(Rotation=(X=0.2598,Y=0.0225,Z=0.2074,W=0.9428),Translation=(X=-0.6835,Y=-6.521,Z=-1.0685),Scale3D=(X=1,Y=1,Z=1))),(Item=(Type=Bone,Name=\"jacket_side_base_r\"),Transform=(Rotation=(X=0.3739,Y=-0.1614,Z ...(truncated)"
    WeightMinimum=0
    Weight=@n53.Result
   )
   &RigUnit_ModifyTransforms(
    ExecutePin=()
    ItemToModify=((Item=(Type=Bone,Name="thigh_l"),Transform=(Rotation=(X=0,Y=0,Z=0,W=1),Translation=(X=0,Y=0,Z=0),Scale3D=(X=1,Y=1,Z=1))),(Item=(Type=Bone,Name="thigh_r"),Transform=(Rotation=(X=0,Y=0,Z=0,W=1),Translation=(X=0,Y=0,Z=0),Scale3D=(X=1,Y=1,Z=1))))
    Weight=1
    WeightMinimum=0
   )
   RigVMRerouteNode(Value=())
   &RigVMFunction_ControlFlowBranch(ExecuteContext=(), Condition=RigVMVariableNode() {variable=isSprinting}.Value)
   &RigVMFunction_ControlFlowBranch(ExecuteContext=(), Condition=RigVMVariableNode() {variable=isSprinting}.Value)
  ]}
]
hierarchy [
 root {loc=<0 0 0>, rot=<0 0 0 1>}
 pelvis {parent=root, loc=<0 -0.5301 98.0571>, rot=<0.0127 0.707 -0.0127 -0.707>}
 spine_01 {parent=pelvis, loc=<6.1208 -0.9328 0>, rot=<0 0 0.034 -0.9994>}
 spine_02 {parent=spine_01, loc=<6.5997 -0.2095 0>, rot=<0 0 0.0238 -0.9997>}
 spine_03 {parent=spine_02, loc=<6.4914 -0.1237 0>, rot=<0 0 0.0149 -0.9999>}
 spine_04 {parent=spine_03, loc=<6.9598 0.0794 0>, rot=<0 0 -0.0132 -0.9999>}
 spine_05 {parent=spine_04, loc=<6.9987 0 0>, rot=<0 0 -0.0015 -1>}
 attach_cape {parent=spine_05, loc=<9.0259 -11.7125 0.0002>, rot=<0.0616 -0.7044 -0.0616 -0.7044>}
 neck_01 {parent=spine_05, loc=<11.3138 0.1529 0>, rot=<0 0 -0.2003 -0.9797>}
 neck_02 {parent=neck_01, loc=<4.5627 0 0>, rot=<0 0 0.0504 -0.9987>}
 head {parent=neck_02, loc=<5.3914 0 0>, rot=<0 0 0.115 -0.9934>}
 attach_backpack {parent=spine_05, loc=<0 -8.35 0>, rot=<-0.0093 -0.707 0.0093 -0.707>}
 clavicle_l {parent=spine_05, loc=<5.4831 -0.1053 -3.2946>, rot=<0.0465 -0.7289 -0.009 -0.6829>}
 upperarm_l {parent=clavicle_l, loc=<9.8658 -1.2136 -0.4105>, rot=<0.0164 -0.3245 -0.0269 -0.9454>}
 lowerarm_l {parent=upperarm_l, loc=<26.465 -0.8345 0.259>, rot=<0.0036 0.0161 -0.2298 -0.9731>}
 hand_l {parent=lowerarm_l, loc=<22.7355 -0.3203 -0.6105>, rot=<-0.6784 -0.0223 0.0735 -0.7307>}
 thumb_01_l {parent=hand_l, loc=<1.3118 -0.3191 -2.459>, rot=<0.6546 -0.4042 0.0664 -0.6354>}
 thumb_02_l {parent=thumb_01_l, loc=<5.9336 -0.0448 0.0401>, rot=<-0.0037 0.0213 0.298 -0.9543>}
 thumb_03_l {parent=thumb_02_l, loc=<2.2223 0.3682 -0.0907>, rot=<0.0118 -0.013 0.0066 -0.9998>}
 index_metacarpal_l {parent=hand_l, loc=<1.6535 0.0183 -2.96>, rot=<0.0158 -0.0124 -0.0173 -0.9996>}
 index_01_l {parent=index_metacarpal_l, loc=<8.3008 0.0357 -0.0524>, rot=<-0.0044 0.0485 0.0924 -0.9945>}
 index_02_l {parent=index_01_l, loc=<3.3587 -0.3004 0.0452>, rot=<-0.002 0.0027 0.3225 -0.9466>}
 index_03_l {parent=index_02_l, loc=<2.5578 -0.123 -0.0252>, rot=<-0.002 0.0067 0.2558 -0.9667>}
 middle_metacarpal_l {parent=hand_l, loc=<2.1856 -0.1725 -0.5984>, rot=<0.006 0.0168 -0.0329 -0.9993>}
 middle_01_l {parent=middle_metacarpal_l, loc=<7.3275 -0.2447 0.0129>, rot=<-0.0208 0.0042 0.0858 -0.9961>}
 middle_02_l {parent=middle_01_l, loc=<4.0123 0.1401 0.1175>, rot=<-0.0218 0.0258 0.3512 -0.9357>}
 middle_03_l {parent=middle_02_l, loc=<2.6077 0.0534 0.1373>, rot=<0.0082 0.0188 0.1708 -0.9851>}
 ring_metacarpal_l {parent=hand_l, loc=<2.2074 -0.3625 1.3638>, rot=<-0.0284 0.0722 -0.0458 -0.9959>}
 ring_01_l {parent=ring_metacarpal_l, loc=<7.0295 -0.6197 -0.1446>, rot=<-0.0489 -0.0252 0.1892 -0.9804>}
 ring_02_l {parent=ring_01_l, loc=<3.6174 0.4647 -0.1027>, rot=<0 0 0.2806 -0.9598>}
 ring_03_l {parent=ring_02_l, loc=<2.4152 0.114 0.0199>, rot=<0 0 0.1189 -0.9929>}
 pinky_metacarpal_l {parent=hand_l, loc=<1.7616 -0.9753 3.3814>, rot=<-0.0397 0.0814 -0.069 -0.9935>}
 pinky_01_l {parent=pinky_metacarpal_l, loc=<7.1709 -0.4831 -0.1094>, rot=<-0.0654 -0.0035 0.2242 -0.9723>}
 pinky_02_l {parent=pinky_01_l, loc=<2.8274 0.0893 -0.1043>, rot=<0.0037 0.0011 0.2854 -0.9584>}
 pinky_03_l {parent=pinky_02_l, loc=<2.2046 0.0307 -0.0078>, rot=<-0.0002 0.0098 0.1445 -0.9895>}
 deform_wrist_dn_base_l {parent=hand_l, loc=<0 -1.5326 0>, rot=<0 0 0 -1>}
 deform_wrist_dn_l {parent=deform_wrist_dn_base_l, loc=<0 0 0>, rot=<0 0 0 1>}
 deform_wrist_up_base_l {parent=hand_l, loc=<0 2.3009 0>, rot=<0 0 0 -1>}
 deform_wrist_up_l {parent=deform_wrist_up_base_l, loc=<0 0 0>, rot=<0 0 0 1>}
 lowerarm_twist_02_l {parent=lowerarm_l, loc=<10.6707 0 0>, rot=<0 0 0 1>}
 lowerarm_twist_01_l {parent=lowerarm_l, loc=<18.316 0 0>, rot=<0 0 0 1>}
 deform_elbow_in_base_l {parent=lowerarm_l, loc=<0.1834 3.6062 -0.2572>, rot=<0 0 0 -1>}
 deform_elbow_in_l {parent=deform_elbow_in_base_l, loc=<0 0 0>, rot=<0 0 0 1>}
 deform_elbow_out_base_l {parent=lowerarm_l, loc=<0 -4 0>, rot=<0 0 0 1>}
 deform_elbow_out_l {parent=deform_elbow_out_base_l, loc=<0 0 0>, rot=<0 0 0 1>}
 upperarm_twist_01_l {parent=upperarm_l, loc=<7.2135 -0.1134 0>, rot=<0 0 0 1>}
 deform_upperarm_bk_base_l {parent=upperarm_twist_01_l, loc=<-6.4053 -4.3746 0.314>, rot=<0 0 0 -1>}
 deform_upperarm_bk_l {parent=deform_upperarm_bk_base_l, loc=<0 0 0>, rot=<0 0 0 1>}
 deform_upperarm_fr_base_l {parent=upperarm_twist_01_l, loc=<-5.6312 5.092 0.3354>, rot=<0 0 0 -1>}
 deform_upperarm_fr_l {parent=deform_upperarm_fr_base_l, loc=<0 0 0>, rot=<0 0 0 1>}
 deform_upperarm_up_base_l {parent=upperarm_twist_01_l, loc=<-8.891 0.4825 3.9517>, rot=<0 0 0 -1>}
 deform_upperarm_up_l {parent=deform_upperarm_up_base_l, loc=<0 0 0>, rot=<0 0 0 1>}
 upperarm_pad_twist_corr_l {parent=upperarm_twist_01_l, loc=<0 0 0>, rot=<0 0 0 1>}
 armPadPivotHigh_l {parent=upperarm_pad_twist_corr_l, loc=<-11.0545 1.8267 5.5268>, rot=<0 0 0 -1>}
 armPadPivotLow_l {parent=armPadPivotHigh_l, loc=<29.3781 -2.1324 2.9792>, rot=<0 0 0 1>}
 armPad_l {parent=armPadPivotLow_l, loc=<-19.4466 1.1655 -3.9692>, rot=<0 0 0 1>}
 armpad_lower_01_l {parent=armPad_l, loc=<14.0015 -1.0758 4.3746>, rot=<-0.0074 0.0744 0.0018 -0.9972>}
 armpad_lower_02_l {parent=armpad_lower_01_l, loc=<5.1836 -0.0535 -0.9386>, rot=<0 0 0 1>}
 armpad_moon_l {parent=armPad_l, loc=<-3.276 0.1727 0.6656>, rot=<-0.0074 0.0812 0.0019 -0.9967>}
 upperarm_twist_02_l {parent=upperarm_l, loc=<19.7998 -0.1134 0>, rot=<0 0 0 1>}
 clavicle_r {parent=spine_05, loc=<5.4831 -0.1053 3.2946>, rot=<0.7289 0.0465 -0.6829 0.009>}
 upperarm_r {parent=clavicle_r, loc=<-9.8658 1.2136 0.4105>, rot=<0.0164 -0.3245 -0.0269 -0.9454>}
 lowerarm_r {parent=upperarm_r, loc=<-26.465 0.8345 -0.259>, rot=<0.0036 0.0161 -0.2298 -0.9731>}
 hand_r {parent=lowerarm_r, loc=<-22.7355 0.3203 0.6106>, rot=<-0.6784 -0.0223 0.0735 -0.7307>}
 thumb_01_r {parent=hand_r, loc=<-1.3118 0.3191 2.459>, rot=<0.6546 -0.4042 0.0664 -0.6354>}
 thumb_02_r {parent=thumb_01_r, loc=<-5.9336 0.0448 -0.0401>, rot=<-0.0037 0.0213 0.298 -0.9543>}
 thumb_03_r {parent=thumb_02_r, loc=<-2.2223 -0.3682 0.0907>, rot=<0.0118 -0.013 0.0066 -0.9998>}
 index_metacarpal_r {parent=hand_r, loc=<-1.6535 -0.0183 2.96>, rot=<0.0158 -0.0124 -0.0173 -0.9996>}
 index_01_r {parent=index_metacarpal_r, loc=<-8.3008 -0.0357 0.0524>, rot=<-0.0044 0.0485 0.0924 -0.9945>}
 index_02_r {parent=index_01_r, loc=<-3.3587 0.3004 -0.0452>, rot=<-0.002 0.0027 0.3225 -0.9466>}
 index_03_r {parent=index_02_r, loc=<-2.5578 0.123 0.0252>, rot=<-0.002 0.0067 0.2558 -0.9667>}
 middle_metacarpal_r {parent=hand_r, loc=<-2.1856 0.1725 0.5984>, rot=<0.006 0.0168 -0.0329 -0.9993>}
 middle_01_r {parent=middle_metacarpal_r, loc=<-7.3275 0.2447 -0.0129>, rot=<-0.0208 0.0042 0.0858 -0.9961>}
 middle_02_r {parent=middle_01_r, loc=<-4.0123 -0.1401 -0.1175>, rot=<-0.0218 0.0258 0.3512 -0.9357>}
 middle_03_r {parent=middle_02_r, loc=<-2.6077 -0.0534 -0.1373>, rot=<0.0082 0.0188 0.1708 -0.9851>}
 ring_metacarpal_r {parent=hand_r, loc=<-2.2074 0.3625 -1.3638>, rot=<-0.0284 0.0722 -0.0458 -0.9959>}
 ring_01_r {parent=ring_metacarpal_r, loc=<-7.0295 0.6197 0.1446>, rot=<-0.0489 -0.0252 0.1892 -0.9804>}
 ring_02_r {parent=ring_01_r, loc=<-3.6174 -0.4647 0.1027>, rot=<0 0 0.2806 -0.9598>}
 ring_03_r {parent=ring_02_r, loc=<-2.4152 -0.114 -0.0199>, rot=<0 0 0.1189 -0.9929>}
 pinky_metacarpal_r {parent=hand_r, loc=<-1.7616 0.9753 -3.3814>, rot=<-0.0397 0.0814 -0.069 -0.9935>}
 pinky_01_r {parent=pinky_metacarpal_r, loc=<-7.1709 0.4831 0.1094>, rot=<-0.0654 -0.0035 0.2242 -0.9723>}
 pinky_02_r {parent=pinky_01_r, loc=<-2.8274 -0.0893 0.1043>, rot=<0.0037 0.0011 0.2854 -0.9584>}
 pinky_03_r {parent=pinky_02_r, loc=<-2.2046 -0.0307 0.0078>, rot=<-0.0002 0.0098 0.1445 -0.9895>}
 deform_wrist_dn_base_r {parent=hand_r, loc=<-0.0001 1.5328 0>, rot=<0 0 0 -1>}
 deform_wrist_dn_r {parent=deform_wrist_dn_base_r, loc=<0 0 0>, rot=<0 0 0 1>}
 deform_wrist_up_base_r {parent=hand_r, loc=<-0.0002 -2.3007 -0.0001>, rot=<0 0 0 -1>}
 deform_wrist_up_r {parent=deform_wrist_up_base_r, loc=<0 0 0>, rot=<0 0 0 1>}
 lowerarm_twist_02_r {parent=lowerarm_r, loc=<-10.6707 0 0>, rot=<0 0 0 -1>}
 lowerarm_twist_01_r {parent=lowerarm_r, loc=<-18.316 0 0>, rot=<0 0 0 -1>}
 deform_elbow_in_base_r {parent=lowerarm_r, loc=<-0.1834 -3.6062 0.2572>, rot=<0 0 0 -1>}
 deform_elbow_in_r {parent=deform_elbow_in_base_r, loc=<0 0 0>, rot=<0 0 0 1>}
 deform_elbow_out_base_r {parent=lowerarm_r, loc=<0 4 -0.0001>, rot=<0 0 0 -1>}
 deform_elbow_out_r {parent=deform_elbow_out_base_r, loc=<0 0 0>, rot=<0 0 0 1>}
 upperarm_twist_01_r {parent=upperarm_r, loc=<-7.2135 0.1134 0>, rot=<0 0 0 -1>}
 deform_upperarm_bk_base_r {parent=upperarm_twist_01_r, loc=<6.4053 4.3746 -0.314>, rot=<0 0 0 -1>}
 deform_upperarm_bk_r {parent=deform_upperarm_bk_base_r, loc=<0 0 0>, rot=<0 0 0 1>}
 deform_upperarm_fr_base_r {parent=upperarm_twist_01_r, loc=<5.6312 -5.092 -0.3354>, rot=<0 0 0 -1>}
 deform_upperarm_fr_r {parent=deform_upperarm_fr_base_r, loc=<0 0 0>, rot=<0 0 0 1>}
 deform_upperarm_up_base_r {parent=upperarm_twist_01_r, loc=<8.891 -0.4825 -3.9517>, rot=<0 0 0 -1>}
 upperarm_pad_twist_corr_r {parent=upperarm_twist_01_r, loc=<-0.0003 0 0.0003>, rot=<0 0 0 1>}
 armPadPivotHigh_r {parent=upperarm_pad_twist_corr_r, loc=<11.0547 -1.8267 -5.527>, rot=<0 0 0 -1>}
 armPadPivotLow_r {parent=armPadPivotHigh_r, loc=<-29.3783 2.1324 -2.979>, rot=<0 0 0 -1>}
 armPad_r {parent=armPadPivotLow_r, loc=<19.4467 -1.1655 3.9691>, rot=<0 0 0 -1>}
 armpad_lower_01_r {parent=armPad_r, loc=<-14.0013 1.0759 -4.3748>, rot=<-0.0074 0.0744 0.0018 -0.9972>}
 armpad_lower_02_r {parent=armpad_lower_01_r, loc=<-5.1834 0.0536 0.9383>, rot=<0 0 0 1>}
 armpad_moon_r {parent=armPad_r, loc=<3.2764 -0.1727 -0.6661>, rot=<-0.0074 0.0812 0.0019 -0.9967>}
 upperarm_twist_02_r {parent=upperarm_r, loc=<-19.7998 0.1134 0>, rot=<0 0 0 -1>}
 dyn_emblem {parent=spine_05, loc=<-0.9675 10.4333 -10.9774>, rot=<-0.3574 -0.533 -0.1888 -0.7433>}
 jacket_spine_corr_base {parent=spine_03, loc=<0 0 0>, rot=<0 0 0 1>}
 jacket_front_base_l {parent=jacket_spine_corr_base, loc=<-3.4281 12.2285 -11.2926>, rot=<0.0199 -0.9513 0.3062 -0.03>}
 jacket_front01_l {parent=jacket_front_base_l, loc=<0 0 0>, rot=<0 0 0 1>}
 jacket_front02_l {parent=jacket_front01_l, loc=<6.5872 0 0>, rot=<0 0 0 1>}
 jacket_front03_l {parent=jacket_front02_l, loc=<5.9778 0 0>, rot=<-0.0601 0 0 -0.9982>}
 jacket_front04_l {parent=jacket_front03_l, loc=<6.0702 0 0>, rot=<0 0 0 1>}
 jacket_front05_l {parent=jacket_front04_l, loc=<8.0204 0 0>, rot=<0 0 0 1>}
 jacket_front_base_r {parent=jacket_spine_corr_base, loc=<-3.428 12.2285 11.2926>, rot=<0.9513 0.0199 -0.03 -0.3062>}
 jacket_front01_r {parent=jacket_front_base_r, loc=<0 0 0>, rot=<0 0 0 1>}
 jacket_front02_r {parent=jacket_front01_r, loc=<-6.5869 0 0>, rot=<0 0 0 1>}
 jacket_front03_r {parent=jacket_front02_r, loc=<-5.9785 0.0001 0>, rot=<-0.0601 0 0 -0.9982>}
 jacket_front04_r {parent=jacket_front03_r, loc=<-6.07 -0.0001 0>, rot=<0 0 0 1>}
 jacket_front05_r {parent=jacket_front04_r, loc=<-8.0203 0 0>, rot=<0 0 0 -1>}
 jacket_side_base_l {parent=jacket_spine_corr_base, loc=<-7.0191 4.2319 -17.0304>, rot=<0.0515 -0.997 0.022 -0.053>}
 jacket_side01_l {parent=jacket_side_base_l, loc=<0 0 0>, rot=<0 0 0 1>}
 jacket_side02_l {parent=jacket_side01_l, loc=<10.3285 0 0>, rot=<0 0 0 1>}
 jacket_side03_l {parent=jacket_side02_l, loc=<11.2734 0 0>, rot=<0 0 0 1>}
 jacket_side_base_r {parent=jacket_spine_corr_base, loc=<-7.0194 4.2319 17.0304>, rot=<0.997 0.0515 -0.053 -0.022>}
 jacket_side01_r {parent=jacket_side_base_r, loc=<0 0 0>, rot=<0 0 0 1>}
 jacket_side02_r {parent=jacket_side01_r, loc=<-10.3284 0 0>, rot=<0 0 0 -1>}
 jacket_side03_r {parent=jacket_side02_r, loc=<-11.2731 0 0>, rot=<0 0 0 -1>}
 jacket_back_base_r {parent=jacket_spine_corr_base, loc=<-6.0794 -10.0285 10.057>, rot=<0.1434 -0.9812 0.128 0.0182>}
 jacket_back01_r {parent=jacket_back_base_r, loc=<0 0 0>, rot=<0 0 0 1>}
 jacket_back02_r {parent=jacket_back01_r, loc=<10.8891 0 0>, rot=<0.0012 -0.0332 -0.0739 -0.9967>}
 jacket_back03_r {parent=jacket_back02_r, loc=<8.1423 0 0>, rot=<0.1305 0 0 -0.9914>}
 jacket_back_base_m {parent=jacket_spine_corr_base, loc=<-6.1544 -11.198 -0.1376>, rot=<0.1531 -0.9882 -0.0003 0.0063>}
 jacket_back01_m {parent=jacket_back_base_m, loc=<0 0 0>, rot=<0 0 0 1>}
 jacket_back02_m {parent=jacket_back01_m, loc=<11.2049 0 0>, rot=<-0.0012 0.0077 -0.0467 -0.9989>}
 jacket_back03_m {parent=jacket_back02_m, loc=<7.9923 0 0>, rot=<0 0 0 1>}
 jacket_back_base_l {parent=jacket_spine_corr_base, loc=<-6.4534 -10.1186 -11.4749>, rot=<0.0918 -0.9721 -0.216 -0.0037>}
 jacket_back01_l {parent=jacket_back_base_l, loc=<0 0 0>, rot=<0 0 0 1>}
 jacket_back02_l {parent=jacket_back01_l, loc=<10.235 0 0>, rot=<-0.0004 0.0124 -0.0098 -0.9999>}
 jacket_back03_l {parent=jacket_back02_l, loc=<8.4368 0 0>, rot=<0 0 0 1>}
 dyn_JacketCloseVertex {parent=spine_02, loc=<0 0 0>, rot=<0 0 0 1>}
 thigh_l {parent=pelvis, loc=<-1.0855 -0.2109 -8.7053>, rot=<0.0009 0.0355 0.0155 -0.9993>}
 calf_l {parent=thigh_l, loc=<-42.5284 -0.0019 -0.7297>, rot=<0.0006 -0.0161 -0.0391 -0.9991>}
 foot_l {parent=calf_l, loc=<-46.7401 -2.5014 -0.592>, rot=<0.0427 -0.0238 0.0353 -0.9982>}
 ball_l {parent=foot_l, loc=<-3.5658 10.8066 -0.1544>, rot=<0.0001 -0.0001 -0.7063 -0.7079>}
 calf_twist_02_l {parent=calf_l, loc=<-19.1546 -1.7943 -0.2638>, rot=<0.0237 0.0095 -0.0035 -0.9997>}
 calf_twist_01_l {parent=calf_l, loc=<-35.8483 -1.9197 -0.5794>, rot=<0.0237 0.0095 -0.0035 -0.9997>}
 deform_knee_in_base_l {parent=calf_l, loc=<0.8031 -5.0231 0.2456>, rot=<0 0 0 -1>}
 deform_knee_in_l {parent=deform_knee_in_base_l, loc=<0 0 0>, rot=<0 0 0 1>}
 deform_knee_out_base_l {parent=calf_l, loc=<0 4 0>, rot=<0 0 0 -1>}
 deform_knee_out_l {parent=deform_knee_out_base_l, loc=<0 0 0>, rot=<0 0 0 1>}
 thigh_twist_01_l {parent=thigh_l, loc=<-27.952 0 0>, rot=<0.0013 0 0 -1>}
 deform_groin_base_l {parent=thigh_l, loc=<0.4801 5.6776 -0.2486>, rot=<0 0 0 -1>}
 deform_groin_l {parent=deform_groin_base_l, loc=<0 0 0>, rot=<0 0 0 1>}
 thigh_r {parent=pelvis, loc=<-1.0853 -0.2109 8.7053>, rot=<0.0355 -0.0009 0.9993 0.0155>}
 calf_r {parent=thigh_r, loc=<42.5284 0.0019 0.7297>, rot=<0.0006 -0.0161 -0.0391 -0.9991>}
 foot_r {parent=calf_r, loc=<46.7401 2.5014 0.592>, rot=<0.0427 -0.0238 0.0353 -0.9982>}
 ball_r {parent=foot_r, loc=<3.5658 -10.8066 0.1544>, rot=<0.0001 -0.0001 -0.7063 -0.7079>}
 calf_twist_02_r {parent=calf_r, loc=<19.1546 1.7943 0.2638>, rot=<0.0237 0.0095 -0.0035 -0.9997>}
 calf_twist_01_r {parent=calf_r, loc=<35.8483 1.9197 0.5794>, rot=<0.0237 0.0095 -0.0035 -0.9997>}
 deform_knee_in_base_r {parent=calf_r, loc=<-0.8031 5.0231 -0.2457>, rot=<0 0 0 -1>}
 deform_knee_in_r {parent=deform_knee_in_base_r, loc=<0 0 0>, rot=<0 0 0 1>}
 deform_knee_out_base_r {parent=calf_r, loc=<0 -4 0>, rot=<0 0 0 -1>}
 deform_knee_out_r {parent=deform_knee_out_base_r, loc=<0 0 0>, rot=<0 0 0 1>}
 thigh_twist_01_r {parent=thigh_r, loc=<27.952 0 0>, rot=<0.0013 0 0 -1>}
 deform_groin_base_r {parent=thigh_r, loc=<-0.4801 -5.6776 0.2486>, rot=<0 0 0 -1>}
 deform_groin_r {parent=deform_groin_base_r, loc=<0 0 0>, rot=<0 0 0 1>}
 belt_base_main {parent=pelvis, loc=<2.2905 12.3121 0.6661>, rot=<-0.0127 -0.707 0.0127 -0.707>}
 belt_handle_base_r {parent=belt_base_main, loc=<-1.2142 0.2718 -1.8516>, rot=<-0.0805 -0.8312 -0.0531 -0.5476>}
 belt_handle01_r {parent=belt_handle_base_r, loc=<0 0 0>, rot=<0 0 0 1>}
 belt_handle02_r {parent=belt_handle01_r, loc=<2.8338 0 0>, rot=<0 0 0 1>}
 belt_handle_base_l {parent=belt_base_main, loc=<1.4967 0.565 -1.7352>, rot=<-0.0499 -0.3882 -0.1173 -0.9127>}
 belt_handle01_l {parent=belt_handle_base_l, loc=<0 0 0>, rot=<0 0 0 1>}
 belt_handle02_l {parent=belt_handle01_l, loc=<2.5718 0 0>, rot=<-0.0381 0.116 0.202 -0.9718>}
 belt_handle03_l {parent=belt_handle02_l, loc=<1.614 1.3051 -1.1527>, rot=<0 0 0 1>}
 deform_glute_base_l {parent=pelvis, loc=<-7.3466 -9.0213 -9.1245>, rot=<0.0172 0.031 -0.4477 -0.8935>}
 deform_glute_l {parent=deform_glute_base_l, loc=<0 0 0>, rot=<0 0 0 1>}
 deform_glute_base_r {parent=pelvis, loc=<-7.3464 -9.0213 9.1245>, rot=<0.031 -0.0172 0.8935 -0.4477>}
 deform_glute_r {parent=deform_glute_base_r, loc=<0 0 0>, rot=<0 0 0 1>}
 beltShield_base {parent=pelvis, loc=<0.1872 9.902 0>, rot=<-0.0127 -0.707 0.0127 -0.707>}
 beltShield_01 {parent=beltShield_base, loc=<0 0 0>, rot=<0 0 0 1>}
 beltShiledSide_l {parent=beltShield_01, loc=<5.7837 0.3773 4.0075>, rot=<0.0019 0.7317 -0.0015 -0.6816>}
 beltShiledSide_r {parent=beltShield_01, loc=<-5.1745 0.3729 4.0413>, rot=<-0.6816 0.0015 0.7317 -0.0019>}
 beltShield_corrUpCorner_r {parent=beltShield_01, loc=<-8.5256 -0.6398 7.3685>, rot=<-0.0397 -0.0333 -0.8186 -0.5721>}
 beltShield_corrUpCorner_l {parent=beltShield_01, loc=<8.5256 -0.6398 7.3686>, rot=<-0.0397 0.0333 0.8186 -0.5721>}
 Skirt_corrective {parent=pelvis, loc=<-9.8962 -7.8098 0.2525>, rot=<-0.1425 -0.9898 0 0>}
 belt_side_base_l {parent=pelvis, loc=<4.6096 5.3247 -14.0181>, rot=<-0.2453 -0.6484 -0.2115 -0.689>}
 belt_side_l {parent=belt_side_base_l, loc=<0 0 0>, rot=<0 -1 0 0>}
 belt_side_base_r {parent=pelvis, loc=<4.61 5.3247 14.0181>, rot=<0.689 0.2115 0.6484 0.2453>}
 belt_side_r {parent=belt_side_base_r, loc=<0 0 0>, rot=<0 0 0 1>}
 beltCorr_bk_l {parent=pelvis, loc=<5.3553 -2.3648 -14.1456>, rot=<0.1384 -0.6904 0.096 -0.7035>}
 beltCorr_bk_r {parent=pelvis, loc=<5.3556 -2.3648 14.1456>, rot=<0.7035 -0.096 0.6904 -0.1384>}
]
variables [
 Custom Idle is Playing Float {type=double, default=0}
 IsCrouching {type=bool, default=False}
 IsSkydive {type=bool, default=False}
 isDBNO {type=bool, default=False}
 isMelee {type=bool, default=False}
 isSprinting {type=bool, default=False}
]
