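# Physics asset for the F_MED_PolishedJade_Mind female player body, set up for secondary motion of its long jacket.
# Capsule chains form front, side and back panels from small base spheres, joined by angular-limited constraints with many locked axes.
# Cross constraints tie the bottom front bones to the sides and outer back bones to the middle.
# Pelvis, spine, belt and multi-capsule thighs provide body collision, while a huge root sphere below acts as a floor.
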PhysicsAsset /BRCosmetics/Characters/Player/Female/Medium/Bodies/F_MED_PolishedJade_Mind/Meshes/F_MED_PolishedJade_Mind_Physics
bodies {
 belt_base_main [{
   shapeType=Capsule
   length=6.1679
   radius=3.6325
   shapeName=belt_base_main_capsule
   center=<0.5405 1.5818 -5.5629>
   rotation=<rot 0.956 100.6717 -95.2959>
  }]
 jacket_back01_l [{
   shapeType=Capsule
   length=9.2409
   radius=4.8275
   shapeName=jacket_side01_l_capsule
   center=<4.8085 -1.7316 1.0442>
   rotation=<rot 0.0002 -1.3219 -167.3929>
  }]
 jacket_back01_m [{shapeType=Capsule, length=9.4685, radius=4.5998, shapeName=jacket_side01_l_capsule, center=<5.9351 -2.0977 0>}]
 jacket_back01_r [{
   shapeType=Capsule
   length=9.9208
   radius=4.1476
   shapeName=jacket_side01_l_capsule
   center=<4.7886 -0.6151 -0.6836>
   rotation=<rot 0.0001 1.322 -20.6232>
  }]
 jacket_back02_l [{
   shapeType=Capsule
   length=11.333
   radius=3.8648
   shapeName=jacket_side02_l_capsule
   center=<4.344 -1.0825 3.0939>
   rotation=<rot 0.0001 0.0002 14.3758>
  }]
 jacket_back02_m [{shapeType=Capsule, length=12.8255, radius=3.9216, shapeName=jacket_side02_l_capsule, center=<3.8809 -2.4448 0>}]
 jacket_back02_r [{
   shapeType=Capsule
   length=11.333
   radius=3.8648
   shapeName=jacket_side02_l_capsule
   center=<3.6695 -1.3035 -3.1441>
   rotation=<rot 3.0471 -1.1096 -20.1007>
  }]
 jacket_back03_l [{
   shapeType=Capsule
   length=16.2213
   radius=4.7356
   shapeName=jacket_side03_l_capsule
   center=<3.322 -1.4472 3.0646>
   rotation=<rot 0.0002 0.0004 15.8138>
  }]
 jacket_back03_m [{shapeType=Capsule, length=16.9134, radius=4.7, shapeName=jacket_side03_l_capsule, center=<5 1.822 -0.0118>, rotation=<rot 0 0.0002 0>}]
 jacket_back03_r [{
   shapeType=Capsule
   length=9.8505
   radius=4.2967
   shapeName=jacket_side03_l_capsule
   center=<3.5755 0.1956 -3.1143>
   rotation=<rot 2.3487 -1.315 -31.8212>
  }]
 jacket_back_base_l [{shapeType=Sphere, radius=1.2424, shapeName=jacket_back_base_l_sphere}]
 jacket_back_base_m [{shapeType=Sphere, radius=1.3433, shapeName=jacket_back_base_m_sphere}]
 jacket_back_base_r [{shapeType=Sphere, radius=1.2424, shapeName=jacket_back_base_l_sphere}]
 jacket_front01_l [{shapeType=Capsule, length=7.6299, radius=3.7183, shapeName=jacket_front01_l_capsule, center=<3.2575 1.252 0>, rotation=<rot 0 0 -4.0043>}]
 jacket_front01_r [{
   shapeType=Capsule
   length=7.9639
   radius=3.3844
   shapeName=jacket_front01_l_capsule
   center=<-3.2576 -0.8459 -0.0342>
   rotation=<rot -0.0054 0.0075 -175.1918>
  }]
 jacket_front02_l [{
   shapeType=Capsule
   length=9.6022
   radius=2.9594
   shapeName=jacket_front02_l_capsule
   center=<3.4154 0.5422 -0.3645>
   rotation=<rot -4.6813 -91.4042 2.7956>
  }]
 jacket_front02_r [{
   shapeType=Capsule
   length=9.4818
   radius=3.0798
   shapeName=jacket_front02_l_capsule
   center=<-3.4154 -0.5422 0.3645>
   rotation=<rot 2.9125 -91.0339 -177.21>
  }]
 jacket_front03_l [{
   shapeType=Capsule
   length=10.4005
   radius=2.9127
   shapeName=jacket_front03_l_capsule
   center=<3.274 0.9524 -0.671>
   rotation=<rot 10.8756 -89.0958 1.6062>
  }]
 jacket_front03_r [{
   shapeType=Capsule
   length=10.0503
   radius=3.2629
   shapeName=jacket_front03_l_capsule
   center=<-3.274 -0.9524 0.671>
   rotation=<rot 5.0588 -89.2699 -178.4256>
  }]
 jacket_front04_l [{
   shapeType=Capsule
   length=12.4915
   radius=3.6262
   shapeName=jacket_front04_l_capsule
   center=<3.6478 2.0107 -0.4649>
   rotation=<rot 12.3979 -90.0911 0.8223>
  }]
 jacket_front04_r [{
   shapeType=Capsule
   length=12.6627
   radius=3.455
   shapeName=jacket_front04_l_capsule
   center=<-3.6478 -2.0107 0.4649>
   rotation=<rot 10.6651 -90.1163 -179.1827>
  }]
 jacket_front05_l [{
   shapeType=Capsule
   length=16.9056
   radius=3.2443
   shapeName=jacket_front05_l_capsule
   center=<2.1681 1.4377 0.0353>
   rotation=<rot 12.8033 -90.1252 -0.5698>
  }]
 jacket_front05_r [{
   shapeType=Capsule
   length=12.37
   radius=3.5265
   shapeName=jacket_front05_l_capsule
   center=<-2.1689 -1.7904 0.0516>
   rotation=<rot 17.7831 -90.1771 179.4165>
  }]
 jacket_front_base_l [{shapeType=Sphere, radius=1.7704, shapeName=jacket_front_base_l_sphere}]
 jacket_front_base_r [{shapeType=Sphere, radius=1.7704, shapeName=jacket_front_base_l_sphere}]
 jacket_side01_l [{
   shapeType=Capsule
   length=8.8689
   radius=4.1476
   shapeName=jacket_side01_l_capsule
   center=<5.5718 -0.8018 0.2618>
   rotation=<rot 0 -1.3223 -85.9665>
  }]
 jacket_side01_r [{
   shapeType=Capsule
   length=8.8689
   radius=4.1476
   shapeName=jacket_side01_l_capsule
   center=<-5.5718 0.8018 -0.2618>
   rotation=<rot 0 -1.3223 90>
  }]
 jacket_side02_l [{
   shapeType=Capsule
   length=11.333
   radius=3.8648
   shapeName=jacket_side02_l_capsule
   center=<4.344 -1.3824 3.1224>
   rotation=<rot 0.1493 -0.9643 -88.8472>
  }]
 jacket_side02_r [{
   shapeType=Capsule
   length=11.333
   radius=3.8648
   shapeName=jacket_side02_l_capsule
   center=<-4.344 1.3824 -3.1224>
   rotation=<rot 0.1493 0.9643 90>
  }]
 jacket_side03_l [{
   shapeType=Capsule
   length=17.9049
   radius=4.5
   shapeName=jacket_side03_l_capsule
   center=<2.0498 -1.8562 3.1345>
   rotation=<rot 0 179.0684 -91.9584>
  }]
 jacket_side03_r [{
   shapeType=Capsule
   length=12.3351
   radius=4.5
   shapeName=jacket_side03_l_capsule
   center=<-2.0498 1.8562 -3.1345>
   rotation=<rot 0.0005 179.0778 81.6373>
  }]
 jacket_side_base_l [{shapeType=Sphere, radius=1.7704, shapeName=jacket_front_base_l_sphere}]
 jacket_side_base_r [{shapeType=Sphere, radius=1.7704, shapeName=jacket_front_base_l_sphere}]
 pelvis [{shapeType=Capsule, length=9, radius=12.7, shapeName=pelvis_capsule, center=<-2.4391 -0.5027 0>}]
 root [{shapeType=Sphere, radius=227.1117, shapeName=root_sphere, center=<0 0 -225.7164>}]
 spine_01 [{shapeType=Capsule, length=4.3113, radius=11.3144, shapeName=spine_01_capsule, center=<0 1.2392 0>, rotation=<rot 0 -90 0>}]
 spine_02 [{shapeType=Capsule, length=4.4691, radius=10.2934, shapeName=spine_02_capsule, center=<0 2.6529 0>}]
 thigh_l [
  {
   shapeType=Capsule
   length=18.2193
   radius=8.9447
   shapeName=thigh_l_capsule
   center=<-15.5207 1.9707 -0.1203>
   rotation=<rot -89.7816 -179.9958 -176.0531>
  }
  {
   shapeType=Capsule
   length=20.3566
   radius=7.6972
   shapeName=thigh_l_capsule
   center=<-15.6795 -1.7708 0.0341>
   rotation=<rot -89.8755 -179.9951 168.5971>
  }
  {
   shapeType=Capsule
   length=14.0771
   radius=6.2727
   shapeName=thigh_l_capsule
   center=<-33.8819 0.7961 0.127>
   rotation=<rot -89.7852 -179.9976 -171.3657>
  }
 ]
 thigh_r [
  {
   shapeType=Capsule
   length=18.4584
   radius=8.7055
   shapeName=thigh_l_capsule
   center=<15.5359 -2.1803 0.3069>
   rotation=<rot -89.7816 -179.9958 3.9469>
  }
  {
   shapeType=Capsule
   length=20.1363
   radius=7.9175
   shapeName=thigh_l_capsule
   center=<15.6795 1.7708 -0.0341>
   rotation=<rot -89.8755 -179.9951 -11.4029>
  }
  {
   shapeType=Capsule
   length=14.0771
   radius=6.2727
   shapeName=thigh_l_capsule
   center=<33.8819 -0.7961 -0.127>
   rotation=<rot -89.7852 -179.9976 8.6343>
  }
 ]
}
constraints [
 {
  bone1Name=jacket_front01_l
  bone2Name=jacket_front_base_l
  swing1LimitDegrees=30
  swing2LimitDegrees=45
  swing2Motion=Locked
  twistLimitDegrees=45
  twistMotion=Locked
 }
 {
  bone1Name=jacket_front02_l
  bone2Name=jacket_front01_l
  swing1LimitDegrees=60
  swing2LimitDegrees=45
  swing2Motion=Locked
  twistLimitDegrees=20
 }
 {
  bone1Name=jacket_front03_l
  bone2Name=jacket_front02_l
  swing1LimitDegrees=65
  swing2LimitDegrees=45
  swing2Motion=Locked
  twistLimitDegrees=30
 }
 {
  bone1Name=jacket_front04_l
  bone2Name=jacket_front03_l
  swing1LimitDegrees=75
  swing2LimitDegrees=45
  swing2Motion=Locked
  twistLimitDegrees=25
 }
 {
  bone1Name=jacket_front05_l
  bone2Name=jacket_front04_l
  swing1LimitDegrees=30
  swing2LimitDegrees=45
  swing2Motion=Locked
  twistLimitDegrees=30
  twistMotion=Locked
 }
 {bone1Name=jacket_side02_l, bone2Name=jacket_side01_l, swing1LimitDegrees=10, swing2LimitDegrees=45, twistLimitDegrees=45, twistMotion=Locked}
 {
  bone1Name=jacket_side01_l
  bone2Name=jacket_side_base_l
  swing1LimitDegrees=20
  swing2LimitDegrees=25
  twistLimitDegrees=45
  twistMotion=Locked
 }
 {bone1Name=jacket_side03_l, bone2Name=jacket_side02_l, swing1LimitDegrees=15, swing2LimitDegrees=15, twistLimitDegrees=45, twistMotion=Locked}
 {
  bone1Name=jacket_front05_l
  bone2Name=jacket_side03_l
  swing1LimitDegrees=45
  swing1Motion=Locked
  swing2LimitDegrees=45
  swing2Motion=Locked
  twistLimitDegrees=45
 }
 {bone1Name=jacket_back03_l, bone2Name=jacket_back02_l, swing1LimitDegrees=30, swing2LimitDegrees=5, twistLimitDegrees=8}
 {
  bone1Name=jacket_back02_l
  bone2Name=jacket_back01_l
  swing1LimitDegrees=15
  swing2LimitDegrees=45
  swing2Motion=Locked
  twistLimitDegrees=45
  twistMotion=Locked
 }
 {
  bone1Name=jacket_back01_l
  bone2Name=jacket_back_base_l
  swing1LimitDegrees=15
  swing2LimitDegrees=30
  twistLimitDegrees=45
  twistMotion=Locked
 }
 {
  bone1Name=jacket_back01_m
  bone2Name=jacket_back_base_m
  swing1LimitDegrees=15
  swing2LimitDegrees=45
  swing2Motion=Locked
  twistLimitDegrees=45
  twistMotion=Locked
 }
 {
  bone1Name=jacket_back02_m
  bone2Name=jacket_back01_m
  swing1LimitDegrees=15
  swing2LimitDegrees=45
  swing2Motion=Locked
  twistLimitDegrees=45
  twistMotion=Locked
 }
 {
  bone1Name=jacket_back03_m
  bone2Name=jacket_back02_m
  swing1LimitDegrees=30
  swing2LimitDegrees=45
  swing2Motion=Locked
  twistLimitDegrees=45
  twistMotion=Locked
 }
 {
  bone1Name=jacket_back03_l
  bone2Name=jacket_back03_m
  swing1LimitDegrees=45
  swing1Motion=Locked
  swing2LimitDegrees=45
  swing2Motion=Locked
  twistLimitDegrees=45
  twistMotion=Locked
 }
 {
  bone1Name=jacket_back01_r
  bone2Name=jacket_back_base_r
  swing1LimitDegrees=15
  swing2LimitDegrees=30
  twistLimitDegrees=45
  twistMotion=Locked
 }
 {
  bone1Name=jacket_back02_r
  bone2Name=jacket_back01_r
  swing1LimitDegrees=15
  swing2LimitDegrees=45
  swing2Motion=Locked
  twistLimitDegrees=45
  twistMotion=Locked
 }
 {bone1Name=jacket_back03_r, bone2Name=jacket_back02_r, swing1LimitDegrees=30, swing2LimitDegrees=5, twistLimitDegrees=8}
 {
  bone1Name=jacket_front01_r
  bone2Name=jacket_front_base_r
  swing1LimitDegrees=20
  swing2LimitDegrees=20
  swing2Motion=Locked
  twistLimitDegrees=10
 }
 {
  bone1Name=jacket_front02_r
  bone2Name=jacket_front01_r
  swing1LimitDegrees=45
  swing2LimitDegrees=45
  swing2Motion=Locked
  twistLimitDegrees=20
  twistMotion=Locked
 }
 {
  bone1Name=jacket_front03_r
  bone2Name=jacket_front02_r
  swing1LimitDegrees=30
  swing2LimitDegrees=45
  swing2Motion=Locked
  twistLimitDegrees=30
 }
 {
  bone1Name=jacket_front04_r
  bone2Name=jacket_front03_r
  swing1LimitDegrees=55
  swing2LimitDegrees=45
  swing2Motion=Locked
  twistLimitDegrees=25
 }
 {
  bone1Name=jacket_front05_r
  bone2Name=jacket_front04_r
  swing1LimitDegrees=60
  swing2LimitDegrees=45
  swing2Motion=Locked
  twistLimitDegrees=30
  twistMotion=Locked
 }
 {
  bone1Name=jacket_side01_r
  bone2Name=jacket_side_base_r
  swing1LimitDegrees=20
  swing2LimitDegrees=25
  twistLimitDegrees=45
  twistMotion=Locked
 }
 {bone1Name=jacket_side02_r, bone2Name=jacket_side01_r, swing1LimitDegrees=10, swing2LimitDegrees=45, twistLimitDegrees=45, twistMotion=Locked}
 {bone1Name=jacket_side03_r, bone2Name=jacket_side02_r, swing1LimitDegrees=15, swing2LimitDegrees=15, twistLimitDegrees=15, twistMotion=Locked}
 {
  bone1Name=jacket_front05_r
  bone2Name=jacket_side03_r
  swing1LimitDegrees=45
  swing1Motion=Locked
  swing2LimitDegrees=45
  swing2Motion=Locked
  twistLimitDegrees=45
 }
 {
  bone1Name=jacket_back03_r
  bone2Name=jacket_back03_m
  swing1LimitDegrees=45
  swing1Motion=Locked
  swing2LimitDegrees=45
  swing2Motion=Locked
  twistLimitDegrees=45
  twistMotion=Locked
 }
]
