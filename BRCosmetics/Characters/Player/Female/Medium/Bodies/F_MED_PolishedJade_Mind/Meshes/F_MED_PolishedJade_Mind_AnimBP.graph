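# Post-process style Anim Blueprint for the Polished Jade Mind female character body cosmetic, adding secondary motion and pose fixes over the parent mesh's pose.
# A state-fed Control Rig precedes chained default, DBNO and crouch RigidBody simulations, additive bone corrections and belt-handle AnimDynamics.
# Each update computes gravity/wind override through a deprecated helper and refreshes movement flags.
# In the front end, idle curves set a bypass value and broadcast glow through Trigger FX.
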AnimBlueprint /BRCosmetics/Characters/Player/Female/Medium/Bodies/F_MED_PolishedJade_Mind/Meshes/F_MED_PolishedJade_Mind_AnimBP
target_skeleton=&/F_MED_PolishedJade_Mind_Skeleton
graphs [
 AnimGraph {nodes [
   n1 = Control Rig!(
    Source=Copy Pose From Mesh() {bUseAttachedParent=true}
    Alpha
    IsCrouching:bool=Get bIsCrouching()
    IsSkydive:bool=Get SkydiveAndParachute() -> (SkydiveAndParachute:bool)
    isDBNO:bool=Get bIsDBNO()
    isMelee:bool=Get bIsPlayingMeleeAnim()
    isSprinting:bool=Get bIsSprinting()
    Custom Idle is Playing Float:double=Get Custom Idle is Playing Float() -> (Custom Idle is Playing Float:float)
   ) -> (Pose) {
    alphaBoolBlend {blendOption=Linear}
    controlRigAssetReference {blueprintRigClass=&/F_MED_PolishedJade_Mind_CtrlRig.F_MED_PolishedJade_Mind_CtrlRig_C}
   }
   n2 = RigidBody!(ComponentPose=Local To Component(LocalPose=@n1), bAlphaBoolEnabled:bool, OverrideWorldGravity:Vector="0,0,0", ExternalForce) -> (Pose) {
    alphaInputType=Bool
    bClampLinearTranslationLimitToRefPose=true
    bOverrideWorldGravity=true
    overridePhysicsAsset=&/F_MED_PolishedJade_Mind_Physics
    simulationSpace=BaseBoneSpace
    alphaBoolBlend {blendInTime=0.1, blendOption=Cubic, blendOutTime=0.1}
    baseBoneRef {boneName=pelvis}
    componentAppliedLinearAccClamp=<3000 1000 1000>
    componentLinearAccScale=<0.35 0.35 0.35>
    componentLinearVelScale=<0.35 0.35 0.35>
    simSpaceSettings {maxAngularAcceleration=100, maxAngularVelocity=2, maxLinearAcceleration=500, maxLinearVelocity=100, velocityScaleZ=0.5, worldAlpha=0.5}
   }
   n3 = RigidBody(ComponentPose=@n2, OverrideWorldGravity="0,0,0") {
    alphaInputType=Bool
    bClampLinearTranslationLimitToRefPose=true
    bOverrideWorldGravity=true
    overridePhysicsAsset=&/F_MED_PolishedJade_MindDBNO_Physics
    simulationSpace=BaseBoneSpace
    alphaBoolBlend {blendInTime=0.1, blendOption=Cubic, blendOutTime=0.1}
    baseBoneRef {boneName=spine_02}
    componentAppliedLinearAccClamp=<3000 1000 1000>
    componentLinearAccScale {x=0.15}
    componentLinearVelScale {x=0.2}
    simSpaceSettings {maxAngularAcceleration=100, maxAngularVelocity=2, maxLinearAcceleration=500, maxLinearVelocity=100, velocityScaleZ=0, worldAlpha=1}
   }
   n4 = RigidBody(ComponentPose=@n3, OverrideWorldGravity="0,0,0") {
    alphaInputType=Bool
    bClampLinearTranslationLimitToRefPose=true
    bOverrideWorldGravity=true
    overridePhysicsAsset=&/F_MED_PolishedJade_Mind_Crouch_Physics
    simulationSpace=BaseBoneSpace
    alphaBoolBlend {blendInTime=0.1, blendOption=Cubic, blendOutTime=0.1}
    baseBoneRef {boneName=spine_02}
    componentAppliedLinearAccClamp=<3000 1000 1000>
    componentLinearAccScale {x=0.15}
    componentLinearVelScale {x=0.2}
    simSpaceSettings {maxAngularAcceleration=100, maxAngularVelocity=2, maxLinearAcceleration=500, maxLinearVelocity=100, velocityScaleZ=0.5, worldAlpha=1}
   }
   n5 = "Transform (Modify) Bone - Bone: jacket_front_base_r"!(ComponentPose=@n4, Alpha=1, Translation="0,-2,2", Rotation="-15,0,-10") -> (Pose) {
    rotationMode=BMM_Additive
    translationMode=BMM_Additive
    alphaBoolBlend {blendOption=Linear}
    boneToModify {boneName=jacket_front_base_r}
    rotation {pitch=-15, roll=-10}
    translation {y=-2, z=2}
   }
   n6 = "Transform (Modify) Bone - Bone: jacket_side_base_r"!(ComponentPose=@n5, Alpha=1, Translation="0,-1,1", Rotation="-15,0,15") -> (Pose) {
    rotationMode=BMM_Additive
    translationMode=BMM_Additive
    alphaBoolBlend {blendOption=Linear}
    boneToModify {boneName=jacket_side_base_r}
    rotation {pitch=-15, roll=15}
    translation {y=-1, z=1}
   }
   n7 = Anim Dynamics - Bone: belt_handle01_r!(ComponentPose=@n6, Alpha=1, GravityOverride:Vector="0,0,0", ExternalForce="0,0,0") -> (Pose) {
    angularDampingOverride=0.9
    bOverrideAngularDamping=true
    bUseGravityOverride=true
    simulationSpace=BoneRelative
    alphaBoolBlend {blendOption=Linear}
    boundBone {boneName=belt_handle01_r}
    componentLinearAccScale=<3 3 3>
    componentLinearVelScale=<3 3 3>
    physicsBodyDefinitions [{
      boundBone {boneName=belt_handle01_r}
      boxExtents=<10 20 20>
      constraintSetup {angularLimitsMax=<0 5 10>, angularLimitsMin=<0 -2 -3>, angularTarget=<1 0 0>}
      localJointOffset=<1 0 0>
     }]
    relativeSpaceBone {boneName=root}
   }
   n8 = Anim Dynamics - Bone: belt_handle02_r!(ComponentPose=@n7, Alpha=1, GravityOverride:Vector="0,0,0", ExternalForce="0,0,0") -> (Pose) {
    angularDampingOverride=0.9
    bOverrideAngularDamping=true
    bUseGravityOverride=true
    simulationSpace=BoneRelative
    alphaBoolBlend {blendOption=Linear}
    boundBone {boneName=belt_handle02_r}
    componentLinearAccScale=<3 3 3>
    componentLinearVelScale=<3 3 3>
    physicsBodyDefinitions [{
      boundBone {boneName=belt_handle02_r}
      boxExtents=<10 20 20>
      constraintSetup {angularLimitsMax=<0 3 5>, angularLimitsMin=<0 -1 -1>, angularTarget=<1 0 0>}
      localJointOffset=<1 0 0>
     }]
    relativeSpaceBone {boneName=root}
   }
   n9 = Anim Dynamics - Bone: belt_handle01_l!(ComponentPose=@n8, Alpha=1, GravityOverride:Vector="0,0,0", ExternalForce="0,0,0") -> (Pose) {
    angularDampingOverride=0.9
    bOverrideAngularDamping=true
    bUseGravityOverride=true
    simulationSpace=BoneRelative
    alphaBoolBlend {blendOption=Linear}
    boundBone {boneName=belt_handle01_l}
    componentLinearAccScale=<3 3 3>
    componentLinearVelScale=<3 3 3>
    physicsBodyDefinitions [{
      boundBone {boneName=belt_handle01_l}
      boxExtents=<10 20 20>
      constraintSetup {angularLimitsMax=<0 5 10>, angularLimitsMin=<0 -2 -3>, angularTarget=<1 0 0>}
      localJointOffset=<1 0 0>
     }]
    relativeSpaceBone {boneName=root}
   }
   n10 = Anim Dynamics - Bone: belt_handle02_l!(ComponentPose=@n9, Alpha=1, GravityOverride:Vector="0,0,0", ExternalForce="0,0,0") -> (Pose) {
    angularDampingOverride=0.9
    bOverrideAngularDamping=true
    bUseGravityOverride=true
    simulationSpace=BoneRelative
    alphaBoolBlend {blendOption=Linear}
    boundBone {boneName=belt_handle02_l}
    componentLinearAccScale=<3 3 3>
    componentLinearVelScale=<3 3 3>
    physicsBodyDefinitions [{
      boundBone {boneName=belt_handle02_l}
      boxExtents=<10 20 20>
      constraintSetup {angularLimitsMax=<0 3 5>, angularLimitsMin=<0 -1 -1>, angularTarget=<1 0 0>}
      localJointOffset=<1 0 0>
     }]
    relativeSpaceBone {boneName=root}
   }
   n11 = "Transform (Modify) Bone - Bone: jacket_back_base_l"!(ComponentPose=@n10, Alpha=1, Translation="2,0,0", Rotation="10,0,-4") -> (Pose) {
    rotationMode=BMM_Additive
    translationMode=BMM_Additive
    alphaBoolBlend {blendOption=Linear}
    boneToModify {boneName=jacket_back_base_l}
    rotation {pitch=10, roll=-4}
    translation {x=2}
   }
   n12 = "Transform (Modify) Bone - Bone: jacket_back_base_r"!(ComponentPose=@n11, Alpha=1, Translation="2,0,0", Rotation="10,0,4") -> (Pose) {
    rotationMode=BMM_Additive
    translationMode=BMM_Additive
    alphaBoolBlend {blendOption=Linear}
    boneToModify {boneName=jacket_back_base_r}
    rotation {pitch=10, roll=4}
    translation {x=2}
   }
   n13 = "Transform (Modify) Bone - Bone: jacket_back_base_m"!(ComponentPose=@n12, Alpha=1, Translation="2,0,0", Rotation="10,0,0") -> (Pose) {
    rotationMode=BMM_Additive
    translationMode=BMM_Additive
    alphaBoolBlend {blendOption=Linear}
    boneToModify {boneName=jacket_back_base_m}
    rotation {pitch=10}
    translation {x=2}
   }
   n14 = "Transform (Modify) Bone - Bone: jacket_side_base_l"!(ComponentPose=@n13, Alpha=1, Translation="2,0,0", Rotation="0,0,0") -> (Pose) {
    rotationMode=BMM_Additive
    translationMode=BMM_Additive
    alphaBoolBlend {blendOption=Linear}
    boneToModify {boneName=jacket_side_base_l}
    translation {x=2}
   }
   n15 = "Transform (Modify) Bone - Bone: jacket_side_base_r"(ComponentPose=@n14, Alpha=1, Translation="-2,0,0", Rotation="0,0,0") {
    rotationMode=BMM_Additive
    translationMode=BMM_Additive
    alphaBoolBlend {blendOption=Linear}
    boneToModify {boneName=jacket_side_base_r}
    translation {x=-2}
   }
   n16 = "Transform (Modify) Bone - Bone: Skirt_corrective"!(ComponentPose=@n15, Alpha=1, Translation="-5,-10,0", Rotation="0,0,0") -> (Pose) {
    rotationMode=BMM_Additive
    rotationSpace=BCS_BoneSpace
    translationMode=BMM_Additive
    translationSpace=BCS_BoneSpace
    alphaBoolBlend {blendOption=Linear}
    boneToModify {boneName=Skirt_corrective}
    translation=<-5 -10>
   }
   n17 = "Transform (Modify) Bone - Bone: armPadPivotHigh_l"(ComponentPose=@n16, Alpha=1, Translation="0,0,1.6", Rotation="0,0,-15", Scale="1,1.15,1") {
    rotationMode=BMM_Additive
    rotationSpace=BCS_BoneSpace
    scaleMode=BMM_Additive
    scaleSpace=BCS_BoneSpace
    translationMode=BMM_Additive
    alphaBoolBlend {blendOption=Linear}
    boneToModify {boneName=armPadPivotHigh_l}
    rotation {roll=-15}
    scale {y=1.15}
    translation {z=1.6}
   }
   n18 = "Transform (Modify) Bone - Bone: armPadPivotHigh_r"(ComponentPose=@n17, Alpha=1, Translation="0,0,1.6", Rotation="0,0,-15", Scale="1,1.15,1") {
    rotationMode=BMM_Additive
    rotationSpace=BCS_BoneSpace
    scaleMode=BMM_Additive
    scaleSpace=BCS_BoneSpace
    translationMode=BMM_Additive
    alphaBoolBlend {blendOption=Linear}
    boneToModify {boneName=armPadPivotHigh_r}
    rotation {roll=-15}
    scale {y=1.15}
    translation {z=1.6}
   }
   Output Pose(Result=Component To Local(ComponentPose=@n18)) {layerGroup=None}
  ]}
 EventGraph {nodes [
   Event BlueprintUpdateAnimation()
   > n1 = Sequence!(execute) -> (then_0, then_1, then_2:Exec, then_3:Exec)
   > n2 = Gravity Override State Function DEPRECATED(CharacterPartAnimInstance=Self-Reference(), GameTimeInSeconds=GetGameTimeInSeconds(), GravityOverrideParams=Get Gravity Override Params())
   > Set Gravity Override(Gravity Override=@n2)
   n3 = Get bIsCrouching()
   n4 = Get bIsDBNO()
   "Set Default(NotDBNOOrCrouch)"(execute=@n1.then_1, "Default(NotDBNOOrCrouch)":bool=AND Boolean(A=NOT Boolean(A=@n3), B=NOT Boolean(A=@n4))) -> (Output_Get:bool)
   > Set SkydiveAndParachute(SkydiveAndParachute:bool=OR Boolean(A=Get bIsSkydiveFloating(), B=Get bIsParachuteOpened())) -> (Output_Get:bool)
   Branch(execute=@n1.then_2, Condition=Get bIsInFrontEnd())
   n5 = GetOwnerFrontEndAnimInstance()
   > Set Custom Idle is Playing Float(Custom Idle is Playing Float:float=GetCurveValue(self=@n5, CurveName=Frontend_Idle_Bypass)) -> (Output_Get:float)
   > Call Trigger FX!(execute:Exec, self:F MED Polished Jade Mind Anim BP Object Reference, Glow Curve:double=GetCurveValue(self=@n5, CurveName=IdleGlow)) -> (then:Exec) {is=title K2Node_CallDelegate}
   Set Is Baseball Sliding(execute=@n1.then_3, Is Baseball Sliding:double="To Float (Boolean)"(InBool=Get bIsBaseballSliding())) -> (Output_Get:double)
  ]}
 Trigger_FX {nodes [Trigger_FX!() -> (then, Glow Curve:double=~) {is=title K2Node_FunctionEntry}]}
]
variables [
 Custom Idle is Playing Float {type=number}
 "Default(NotDBNOOrCrouch)" {type=boolean}
 Gravity Override {type=Vector}
 Gravity Override Params {
  type=GravityOverrideParamsStruct
  default {
   onGround_Moving {gravityOverride=<0 0 -500>}
   skydive_FloatingDown {gravityOverride=<0 -300 -200>, windAmplitude=<100 -150 50>, windFrequency=<1.5 4 4>}
   skydive_GoingUp {gravityOverride=<0 -400 -50>, windAmplitude=<100 -100 50>, windFrequency=<1.5 4 4>}
   skydive_Parachuting {windAmplitude=<50 -250 200>, windFrequency=<1 3 4>}
   tactical_Sprint {gravityOverride=<0 -100 -980>}
  }
 }
 Is Baseball Sliding {type=number}
 SkydiveAndParachute {type=boolean}
 Trigger_FX {default=()}
]
