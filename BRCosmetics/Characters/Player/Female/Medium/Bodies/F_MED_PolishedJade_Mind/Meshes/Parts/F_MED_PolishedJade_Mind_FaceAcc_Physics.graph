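# Physics asset for the face accessory mesh of the PolishedJade Mind female medium cosmetic body, defining simulated bodies for its hair curls.
# A root hair sphere anchors six three-bone sphere chains: left and right curl, frontCurl and sideCurl.
# Constraints link each bone to its parent with twist locked and limited swing, locking one swing axis on sideCurl and some other joints.
# Front and side chains differ between left and right in sphere size and swing limits.
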PhysicsAsset /BRCosmetics/Characters/Player/Female/Medium/Bodies/F_MED_PolishedJade_Mind/Meshes/Parts/F_MED_PolishedJade_Mind_FaceAcc_Physics
bodies {
 hair [{shapeType=Sphere, radius=1.7225, shapeName=hair_sphere}]
 hair_curl01_l [{shapeType=Sphere, radius=0.9867, shapeName=hair_curl01_l_sphere}]
 hair_curl01_r [{shapeType=Sphere, radius=0.9867, shapeName=hair_curl01_r_sphere}]
 hair_curl02_l [{shapeType=Sphere, radius=4.4297, shapeName=hair_curl02_l_sphere}]
 hair_curl02_r [{shapeType=Sphere, radius=4.4297, shapeName=hair_curl02_r_sphere}]
 hair_curl_03_l [{shapeType=Sphere, radius=2.5651, shapeName=hair_curl_03_l_sphere, center=<1.3674 0 0>}]
 hair_curl_03_r [{shapeType=Sphere, radius=2.5651, shapeName=hair_curl_03_r_sphere, center=<-1.8643 0 0>}]
 hair_frontCurl_01_l [{shapeType=Sphere, radius=4.2264, shapeName=hair_frontCurl_01_l_sphere, center=<0.35 -0.3593 -0.097>}]
 hair_frontCurl_01_r [{shapeType=Sphere, radius=2.4523, shapeName=hair_frontCurl_01_r_sphere, center=<0.7995 0.0396 -0.2689>}]
 hair_frontCurl_02_l [{shapeType=Sphere, radius=3.413, shapeName=hair_frontCurl_02_l_sphere, center=<1.1127 -0.5503 0.4535>}]
 hair_frontCurl_02_r [{shapeType=Sphere, radius=1.4833, shapeName=hair_frontCurl_02_r_sphere, center=<0.8533 -0.0221 -0.0123>}]
 hair_frontCurl_03_l [{shapeType=Sphere, radius=2.6359, shapeName=hair_frontCurl_03_l_sphere, center=<1.0871 -0.1224 -1.3397>}]
 hair_frontCurl_03_r [{shapeType=Sphere, radius=1.2665, shapeName=hair_frontCurl_03_r_sphere, center=<0.9701 -0.2903 -0.0274>}]
 hair_sideCurl_01_l [{shapeType=Sphere, radius=2.0863, shapeName=hair_sideCurl_01_l_sphere, center=<1.6987 -0.2155 0.0553>}]
 hair_sideCurl_01_r [{shapeType=Sphere, radius=3.8056, shapeName=hair_sideCurl_01_r_sphere, center=<1.3393 -0.3406 -0.5446>}]
 hair_sideCurl_02_l [{shapeType=Sphere, radius=2.0683, shapeName=hair_sideCurl_02_l_sphere, center=<1.3952 -0.0354 0.0089>}]
 hair_sideCurl_02_r [{shapeType=Sphere, radius=4.0558, shapeName=hair_sideCurl_02_r_sphere, center=<1.3297 0.013 0.0646>}]
 hair_sideCurl_03_l [{shapeType=Sphere, radius=1.5746, shapeName=hair_sideCurl_03_l_sphere, center=<1.3201 -0.4302 -0.3906>}]
 hair_sideCurl_03_r [{shapeType=Sphere, radius=3.7732, shapeName=hair_sideCurl_03_r_sphere, center=<0.8123 0.5513 1.2309>}]
}
constraints [
 {bone1Name=hair_curl01_l, bone2Name=hair, swing1LimitDegrees=6, swing2LimitDegrees=5, twistLimitDegrees=45, twistMotion=Locked}
 {
  bone1Name=hair_curl02_l
  bone2Name=hair_curl01_l
  swing1LimitDegrees=10
  swing2LimitDegrees=10
  swing2Motion=Locked
  twistLimitDegrees=45
  twistMotion=Locked
 }
 {bone1Name=hair_curl_03_l, bone2Name=hair_curl02_l, swing1LimitDegrees=10, swing2LimitDegrees=10, twistLimitDegrees=45, twistMotion=Locked}
 {bone1Name=hair_curl01_r, bone2Name=hair, swing1LimitDegrees=6, swing2LimitDegrees=5, twistLimitDegrees=45, twistMotion=Locked}
 {
  bone1Name=hair_curl02_r
  bone2Name=hair_curl01_r
  swing1LimitDegrees=10
  swing2LimitDegrees=10
  swing2Motion=Locked
  twistLimitDegrees=45
  twistMotion=Locked
 }
 {bone1Name=hair_curl_03_r, bone2Name=hair_curl02_r, swing1LimitDegrees=10, swing2LimitDegrees=10, twistLimitDegrees=45, twistMotion=Locked}
 {bone1Name=hair_frontCurl_01_l, bone2Name=hair, swing1LimitDegrees=3, swing2LimitDegrees=7, twistLimitDegrees=45, twistMotion=Locked}
 {
  bone1Name=hair_frontCurl_02_l
  bone2Name=hair_frontCurl_01_l
  swing1LimitDegrees=3
  swing2LimitDegrees=9
  twistLimitDegrees=45
  twistMotion=Locked
 }
 {
  bone1Name=hair_frontCurl_03_l
  bone2Name=hair_frontCurl_02_l
  swing1LimitDegrees=3
  swing2LimitDegrees=7
  twistLimitDegrees=45
  twistMotion=Locked
 }
 {bone1Name=hair_frontCurl_01_r, bone2Name=hair, swing1LimitDegrees=3, swing2LimitDegrees=3, twistLimitDegrees=45, twistMotion=Locked}
 {
  bone1Name=hair_frontCurl_02_r
  bone2Name=hair_frontCurl_01_r
  swing1LimitDegrees=15
  swing2LimitDegrees=45
  swing2Motion=Locked
  twistLimitDegrees=45
  twistMotion=Locked
 }
 {
  bone1Name=hair_frontCurl_03_r
  bone2Name=hair_frontCurl_02_r
  swing1LimitDegrees=15
  swing2LimitDegrees=45
  swing2Motion=Locked
  twistLimitDegrees=45
  twistMotion=Locked
 }
 {
  bone1Name=hair_sideCurl_01_l
  bone2Name=hair
  swing1LimitDegrees=10
  swing1Motion=Locked
  swing2LimitDegrees=25
  twistLimitDegrees=45
  twistMotion=Locked
 }
 {
  bone1Name=hair_sideCurl_02_l
  bone2Name=hair_sideCurl_01_l
  swing1LimitDegrees=10
  swing1Motion=Locked
  swing2LimitDegrees=30
  twistLimitDegrees=45
  twistMotion=Locked
 }
 {
  bone1Name=hair_sideCurl_03_l
  bone2Name=hair_sideCurl_02_l
  swing1LimitDegrees=10
  swing1Motion=Locked
  swing2LimitDegrees=30
  twistLimitDegrees=45
  twistMotion=Locked
 }
 {
  bone1Name=hair_sideCurl_01_r
  bone2Name=hair
  swing1LimitDegrees=5
  swing1Motion=Locked
  swing2LimitDegrees=5
  twistLimitDegrees=45
  twistMotion=Locked
 }
 {
  bone1Name=hair_sideCurl_02_r
  bone2Name=hair_sideCurl_01_r
  swing1LimitDegrees=3
  swing1Motion=Locked
  swing2LimitDegrees=10
  twistLimitDegrees=45
  twistMotion=Locked
 }
 {
  bone1Name=hair_sideCurl_03_r
  bone2Name=hair_sideCurl_02_r
  swing1LimitDegrees=3
  swing1Motion=Locked
  swing2LimitDegrees=15
  twistLimitDegrees=45
  twistMotion=Locked
 }
]
