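# Post-copy secondary-motion AnimBP for the PolishedJade Mind female face accessory mesh, adding physics to its hair curls.
# It copies the attached parent's pose, then chains zero-gravity, angular-spring AnimDynamics through three-bone front and side curls on both sides.
# A RigidBody node based on the hair bone follows, using the FaceAcc physics asset.
# An older unconnected node set remains, and the deprecated-function-driven Gravity Override variable feeds no node.
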AnimBlueprint /BRCosmetics/Characters/Player/Female/Medium/Bodies/F_MED_PolishedJade_Mind/Meshes/Parts/F_MED_PolishedJade_Mind_FaceAcc_AnimBP
target_skeleton=&/F_MED_PolishedJade_Mind_FaceAcc_Skeleton
graphs [
 AnimGraph {nodes [
   n1 = Anim Dynamics - Bone: hair_sideCurl_01_l!(ComponentPose, Alpha=1, GravityOverride:Vector="0,0,0", ExternalForce="0,0,0") -> (Pose) {
    angularDampingOverride=0.9
    angularSpringConstant=100
    bAngularSpring=true
    bLinearSpring=true
    bOverrideAngularDamping=true
    bOverrideLinearDamping=true
    bUseGravityOverride=true
    linearDampingOverride=0.8
    linearSpringConstant=150
    simulationSpace=BoneRelative
    alphaBoolBlend {blendOption=Linear}
    boundBone {boneName=hair_sideCurl_01_l}
    componentAppliedLinearAccClamp=<10000 10000 10000>
    componentLinearAccScale=<3 3 3>
    componentLinearVelScale=<3 3 3>
    physicsBodyDefinitions [{
      boundBone {boneName=hair_sideCurl_01_l}
      boxExtents=<10 20 20>
      constraintSetup {angularLimitsMax=<0 5 0>, angularLimitsMin=<0 -10 -5>, angularTarget=<1 0 0>}
      localJointOffset=<3 0 0>
     }]
    relativeSpaceBone {boneName=hair}
   }
   n2 = Anim Dynamics - Bone: hair_sideCurl_02_l!(ComponentPose=@n1, Alpha=1, GravityOverride:Vector="0,0,0", ExternalForce="0,0,0") -> (Pose) {
    angularDampingOverride=0.9
    angularSpringConstant=200
    bAngularSpring=true
    bOverrideAngularDamping=true
    bUseGravityOverride=true
    linearDampingOverride=0.8
    linearSpringConstant=150
    simulationSpace=BoneRelative
    alphaBoolBlend {blendOption=Linear}
    boundBone {boneName=hair_sideCurl_02_l}
    componentAppliedLinearAccClamp=<10000 10000 10000>
    componentLinearAccScale=<3 3 3>
    componentLinearVelScale=<3 3 3>
    physicsBodyDefinitions [{
      boundBone {boneName=hair_sideCurl_02_l}
      boxExtents=<10 20 20>
      constraintSetup {angularLimitsMax=<0 10 0>, angularLimitsMin=<0 -15 -5>, angularTarget=<1 0 0>}
      localJointOffset=<3 0 0>
     }]
    relativeSpaceBone {boneName=hair_sideCurl_01_l}
   }
   n3 = Anim Dynamics - Bone: hair_sideCurl_03_l!(ComponentPose=@n2, Alpha=1, GravityOverride:Vector="0,0,0", ExternalForce="0,0,0") -> (Pose) {
    angularDampingOverride=0.9
    angularSpringConstant=200
    bAngularSpring=true
    bOverrideAngularDamping=true
    bOverrideLinearDamping=true
    bUseGravityOverride=true
    linearDampingOverride=0.8
    linearSpringConstant=150
    simulationSpace=BoneRelative
    alphaBoolBlend {blendOption=Linear}
    boundBone {boneName=hair_sideCurl_03_l}
    componentAppliedLinearAccClamp=<10000 10000 10000>
    componentLinearAccScale=<3 3 3>
    componentLinearVelScale=<3 3 3>
    physicsBodyDefinitions [{
      boundBone {boneName=hair_sideCurl_03_l}
      boxExtents=<10 20 20>
      constraintSetup {angularLimitsMax=<0 10 0>, angularLimitsMin=<0 -10 -5>, angularTarget=<1 0 0>}
      localJointOffset=<3 0 0>
     }]
    relativeSpaceBone {boneName=hair_sideCurl_02_l}
   }
   n4 = Anim Dynamics - Bone: hair_sideCurl_01_r!(ComponentPose=@n3, Alpha=1, GravityOverride:Vector="0,0,0", ExternalForce="0,0,0") -> (Pose) {
    angularDampingOverride=0.9
    angularSpringConstant=100
    bAngularSpring=true
    bOverrideAngularDamping=true
    bUseGravityOverride=true
    linearDampingOverride=0.8
    linearSpringConstant=150
    simulationSpace=BoneRelative
    alphaBoolBlend {blendOption=Linear}
    boundBone {boneName=hair_sideCurl_01_r}
    componentAppliedLinearAccClamp=<10000 10000 10000>
    componentLinearAccScale=<3 3 3>
    componentLinearVelScale=<3 3 3>
    physicsBodyDefinitions [{
      boundBone {boneName=hair_sideCurl_01_r}
      boxExtents=<20 40 40>
      constraintSetup {angularLimitsMax=<2 3 0>, angularLimitsMin=<-2 -5 -2>, angularTarget=<1 0 0>}
      localJointOffset=<3 0 0>
     }]
    relativeSpaceBone {boneName=hair}
   }
   n5 = Anim Dynamics - Bone: hair_sideCurl_02_r!(ComponentPose=@n4, Alpha=1, GravityOverride:Vector="0,0,0", ExternalForce="0,0,0") -> (Pose) {
    angularDampingOverride=0.9
    angularSpringConstant=200
    bAngularSpring=true
    bOverrideAngularDamping=true
    bUseGravityOverride=true
    linearDampingOverride=0.8
    linearSpringConstant=150
    simulationSpace=BoneRelative
    alphaBoolBlend {blendOption=Linear}
    boundBone {boneName=hair_sideCurl_02_r}
    componentAppliedLinearAccClamp=<10000 10000 10000>
    componentLinearAccScale=<3 3 3>
    componentLinearVelScale=<3 3 3>
    physicsBodyDefinitions [{
      boundBone {boneName=hair_sideCurl_02_r}
      boxExtents=<20 40 40>
      constraintSetup {angularLimitsMax=<0 10 0>, angularLimitsMin=<0 -15 -2>, angularTarget=<1 0 0>}
      localJointOffset=<3 0 0>
     }]
    relativeSpaceBone {boneName=hair_sideCurl_01_r}
   }
   n6 = Anim Dynamics - Bone: hair_sideCurl_03_r!(ComponentPose=@n5, Alpha=1, GravityOverride:Vector="0,0,0", ExternalForce="0,0,0") -> (Pose) {
    angularDampingOverride=0.9
    angularSpringConstant=200
    bAngularSpring=true
    bOverrideAngularDamping=true
    bUseGravityOverride=true
    linearDampingOverride=0.8
    linearSpringConstant=150
    simulationSpace=BoneRelative
    alphaBoolBlend {blendOption=Linear}
    boundBone {boneName=hair_sideCurl_03_r}
    componentAppliedLinearAccClamp=<10000 10000 10000>
    componentLinearAccScale=<3 3 3>
    componentLinearVelScale=<3 3 3>
    physicsBodyDefinitions [{
      boundBone {boneName=hair_sideCurl_03_r}
      boxExtents=<20 40 40>
      constraintSetup {angularLimitsMax=<0 10 0>, angularLimitsMin=<0 -10 -5>, angularTarget=<1 0 0>}
      localJointOffset=<3 0 0>
     }]
    relativeSpaceBone {boneName=hair_sideCurl_02_r}
   }
   n7 = Anim Dynamics - Bone: hair_frontCurl_01_r!(ComponentPose=@n6, Alpha=1, GravityOverride:Vector="0,0,0", ExternalForce="0,0,0") -> (Pose) {
    angularDampingOverride=0.9
    angularSpringConstant=50
    bAngularSpring=true
    bOverrideAngularDamping=true
    bUseGravityOverride=true
    linearDampingOverride=0.8
    linearSpringConstant=150
    simulationSpace=BoneRelative
    alphaBoolBlend {blendOption=Linear}
    boundBone {boneName=hair_frontCurl_01_r}
    componentAppliedLinearAccClamp=<10000 10000 10000>
    componentLinearAccScale=<3 3 3>
    componentLinearVelScale=<3 3 3>
    physicsBodyDefinitions [{
      boundBone {boneName=hair_frontCurl_01_r}
      boxExtents=<15 30 30>
      constraintSetup {angularLimitsMax=<0 5 10>, angularLimitsMin=<0 0 -10>, angularTarget=<1 0 0>}
      localJointOffset=<0 1 0>
     }]
    relativeSpaceBone {boneName=hair}
   }
   n8 = Anim Dynamics - Bone: hair_frontCurl_02_r!(ComponentPose=@n7, Alpha=1, GravityOverride:Vector="0,0,0", ExternalForce="0,0,0") -> (Pose) {
    angularDampingOverride=0.9
    angularSpringConstant=100
    bAngularSpring=true
    bOverrideAngularDamping=true
    bUseGravityOverride=true
    linearDampingOverride=0.8
    linearSpringConstant=150
    simulationSpace=BoneRelative
    alphaBoolBlend {blendOption=Linear}
    boundBone {boneName=hair_frontCurl_02_r}
    componentAppliedLinearAccClamp=<10000 10000 10000>
    componentLinearAccScale=<3 3 3>
    componentLinearVelScale=<3 3 3>
    physicsBodyDefinitions [{
      boundBone {boneName=hair_frontCurl_02_r}
      boxExtents=<15 30 30>
      constraintSetup {angularLimitsMax=<0 20 15>, angularLimitsMin=<0 0 -15>, angularTarget=<1 0 0>}
      localJointOffset=<0 2 0>
     }]
    relativeSpaceBone {boneName=hair_frontCurl_01_r}
   }
   n9 = Anim Dynamics - Bone: hair_frontCurl_03_r!(ComponentPose=@n8, Alpha=1, GravityOverride:Vector="0,0,0", ExternalForce="0,0,0") -> (Pose) {
    angularDampingOverride=0.9
    angularSpringConstant=250
    bAngularSpring=true
    bOverrideAngularDamping=true
    bUseGravityOverride=true
    linearDampingOverride=0.8
    linearSpringConstant=150
    simulationSpace=BoneRelative
    alphaBoolBlend {blendOption=Linear}
    boundBone {boneName=hair_frontCurl_03_r}
    componentAppliedLinearAccClamp=<10000 10000 10000>
    componentLinearAccScale=<3 3 3>
    componentLinearVelScale=<3 3 3>
    physicsBodyDefinitions [{
      boundBone {boneName=hair_frontCurl_03_r}
      boxExtents=<15 30 30>
      constraintSetup {angularLimitsMax=<0 20 10>, angularLimitsMin=<0 0 -30>, angularTarget=<1 0 0>}
      localJointOffset=<0 2 0>
     }]
    relativeSpaceBone {boneName=hair_frontCurl_02_r}
   }
   Anim Dynamics - Bone: hair_frontCurl_01_l!(ComponentPose=@n9, Alpha=1, GravityOverride:Vector="0,0,0", ExternalForce="0,0,0") -> (Pose) {
    angularDampingOverride=0.9
    angularSpringConstant=50
    bAngularSpring=true
    bOverrideAngularDamping=true
    bUseGravityOverride=true
    linearDampingOverride=0.8
    linearSpringConstant=150
    simulationSpace=BoneRelative
    alphaBoolBlend {blendOption=Linear}
    boundBone {boneName=hair_frontCurl_01_l}
    componentAppliedLinearAccClamp=<10000 10000 10000>
    componentLinearAccScale=<3 3 3>
    componentLinearVelScale=<3 3 3>
    physicsBodyDefinitions [{
      boundBone {boneName=hair_frontCurl_01_l}
      boxExtents=<15 30 30>
      constraintSetup {angularLimitsMax=<0 5 0>, angularLimitsMin=<-1 0 -5>, angularTarget=<1 0 0>}
      localJointOffset=<3 0 0>
     }]
    relativeSpaceBone {boneName=hair}
   }
   n10 = Anim Dynamics - Bone: hair_frontCurl_02_l!(ComponentPose, Alpha=1, GravityOverride:Vector="0,0,0", ExternalForce="0,0,0") -> (Pose) {
    angularDampingOverride=0.9
    angularSpringConstant=100
    bAngularSpring=true
    bOverrideAngularDamping=true
    bUseGravityOverride=true
    linearDampingOverride=0.8
    linearSpringConstant=150
    simulationSpace=BoneRelative
    alphaBoolBlend {blendOption=Linear}
    boundBone {boneName=hair_frontCurl_02_l}
    componentAppliedLinearAccClamp=<10000 10000 10000>
    componentLinearAccScale=<3 3 3>
    componentLinearVelScale=<3 3 3>
    physicsBodyDefinitions [{
      boundBone {boneName=hair_frontCurl_02_l}
      boxExtents=<15 30 30>
      constraintSetup {angularLimitsMax=<0 10 2>, angularLimitsMin=<-3 0 -2>, angularTarget=<1 0 0>}
      localJointOffset=<0 0 2>
     }]
    relativeSpaceBone {boneName=hair_frontCurl_01_l}
   }
   Anim Dynamics - Bone: hair_frontCurl_03_l!(ComponentPose=@n10, Alpha=1, GravityOverride:Vector="0,0,0", ExternalForce="0,0,0") -> (Pose) {
    angularDampingOverride=0.9
    angularSpringConstant=250
    bAngularSpring=true
    bOverrideAngularDamping=true
    bUseGravityOverride=true
    linearDampingOverride=0.8
    linearSpringConstant=150
    simulationSpace=BoneRelative
    alphaBoolBlend {blendOption=Linear}
    boundBone {boneName=hair_frontCurl_03_l}
    componentAppliedLinearAccClamp=<10000 10000 10000>
    componentLinearAccScale=<3 3 3>
    componentLinearVelScale=<3 3 3>
    physicsBodyDefinitions [{
      boundBone {boneName=hair_frontCurl_03_l}
      boxExtents=<15 30 30>
      constraintSetup {angularLimitsMax=<0 5 3>, angularLimitsMin=<-2 -2 -1>, angularTarget=<1 0 0>}
      localJointOffset=<0 0 0>
     }]
    relativeSpaceBone {boneName=hair_frontCurl_02_l}
   }
   RigidBody!(ComponentPose, Alpha, OverrideWorldGravity:Vector="0,0,-980", ExternalForce) -> (Pose) {
    bClampLinearTranslationLimitToRefPose=true
    bOverrideWorldGravity=true
    overridePhysicsAsset=&/F_MED_PolishedJade_Mind_FaceAcc_Physics
    simulationSpace=BaseBoneSpace
    alphaBoolBlend {blendOption=Linear}
    baseBoneRef {boneName=neck_01}
    componentLinearAccScale=<0.5 0.5 0.3>
    componentLinearVelScale=<0.5 0.5 0.3>
    overrideWorldGravity {z=-980}
    simSpaceSettings {maxAngularAcceleration=500, maxAngularVelocity=3, maxLinearAcceleration=400, maxLinearVelocity=250, velocityScaleZ=0, worldAlpha=0.7}
   }
   n11 = Anim Dynamics - Bone: hair_sideCurl_01_l(
    ComponentPose=Local To Component(LocalPose=Copy Pose From Mesh() {bUseAttachedParent=true})
    Alpha=1
    GravityOverride="0,0,0"
    ExternalForce="0,0,0"
   ) {
    angularDampingOverride=0.8
    angularSpringConstant=100
    bAngularSpring=true
    bLinearSpring=true
    bOverrideAngularDamping=true
    bOverrideLinearDamping=true
    bUseGravityOverride=true
    linearDampingOverride=0.8
    linearSpringConstant=150
    simulationSpace=BoneRelative
    alphaBoolBlend {blendOption=Linear}
    boundBone {boneName=hair_sideCurl_01_l}
    componentLinearAccScale=<5 5 5>
    componentLinearVelScale=<5 5 5>
    physicsBodyDefinitions [{
      boundBone {boneName=hair_sideCurl_01_l}
      boxExtents=<10 10 10>
      constraintSetup {angularLimitsMax=<2 5 0>, angularLimitsMin=<-2 -5 -5>, angularTarget=<1 0 0>}
      localJointOffset=<0.1 0 0>
     }]
    relativeSpaceBone {boneName=hair}
   }
   n12 = Anim Dynamics - Bone: hair_sideCurl_02_l(ComponentPose=@n11, Alpha=1, GravityOverride="0,0,0", ExternalForce="0,0,0") {
    angularDampingOverride=0.9
    angularSpringConstant=200
    bAngularSpring=true
    bOverrideAngularDamping=true
    bUseGravityOverride=true
    linearDampingOverride=0.8
    linearSpringConstant=150
    simulationSpace=BoneRelative
    alphaBoolBlend {blendOption=Linear}
    boundBone {boneName=hair_sideCurl_02_l}
    componentLinearAccScale=<3 3 3>
    componentLinearVelScale=<3 3 3>
    physicsBodyDefinitions [{
      boundBone {boneName=hair_sideCurl_02_l}
      boxExtents=<10 10 10>
      constraintSetup {angularLimitsMax=<2 10 2>, angularLimitsMin=<-2 -25 -5>, angularTarget=<1 0 0>}
      localJointOffset=<0.1 0 0>
     }]
    relativeSpaceBone {boneName=hair_sideCurl_01_l}
   }
   n13 = Anim Dynamics - Bone: hair_sideCurl_03_l(ComponentPose=@n12, Alpha=1, GravityOverride="0,0,0", ExternalForce="0,0,0") {
    angularDampingOverride=0.8
    angularSpringConstant=200
    bAngularSpring=true
    bOverrideAngularDamping=true
    bOverrideLinearDamping=true
    bUseGravityOverride=true
    linearDampingOverride=0.8
    linearSpringConstant=150
    simulationSpace=BoneRelative
    alphaBoolBlend {blendOption=Linear}
    boundBone {boneName=hair_sideCurl_03_l}
    componentLinearAccScale=<3 3 3>
    componentLinearVelScale=<3 3 3>
    physicsBodyDefinitions [{
      boundBone {boneName=hair_sideCurl_03_l}
      boxExtents=<10 10 10>
      constraintSetup {angularLimitsMax=<2 10 3>, angularLimitsMin=<-2 -10 -5>, angularTarget=<1 0 0>}
      localJointOffset=<0.1 0 0>
     }]
    relativeSpaceBone {boneName=hair_sideCurl_02_l}
   }
   n14 = Anim Dynamics - Bone: hair_sideCurl_01_r(ComponentPose=@n13, Alpha=1, GravityOverride="0,0,0", ExternalForce="0,0,0") {
    angularDampingOverride=0.9
    angularSpringConstant=100
    bAngularSpring=true
    bOverrideAngularDamping=true
    bUseGravityOverride=true
    linearDampingOverride=0.8
    linearSpringConstant=150
    simulationSpace=BoneRelative
    alphaBoolBlend {blendOption=Linear}
    boundBone {boneName=hair_sideCurl_01_r}
    componentLinearAccScale=<5 5 5>
    componentLinearVelScale=<5 5 5>
    physicsBodyDefinitions [{
      boundBone {boneName=hair_sideCurl_01_r}
      boxExtents=<10 10 10>
      constraintSetup {angularLimitsMax=<2 3 2>, angularLimitsMin=<-2 -1 -2>, angularTarget=<1 0 0>}
      localJointOffset=<0.1 0 0>
     }]
    relativeSpaceBone {boneName=hair}
   }
   n15 = Anim Dynamics - Bone: hair_sideCurl_02_r(ComponentPose=@n14, Alpha=1, GravityOverride="0,0,0", ExternalForce="0,0,0") {
    angularDampingOverride=0.9
    angularSpringConstant=200
    bAngularSpring=true
    bOverrideAngularDamping=true
    bUseGravityOverride=true
    linearDampingOverride=0.8
    linearSpringConstant=150
    simulationSpace=BoneRelative
    alphaBoolBlend {blendOption=Linear}
    boundBone {boneName=hair_sideCurl_02_r}
    componentLinearAccScale=<3 3 3>
    componentLinearVelScale=<3 3 3>
    physicsBodyDefinitions [{
      boundBone {boneName=hair_sideCurl_02_r}
      boxExtents=<10 10 10>
      constraintSetup {angularLimitsMax=<2 10 2>, angularLimitsMin=<-2 -15 -2>, angularTarget=<1 0 0>}
      localJointOffset=<0.1 0 0>
     }]
    relativeSpaceBone {boneName=hair_sideCurl_01_r}
   }
   n16 = Anim Dynamics - Bone: hair_sideCurl_03_r(ComponentPose=@n15, Alpha=1, GravityOverride="0,0,0", ExternalForce="0,0,0") {
    angularDampingOverride=0.9
    angularSpringConstant=200
    bAngularSpring=true
    bOverrideAngularDamping=true
    bUseGravityOverride=true
    linearDampingOverride=0.8
    linearSpringConstant=150
    simulationSpace=BoneRelative
    alphaBoolBlend {blendOption=Linear}
    boundBone {boneName=hair_sideCurl_03_r}
    componentLinearAccScale=<3 3 3>
    componentLinearVelScale=<3 3 3>
    physicsBodyDefinitions [{
      boundBone {boneName=hair_sideCurl_03_r}
      boxExtents=<10 10 10>
      constraintSetup {angularLimitsMax=<2 10 5>, angularLimitsMin=<-2 -10 -5>, angularTarget=<1 0 0>}
      localJointOffset=<0.1 0 0>
     }]
    relativeSpaceBone {boneName=hair_sideCurl_02_r}
   }
   n17 = Anim Dynamics - Bone: hair_frontCurl_01_r(ComponentPose=@n16, Alpha=1, GravityOverride="0,0,0", ExternalForce="0,0,0") {
    angularDampingOverride=0.8
    angularSpringConstant=50
    bAngularSpring=true
    bOverrideAngularDamping=true
    bUseGravityOverride=true
    linearDampingOverride=0.8
    linearSpringConstant=150
    simulationSpace=BoneRelative
    alphaBoolBlend {blendOption=Linear}
    boundBone {boneName=hair_frontCurl_01_r}
    componentLinearAccScale=<3 3 5>
    componentLinearVelScale=<3 3 5>
    physicsBodyDefinitions [{
      boundBone {boneName=hair_frontCurl_01_r}
      boxExtents=<15 15 15>
      constraintSetup {angularLimitsMax=<0 5 10>, angularLimitsMin=<0 0 -10>, angularTarget=<1 0 0>}
      localJointOffset=<0.1 0 0.1>
     }]
    relativeSpaceBone {boneName=hair}
   }
   n18 = Anim Dynamics - Bone: hair_frontCurl_02_r(ComponentPose=@n17, Alpha=1, GravityOverride="0,0,0", ExternalForce="0,0,0") {
    angularDampingOverride=0.8
    angularSpringConstant=100
    bAngularSpring=true
    bOverrideAngularDamping=true
    bUseGravityOverride=true
    linearDampingOverride=0.8
    linearSpringConstant=150
    simulationSpace=BoneRelative
    alphaBoolBlend {blendOption=Linear}
    boundBone {boneName=hair_frontCurl_02_r}
    componentLinearAccScale=<3 3 3>
    componentLinearVelScale=<3 3 3>
    physicsBodyDefinitions [{
      boundBone {boneName=hair_frontCurl_02_r}
      boxExtents=<12 12 12>
      constraintSetup {angularLimitsMax=<0 20 15>, angularLimitsMin=<0 0 -15>, angularTarget=<1 0 0>}
      localJointOffset=<0.1 0 0.1>
     }]
    relativeSpaceBone {boneName=hair_frontCurl_01_r}
   }
   n19 = Anim Dynamics - Bone: hair_frontCurl_03_r(ComponentPose=@n18, Alpha=1, GravityOverride="0,0,0", ExternalForce="0,0,0") {
    angularDampingOverride=0.9
    angularSpringConstant=250
    bAngularSpring=true
    bOverrideAngularDamping=true
    bUseGravityOverride=true
    linearDampingOverride=0.8
    linearSpringConstant=150
    simulationSpace=BoneRelative
    alphaBoolBlend {blendOption=Linear}
    boundBone {boneName=hair_frontCurl_03_r}
    componentLinearAccScale=<3 3 3>
    componentLinearVelScale=<3 3 3>
    physicsBodyDefinitions [{
      boundBone {boneName=hair_frontCurl_03_r}
      boxExtents=<10 10 10>
      constraintSetup {angularLimitsMax=<0 20 10>, angularLimitsMin=<0 0 -30>, angularTarget=<1 0 0>}
      localJointOffset=<0.1 0 0.1>
     }]
    relativeSpaceBone {boneName=hair_frontCurl_02_r}
   }
   n20 = Anim Dynamics - Bone: hair_frontCurl_01_l(ComponentPose=@n19, Alpha=1, GravityOverride="0,0,0", ExternalForce="0,0,0") {
    angularDampingOverride=0.8
    angularSpringConstant=50
    bAngularSpring=true
    bOverrideAngularDamping=true
    bUseGravityOverride=true
    linearDampingOverride=0.8
    linearSpringConstant=150
    simulationSpace=BoneRelative
    alphaBoolBlend {blendOption=Linear}
    boundBone {boneName=hair_frontCurl_01_l}
    componentLinearAccScale=<3 3 5>
    componentLinearVelScale=<3 3 5>
    physicsBodyDefinitions [{
      boundBone {boneName=hair_frontCurl_01_l}
      boxExtents=<10 10 10>
      constraintSetup {angularLimitsMax=<5 5 1>, angularLimitsMin=<-5 0 -10>, angularTarget=<1 0 0>}
      localJointOffset=<-0.1 0 0>
     }]
    relativeSpaceBone {boneName=hair}
   }
   n21 = Anim Dynamics - Bone: hair_frontCurl_02_l(ComponentPose=@n20, Alpha=1, GravityOverride="0,0,0", ExternalForce="0,0,0") {
    angularDampingOverride=0.8
    angularSpringConstant=100
    bAngularSpring=true
    bOverrideAngularDamping=true
    bUseGravityOverride=true
    linearDampingOverride=0.8
    linearSpringConstant=150
    simulationSpace=BoneRelative
    alphaBoolBlend {blendOption=Linear}
    boundBone {boneName=hair_frontCurl_02_l}
    componentLinearAccScale=<3 3 3>
    componentLinearVelScale=<3 3 3>
    physicsBodyDefinitions [{
      boundBone {boneName=hair_frontCurl_02_l}
      boxExtents=<12 12 12>
      constraintSetup {angularLimitsMax=<5 20 5>, angularLimitsMin=<-5 0 -5>, angularTarget=<1 0 0>}
      localJointOffset=<0 -0.1 0>
     }]
    relativeSpaceBone {boneName=hair_frontCurl_01_l}
   }
   n22 = Anim Dynamics - Bone: hair_frontCurl_03_l(ComponentPose=@n21, Alpha=1, GravityOverride="0,0,0", ExternalForce="0,0,0") {
    angularDampingOverride=0.9
    angularSpringConstant=250
    bAngularSpring=true
    bOverrideAngularDamping=true
    bUseGravityOverride=true
    linearDampingOverride=0.8
    linearSpringConstant=150
    simulationSpace=BoneRelative
    alphaBoolBlend {blendOption=Linear}
    boundBone {boneName=hair_frontCurl_03_l}
    componentLinearAccScale=<3 3 3>
    componentLinearVelScale=<3 3 3>
    physicsBodyDefinitions [{
      boundBone {boneName=hair_frontCurl_03_l}
      boxExtents=<10 10 10>
      constraintSetup {angularLimitsMax=<10 5 1>, angularLimitsMin=<-10 -5 -10>, angularTarget=<1 0 0>}
      localJointOffset=<0 0.1 0>
     }]
    relativeSpaceBone {boneName=hair_frontCurl_02_l}
   }
   n23 = RigidBody(ComponentPose=@n22, OverrideWorldGravity="0,0,-980") {
    bOverrideWorldGravity=true
    overridePhysicsAsset=&F_MED_PolishedJade_Mind_FaceAcc_Physics
    simulationSpace=BaseBoneSpace
    alphaBoolBlend {blendOption=Linear}
    baseBoneRef {boneName=hair}
    componentLinearAccScale=<0.25 0.25 0.5>
    componentLinearVelScale=<0.25 0.25 0.5>
    overrideWorldGravity {z=-980}
    simSpaceSettings {maxAngularAcceleration=400, maxAngularVelocity=60, maxLinearAcceleration=400, maxLinearVelocity=30, velocityScaleZ=0.3, worldAlpha=0.5}
   }
   Output Pose(Result=Component To Local(ComponentPose=@n23)) {layerGroup=None}
  ]}
 EventGraph {nodes [
   Event BlueprintUpdateAnimation()
   > n1 = Gravity Override State Function DEPRECATED(CharacterPartAnimInstance=Self-Reference(), GameTimeInSeconds=GetGameTimeInSeconds(), GravityOverrideParams=Get Gravity Override Params())
   > Set Gravity Override(Gravity Override=@n1)
  ]}
]
variables [
 Gravity Override {type=Vector}
 Gravity Override Params {type=GravityOverrideParamsStruct}
]
